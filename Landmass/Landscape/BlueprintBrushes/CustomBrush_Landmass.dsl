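; CustomBrush_Landmass is a Landscape Blueprint brush that draws heightmap and weightmap edits from spline shapes, spline meshes, static meshes or a simple material.
; BrushSetup picks materials, builds meshes, prepares render targets and bakes an optional elevation curve texture.
; RenderLayer rasterizes splines or captures mesh depth through LandmassBrushManager, running Jump Flood distance fields for spline types and caching the result.
; Helper functions adjust spline tangents and recenter the actor.
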(event EventBeginPlay)

(event Collision|EventActorBeginOverlap
  (params (OtherActor :type Actor%20Object%20Reference)))

(event EventInitialize
  (params (InLandscapeTransform :type Transform%20%28by%20ref%29) (InLandscapeSize :type Int%20Point%20Structure%20%28by%20ref%29) (InLandscapeRenderTargetSize :type Int%20Point%20Structure%20%28by%20ref%29))
  (Development|PrintString "Brush Initialize")
  (Variables|5%29Debug|SetDistanceFieldIsCached false)
  (Landmass|SpawnOrUpdateManager (Variables|5%29Debug|GetBrushManager) InLandscapeTransform InLandscapeSize InLandscapeRenderTargetSize))

(event EventTick
  (params (DeltaSeconds :type Float%20%28single-precision%29))
  (if (Variables|Default|GetT)
    (Spline|SetLocationatSplinePoint (Variables|Default|GetSpline) 3 (+ (Spline|GetLocationatSplinePoint (Variables|Default|GetSpline) 3 "Local") "0.000000,10.000000, 0") "Local" true)
    (elif (Variables|5%29Debug|GetForce%20Tick%20Perf%20Test)
      (Landscape|RequestLandscapeUpdate false))))

(fn AutoSplineTangents
  (for _index (range (Spline|GetNumberOfSplinePoints (Variables|Default|GetSpline)))
    (Spline|SetTangentatSplinePoint (Variables|Default|GetSpline) _index (/ (- (Spline|GetLocationatSplinePoint (Variables|Default|GetSpline) (+ _index 1) "Local") (select (== _index 0) (+ (- (Spline|GetLocationatSplinePoint (Variables|Default|GetSpline) _index "Local") (Spline|GetLocationatSplinePoint (Variables|Default|GetSpline) (+ _index 1) "Local")) (Spline|GetLocationatSplinePoint (Variables|Default|GetSpline) _index "Local")) (Spline|GetLocationatSplinePoint (Variables|Default|GetSpline) (- _index 1) "Local"))) 2.0) "Local" true))
  (Variables|5%29Debug|SetDistanceFieldIsCached false)
  (Landscape|RequestLandscapeUpdate false))

(fn BrushSetup
  (bind _splinemeshesforrender (Variables|Default|GetSplineMeshesForRender))
  (bind _splinemeshes (Variables|Default|GetSplineMeshes))
  (bind _splinemeshbp (Variables|5%29Debug|GetSplineMeshBP))
  (bind _splinemeshbp_2 (Variables|5%29Debug|GetSplineMeshBP))
  (Utilities|FlowControl|Sequence
    (:then_0
      (Utilities|Array|Clear (Variables|Settings|GetAffectedWeightmapLayers))
      (bind _keys (Utilities|Map|Keys (Variables|1%29BrushSettings|PaintLayers|GetPaint%20Layer%20Target%20Settings)))
      (Variables|Settings|SetAffectWeightmap (> (Utilities|Array|Length _keys) 0))
      (if (> (Utilities|Array|Length _keys) 0)
        (for _array_element _keys
          (Utilities|Array|AddUnique (Variables|Settings|GetAffectedWeightmapLayers) _array_element))))
    (:then_1
      (Variables|1%29BrushSettings|Material|SetBrush%20Material (Utilities|Select :Index (Variables|1%29BrushSettings|GetBrush%20Type) :NewEnumerator0 (Utilities|Select :Index (Variables|1%29BrushSettings|Falloff|GetFalloff%20Mode) :NewEnumerator0 "/Landmass/Landscape/BlueprintBrushes/Materials/LandmassBrush_Angle.LandmassBrush_Angle" :NewEnumerator1 "/Landmass/Landscape/BlueprintBrushes/Materials/LandmassBrush_Width.LandmassBrush_Width") :NewEnumerator1 (Utilities|Select :Index (Variables|1%29BrushSettings|Falloff|GetFalloff%20Mode) :NewEnumerator0 "/Landmass/Landscape/BlueprintBrushes/Materials/LandmassBrush_Angle.LandmassBrush_Angle" :NewEnumerator1 "/Landmass/Landscape/BlueprintBrushes/Materials/LandmassBrush_Width.LandmassBrush_Width") :NewEnumerator5 (Utilities|Select :Index (Variables|1%29BrushSettings|Falloff|GetFalloff%20Mode) :NewEnumerator0 "/Landmass/Landscape/BlueprintBrushes/Materials/LandmassBrush_Angle.LandmassBrush_Angle" :NewEnumerator1 "/Landmass/Landscape/BlueprintBrushes/Materials/LandmassBrush_Width.LandmassBrush_Width") :NewEnumerator7 "/Landmass/Landscape/BlueprintBrushes/Materials/M_SimpleBrush.M_SimpleBrush"))
      (Variables|5%29Debug|SetUsedMaterial (select (Variables|1%29BrushSettings|Material|GetOverride%20Material) (Variables|1%29BrushSettings|Material|GetCustom%20Override%20Material) (Variables|1%29BrushSettings|Material|GetBrush%20Material)))
      (Variables|5%29Debug|SetCurlparams (Math|Color|MakeColor (Variables|2%29Effects|CurlNoise|GetCurl%201%20Tiling) (Variables|2%29Effects|CurlNoise|GetCurl%201%20Strength) (Variables|2%29Effects|CurlNoise|GetCurl%202%20Tiling) (Variables|2%29Effects|CurlNoise|GetCurl%202%20Strength)))
      (Variables|1%29BrushSettings|Falloff|SetCap%20Shape (Utilities|Select :Index (Variables|1%29BrushSettings|Falloff|GetFalloff%20Mode) :NewEnumerator0 (Variables|1%29BrushSettings|Falloff|GetCap%20Shape) :NewEnumerator1 true))
      (Utilities|FlowControl|Sequence
        (:then_0
          ;[1 1 1] Mesh Setup
          (switch Utilities|FlowControl|Switch|SwitchonBrushTypes (Variables|1%29BrushSettings|GetBrush%20Type)
            (:NewEnumerator1
              (Utilities|IsValid (Variables|1%29BrushSettings|SplineMesh|GetSpline%20Mesh)
                (:Is%20Valid
                  (bind _output_get (Variables|5%29Debug|SetSpline%20Length (Spline|GetSplineLength (Variables|Default|GetSpline))))
                  (if (> (Variables|5%29Debug|GetSpline%20Length) (Variables|1%29BrushSettings|SplineMesh|GetDefault%20Mesh%20Length))
                    ;; Construct Splines
                    (Utilities|IsValid (Variables|1%29BrushSettings|SplineMesh|GetSpline%20Mesh)
                      (:Is%20Valid
                        (for _array_element_2 _splinemeshes
                          (Components|DestroyComponent _array_element_2))
                        (Utilities|Array|Clear _splinemeshes)
                        (for _array_element_3 _splinemeshesforrender
                          (Components|DestroyComponent _array_element_3))
                        (Utilities|Array|Clear _splinemeshesforrender)
                        (if (and (Variables|1%29BrushSettings|SplineMesh|GetShow%20Splines) (Utilities|IsValid (Variables|1%29BrushSettings|SplineMesh|GetFlatten%20Mesh) :!class /Script/Engine.KismetSystemLibrary))
                          (Utilities|FlowControl|Sequence
                            (:then_0
                              (Utilities|IsValid _splinemeshbp
                                (:Is%20Valid
                                  (if (Utilities|Equal(Object) (Actor|GetOwner _splinemeshbp) self)
                                    (Class|SplineMeshChild|UpdateValues _splinemeshbp_2 (Variables|Default|GetSpline) (Variables|1%29BrushSettings|SplineMesh|GetSpline%20Material) (Variables|1%29BrushSettings|SplineMesh|GetSpline%20Mesh) (Variables|1%29BrushSettings|SplineMesh|GetSpline%20Mesh%20Scale) (Variables|1%29BrushSettings|SplineMesh|GetDefault%20Mesh%20Length))
                                    (Class|SplineMeshChild|SetupSplineMeshComponents _splinemeshbp_2)
                                    (Transformation|SetActorLocation _splinemeshbp_2 (Transformation|GetActorLocation) false false)
                                    (else
                                      _)))
                                (:Is%20Not%20Valid
                                  _))
                              (bind _returnvalue (Game|SpawnActorfromClass "/Landmass/Landscape/BlueprintBrushes/SplineMeshChild.SplineMeshChild_C" (Math|Conversions|ToTransform(Vector) (Transformation|GetActorLocation)) "Undefined" "OverrideRootScale" self (Variables|Default|GetSpline) (Variables|1%29BrushSettings|SplineMesh|GetSpline%20Material) (Variables|1%29BrushSettings|SplineMesh|GetSpline%20Mesh) (Variables|1%29BrushSettings|SplineMesh|GetSpline%20Mesh%20Scale) (Variables|1%29BrushSettings|SplineMesh|GetDefault%20Mesh%20Length) 0))
                              (bind _output_get_2 (Variables|5%29Debug|SetSplineMeshBP _returnvalue))
                              (Transformation|AttachActorToActor _output_get_2 self "None" "KeepWorld" "KeepWorld" "KeepWorld" false)
                              (Class|SplineMeshChild|SetupSplineMeshComponents _output_get_2))
                            (:then_1
                              _))
                          (else
                            (Utilities|FlowControl|Sequence
                              (:then_0
                                (for _index (range (Math|Float|Floor (/ (Variables|5%29Debug|GetSpline%20Length) (Variables|1%29BrushSettings|SplineMesh|GetDefault%20Mesh%20Length))))
                                  (bind _component (Utilities|DistanceBasedSplineMesh _index (Variables|1%29BrushSettings|SplineMesh|GetSpline%20Mesh) (Variables|1%29BrushSettings|SplineMesh|GetSpline%20Material) (Variables|1%29BrushSettings|SplineMesh|GetSpline%20Mesh%20Scale)))
                                  (Utilities|Array|Add (Variables|Default|GetSplineMeshes) _component)
                                  (Rendering|SetVisibility _component (Variables|1%29BrushSettings|SplineMesh|GetShow%20Splines) false)
                                  (Development|SetHiddeninGame _component false false)
                                  (Utilities|IsValid (Variables|1%29BrushSettings|SplineMesh|GetFlatten%20Mesh)
                                    (:Is%20Not%20Valid
                                      (Utilities|Array|Add (Variables|Default|GetSplineMeshesForRender) _component)))))
                              (:then_1
                                _))))
                        (Utilities|IsValid (Variables|1%29BrushSettings|SplineMesh|GetFlatten%20Mesh)
                          (:Is%20Valid
                            (for _index_2 (range (Math|Float|Floor (/ (Variables|5%29Debug|GetSpline%20Length) (Variables|1%29BrushSettings|SplineMesh|GetDefault%20Mesh%20Length))))
                              (bind _component_2 (Utilities|DistanceBasedSplineMesh _index_2 (Variables|1%29BrushSettings|SplineMesh|GetFlatten%20Mesh) 0 (Variables|1%29BrushSettings|SplineMesh|GetFlatten%20Mesh%20Scale)))
                              (Utilities|Array|Add (Variables|Default|GetSplineMeshesForRender) _component_2)
                              (Rendering|SetVisibility _component_2 (Variables|1%29BrushSettings|SplineMesh|GetShow%20Flatten%20Mesh) false))))))))))
            (:NewEnumerator5
              (Utilities|IsValid (Variables|1%29BrushSettings|StaticMesh|GetStatic%20Mesh)
                (:Is%20Valid
                  (Components|StaticMesh|SetStaticMesh (Variables|Default|GetBrush%20Mesh) (Variables|1%29BrushSettings|StaticMesh|GetStatic%20Mesh)))))))
        (:then_1
          (Class|LandmassBrushManager|SetMPCParams (Variables|5%29Debug|GetBrushManager))
          (Utilities|CreateMIDIfNeeded (Variables|5%29Debug|GetCopyRTMID) "/Landmass/Landscape/BlueprintBrushes/Materials/CopyRT_WithAlpha.CopyRT_WithAlpha"
            (:out
              (Utilities|CreateMIDIfNeeded (Variables|5%29Debug|GetDrawBrushMID) (Variables|5%29Debug|GetUsedMaterial)
                (:out
                  (Utilities|CreateMIDIfNeeded (Variables|5%29Debug|GetWeightmapMID) "/Landmass/Landscape/BlueprintBrushes/Materials/LandmassBrush_Weightmap.LandmassBrush_Weightmap"
                    (:out
                      (Utilities|CreateMIDIfNeeded (Variables|5%29Debug|GetRender%20Distance%20Field%20MID) "/Landmass/Landscape/BlueprintBrushes/Materials/CacheDistanceField_RG8.CacheDistanceField_RG8"
                        (:out
                          (Utilities|CreateRTIfNullOrChanged (Variables|5%29Debug|GetDistance%20Field%20Cache))
                          (Utilities|SetMIDParams))))))))))
        (:then_2
          ;[1 1 1] Curve Setup
          (if (Variables|2%29Effects|Curves|GetUse%20Curve%20Channel)
            (Utilities|IsValid (Variables|2%29Effects|Curves|GetElevation%20Curve%20Asset)
              (:Is%20Valid
                (Utilities|IsValid (Variables|2%29Effects|Curves|GetCurve%20RT)
                  (:Is%20Valid
                    (if (Variables|5%29Debug|GetCurveTextureRendered)
                      _))
                  (:Is%20Not%20Valid
                    (bind _returnvalue_2 (Rendering|CreateRenderTarget2D 256 1 "RTF_R16f"))
                    (Variables|2%29Effects|Curves|SetCurve%20RT _returnvalue_2)
                    (Variables|5%29Debug|SetCurveTextureRendered false)))
                (CallFunction|UpdateCurves)))))))))

(fn CacheDistanceField
  (bind _distance_field_cache (Variables|5%29Debug|GetDistance%20Field%20Cache))
  (Rendering|ClearRenderTarget2D _distance_field_cache)
  (Rendering|DrawMaterialtoRenderTarget _distance_field_cache (Variables|5%29Debug|GetRender%20Distance%20Field%20MID))
  (Variables|5%29Debug|SetDistanceFieldIsCached true))

(fn CenterActorsOnSplinePoints
  (bind _spline (Variables|Default|GetSpline))
  (bind _returnvalue (Spline|GetNumberOfSplinePoints _spline))
  (bind _returnvalue_2 (- _returnvalue 1))
  (bind _accumavg (LocalVariables|GetAccumAvg))
  (for _index (range _returnvalue)
    (LocalVariables|SetAccumAvg (+ (LocalVariables|GetAccumAvg) (Spline|GetLocationatSplinePoint _spline _index "World"))))
  (LocalVariables|SetAccumAvg (* (/ (LocalVariables|GetAccumAvg) (Math|Conversions|ToFloat(Integer) _returnvalue)) "1.000000,1.000000, 0"))
  (for _index_2 (range _returnvalue)
    (Spline|SetLocationatSplinePoint (Variables|Default|GetSpline) _index_2 (- (Spline|GetLocationatSplinePoint (Variables|Default|GetSpline) _index_2 "World") (LocalVariables|GetAccumAvg)) "World" true))
  (Transformation|AddActorWorldOffset _accumavg)
  (Variables|1%29BrushSettings|Falloff|SetZ%20Offset (- (Variables|1%29BrushSettings|Falloff|GetZ%20Offset) (.z _accumavg)))
  (Variables|Default|SetDynamicPreviewPoint (Math|Color|MakeColor 0.0 0.0 0.0 0.0))
  (Variables|5%29Debug|SetDistanceFieldIsCached false)
  (Landscape|RequestLandscapeUpdate false))

(fn DrawBrushMaterial
  (bind _brushmanager (Variables|5%29Debug|GetBrushManager))
  (bind _heightmap_render_rt (Variables|/Landmass/Landscape/BlueprintBrushes/LandmassBrushManager.LandmassBrushManager_C|GetHeightmap%20Render%20RT _brushmanager))
  (bind _brushmanager_2 (Variables|5%29Debug|GetBrushManager))
  (bind _weightmap_rt (Variables|/Landmass/Landscape/BlueprintBrushes/LandmassBrushManager.LandmassBrushManager_C|GetWeightmap%20RT _brushmanager_2))
  (if (Variables|5%29Debug|GetHeightmapRender)
    (Rendering|ClearRenderTarget2D _heightmap_render_rt)
    (Rendering|DrawMaterialtoRenderTarget _heightmap_render_rt (Variables|5%29Debug|GetDrawBrushMID))
    (else
      (Rendering|ClearRenderTarget2D _weightmap_rt)
      (Rendering|DrawMaterialtoRenderTarget _weightmap_rt (Variables|5%29Debug|GetWeightmapMID)))))

(fn DrawCustomShape)

(fn DrawCanvasShape
  (bind (_r _g _b _a) (Utilities|Struct|BreakLinearColor (Variables|Default|GetDynamicPreviewPoint)))
  (bind _brushmanager (Variables|5%29Debug|GetBrushManager))
  (bind _depth_and_shape_rt_a (Variables|/Landmass/Landscape/BlueprintBrushes/LandmassBrushManager.LandmassBrushManager_C|GetDepth%20and%20Shape%20RT%20A _brushmanager))
  (Utilities|Array|Clear (Variables|Default|GetCanvasUVTris))
  (for _index (range (- (Spline|GetNumberOfSplinePoints (Variables|Default|GetSpline)) 2))
    (Utilities|Array|Add (Variables|Default|GetCanvasUVTris) (Utilities|Struct|MakeCanvasUVTri (Math|Conversions|ToVector2D(Vector) (+ (Math|Transform|InverseTransformLocation (Variables|/Landmass/Landscape/BlueprintBrushes/LandmassBrushManager.LandmassBrushManager_C|GetLS%20Transform (Variables|5%29Debug|GetBrushManager)) (Spline|GetLocationatSplinePoint (Variables|Default|GetSpline) 0 "World")) 0.5)) "(X=0.000000,Y=0.000000)" "(R=1.000000,G=0.000000,B=0.000000,A=1.000000)" (Math|Conversions|ToVector2D(Vector) (+ (Math|Transform|InverseTransformLocation (Variables|/Landmass/Landscape/BlueprintBrushes/LandmassBrushManager.LandmassBrushManager_C|GetLS%20Transform (Variables|5%29Debug|GetBrushManager)) (Spline|GetLocationatSplinePoint (Variables|Default|GetSpline) (+ (+ _index 1) 1) "World")) 0.5)) "(X=0.000000,Y=0.000000)" "(R=0.000000,G=1.000000,B=0.245000,A=1.000000)" (Math|Conversions|ToVector2D(Vector) (+ (Math|Transform|InverseTransformLocation (Variables|/Landmass/Landscape/BlueprintBrushes/LandmassBrushManager.LandmassBrushManager_C|GetLS%20Transform (Variables|5%29Debug|GetBrushManager)) (Spline|GetLocationatSplinePoint (Variables|Default|GetSpline) (+ _index 1) "World")) 0.5)) "(X=0.000000,Y=0.000000)" "(R=0.000306,G=0.000000,B=1.000000,A=1.000000)")))
  (if (== _a 1.0)
    (Utilities|FlowControl|Sequence
      (:then_0
        (Utilities|Array|Add (Variables|Default|GetCanvasUVTris) (Utilities|Struct|MakeCanvasUVTri (Math|Conversions|ToVector2D(Vector) (+ (Math|Transform|InverseTransformLocation (Variables|/Landmass/Landscape/BlueprintBrushes/LandmassBrushManager.LandmassBrushManager_C|GetLS%20Transform (Variables|5%29Debug|GetBrushManager)) (Spline|GetLocationatSplinePoint (Variables|Default|GetSpline) 0 "World")) 0.5)) "(X=0.000000,Y=0.000000)" "(R=1.000000,G=0.000000,B=0.000000,A=1.000000)" (Math|Conversions|ToVector2D(Vector) (+ (Math|Transform|InverseTransformLocation (Variables|/Landmass/Landscape/BlueprintBrushes/LandmassBrushManager.LandmassBrushManager_C|GetLS%20Transform (Variables|5%29Debug|GetBrushManager)) (Math|Conversions|ToVector(LinearColor) (Variables|Default|GetDynamicPreviewPoint))) 0.5)) "(X=0.000000,Y=0.000000)" "(R=0.000000,G=1.000000,B=0.245000,A=1.000000)" (Math|Conversions|ToVector2D(Vector) (+ (Math|Transform|InverseTransformLocation (Variables|/Landmass/Landscape/BlueprintBrushes/LandmassBrushManager.LandmassBrushManager_C|GetLS%20Transform (Variables|5%29Debug|GetBrushManager)) (Spline|GetLocationatSplinePoint (Variables|Default|GetSpline) (- (Spline|GetNumberOfSplinePoints (Variables|Default|GetSpline)) 1) "World")) 0.5)) "(X=0.000000,Y=0.000000)" "(R=0.000306,G=0.000000,B=1.000000,A=1.000000)")))
      (:then_1
        _)))
  (Rendering|ClearRenderTarget2D _depth_and_shape_rt_a)
  (bind (_canvas _size _context) (Rendering|BeginDrawCanvastoRenderTarget _depth_and_shape_rt_a))
  (Utilities|IsValid _canvas
    (:Is%20Valid
      (Utilities|IsValid (Variables|5%29Debug|GetShapeMID)
        (:Is%20Valid
          _)
        (:Is%20Not%20Valid
          (bind _returnvalue (Rendering|Material|CreateDynamicMaterialInstance "/Landmass/Landscape/OldPrototype_BP/Mats/Debug/Intersections.Intersections" "None" "None" :!class /Script/Engine.KismetMaterialLibrary))
          (bind _output_get (Variables|5%29Debug|SetShapeMID _returnvalue))))
      (Rendering|Material|SetScalarParameterValue _output_get "Z" (+ (.z (Transformation|GetActorLocation)) (Variables|1%29BrushSettings|Falloff|GetZ%20Offset)) :!class /Script/Engine.MaterialInstanceDynamic)
      (Canvas|DrawMaterialTriangles _canvas (Variables|5%29Debug|GetShapeMID) (Variables|Default|GetCanvasUVTris))
      (Rendering|EndDrawCanvastoRenderTarget _context))))

(fn ExportRT
  (Rendering|ExportRenderTarget (Variables|5%29Debug|GetDistance%20Field%20Cache) (Variables|5%29Debug|Getpath) (Variables|5%29Debug|GetName)))

(fn ForceUpdate
  (bind _splinemeshesforrender (Variables|Default|GetSplineMeshesForRender))
  ;[1 1 1] Spline Mesh
  (for _array_element _splinemeshesforrender
    (Rendering|SetVisibility _array_element true false))
  (Class|LandmassBrushManager|CaptureMeshDepth (Variables|5%29Debug|GetBrushManager) _splinemeshesforrender)
  (for _array_element_2 (Variables|Default|GetSplineMeshesForRender)
    (Rendering|SetVisibility _array_element_2 (select (Utilities|IsValid (Variables|1%29BrushSettings|SplineMesh|GetFlatten%20Mesh) :!class /Script/Engine.KismetSystemLibrary) (Variables|1%29BrushSettings|SplineMesh|GetShow%20Flatten%20Mesh) (Variables|1%29BrushSettings|SplineMesh|GetShow%20Splines)) false))
  (Class|JumpFloodComponent2D|JumpFlood (Variables|/Landmass/Landscape/BlueprintBrushes/LandmassBrushManager.LandmassBrushManager_C|GetJumpFlood_Component2D (Variables|5%29Debug|GetBrushManager)) (Variables|/Landmass/Landscape/BlueprintBrushes/LandmassBrushManager.LandmassBrushManager_C|GetDepth%20and%20Shape%20RT%20A (Variables|5%29Debug|GetBrushManager)) 5e+04 (Variables|5%29Debug|GetCurlparams) true 0.0 :!class /Landmass/DistanceFields/Blueprints/JumpFlood_Component2D.JumpFlood_Component2D_C))

(fn GetBlueprintRenderDependencies
  (return (Utilities|Array|MakeArray "/Engine/MapTemplates/TilingNoise05.TilingNoise05" "/Landmass/Landscape/BlueprintBrushes/Materials/CacheDistanceField_RG8.CacheDistanceField_RG8" "/Landmass/Landscape/BlueprintBrushes/Materials/CopyRT_WithAlpha.CopyRT_WithAlpha" "/Landmass/Landscape/BlueprintBrushes/Materials/LandmassBrush_Angle.LandmassBrush_Angle" "/Landmass/Landscape/BlueprintBrushes/Materials/LandmassBrush_Weightmap.LandmassBrush_Weightmap" "/Landmass/Landscape/BlueprintBrushes/Materials/LandmassBrush_Width.LandmassBrush_Width" "/Landmass/Landscape/BlueprintBrushes/Materials/M_SimpleBrush.M_SimpleBrush" "/Landmass/Landscape/OldPrototype_BP/Mats/Debug/Intersections.Intersections" "/Landmass/DistanceFields/Materials/JumpFloodStep.JumpFloodStep" "/Landmass/DistanceFields/Materials/DetectEdges.DetectEdges" "/Landmass/DistanceFields/Materials/BlurEdges.BlurEdges" (Variables|2%29Effects|Displacement|GetDisplacement%20Texture))))

(fn HardEdgeTangents
  (for _index (range (Spline|GetNumberOfSplinePoints (Variables|Default|GetSpline)))
    (Spline|SetTangentatSplinePoint (Variables|Default|GetSpline) _index (Math|Vector|MakeVector 0.0) "Local" true))
  (Variables|5%29Debug|SetDistanceFieldIsCached false)
  (Landscape|RequestLandscapeUpdate false))

(fn RenderLayer
  (params (InParameters :type Landscape%20Brush%20Parameters%20Structure%20%28by%20ref%29))
  (bind (_renderareaworldtransform _renderareasize _combinedresult _layertype _weightmaplayername) (Utilities|Struct|BreakLandscapeBrushParameters InParameters))
  (bind _splinemeshesforrender (Variables|Default|GetSplineMeshesForRender))
  (Variables|5%29Debug|SetHeightmapRender (Utilities|Enum|Equal(Enum) _layertype "Heightmap"))
  (Variables|5%29Debug|SetCurPaintLayer _weightmaplayername)
  (CallFunction|BrushSetup)
  (Rendering|Material|SetTextureParameterValue (Variables|5%29Debug|GetDrawBrushMID) "HeightRT" _combinedresult)
  (if (not (Utilities|Enum|Equal(Enum) (Variables|1%29BrushSettings|GetBrush%20Type) "NewEnumerator7"))
    (if (not (Variables|5%29Debug|GetDistanceFieldIsCached))
      (switch Utilities|FlowControl|Switch|SwitchonBrushTypes (Variables|1%29BrushSettings|GetBrush%20Type)
        (:NewEnumerator0
          (if (> (Spline|GetNumberOfSplinePoints (Variables|Default|GetSpline)) 2)
            (Default|RasterizeSplineShape (Variables|5%29Debug|GetBrushManager) (Variables|Default|GetSpline) (+ (.z (Transformation|GetActorLocation)) (Variables|1%29BrushSettings|Falloff|GetZ%20Offset)) 0.0 false)
            (Class|JumpFloodComponent2D|JumpFlood (Variables|/Landmass/Landscape/BlueprintBrushes/LandmassBrushManager.LandmassBrushManager_C|GetJumpFlood_Component2D (Variables|5%29Debug|GetBrushManager)) (Variables|/Landmass/Landscape/BlueprintBrushes/LandmassBrushManager.LandmassBrushManager_C|GetDepth%20and%20Shape%20RT%20A (Variables|5%29Debug|GetBrushManager)) 5e+04 (Variables|5%29Debug|GetCurlparams) false (+ (.z (Transformation|GetActorLocation)) (Variables|1%29BrushSettings|Falloff|GetZ%20Offset)) :!class /Landmass/DistanceFields/Blueprints/JumpFlood_Component2D.JumpFlood_Component2D_C)))
        (:NewEnumerator1
          (if (> (Spline|GetSplineLength (Variables|Default|GetSpline)) (* (Math|Float|Max(Float) (.x (Transformation|GetActorScale3D)) (.y (Transformation|GetActorScale3D))) (Variables|1%29BrushSettings|SplineMesh|GetDefault%20Mesh%20Length)))
            ;[1 1 1] Spline Mesh
            (for _array_element _splinemeshesforrender
              (Rendering|SetVisibility _array_element true false))
            (Class|LandmassBrushManager|CaptureMeshDepth (Variables|5%29Debug|GetBrushManager) _splinemeshesforrender)
            (for _array_element_2 (Variables|Default|GetSplineMeshesForRender)
              (Rendering|SetVisibility _array_element_2 (select (Utilities|IsValid (Variables|1%29BrushSettings|SplineMesh|GetFlatten%20Mesh) :!class /Script/Engine.KismetSystemLibrary) (Variables|1%29BrushSettings|SplineMesh|GetShow%20Flatten%20Mesh) (Variables|1%29BrushSettings|SplineMesh|GetShow%20Splines)) false))
            (Class|JumpFloodComponent2D|JumpFlood (Variables|/Landmass/Landscape/BlueprintBrushes/LandmassBrushManager.LandmassBrushManager_C|GetJumpFlood_Component2D (Variables|5%29Debug|GetBrushManager)) (Variables|/Landmass/Landscape/BlueprintBrushes/LandmassBrushManager.LandmassBrushManager_C|GetDepth%20and%20Shape%20RT%20A (Variables|5%29Debug|GetBrushManager)) 5e+04 (Variables|5%29Debug|GetCurlparams) true (Variables|1%29BrushSettings|Falloff|GetZ%20Offset) :!class /Landmass/DistanceFields/Blueprints/JumpFlood_Component2D.JumpFlood_Component2D_C)
            (else
              (Development|PrintString "Spline Not Long Enough to Render"))))
        (:NewEnumerator5
          ;[1 1 1] Comment
          (Development|SetHiddeninGame (Variables|Default|GetBrush%20Mesh) false false)
          (Class|LandmassBrushManager|CaptureMeshDepth (Variables|5%29Debug|GetBrushManager) (Utilities|Array|MakeArray (Variables|Default|GetBrush%20Mesh)))
          (Development|SetHiddeninGame (Variables|Default|GetBrush%20Mesh) true false))
        (:NewEnumerator7
          (if false
            (Development|PrintString "Cached or Material Only Draw"))
          (CallFunction|DrawBrushMaterial)
          (return (select (Variables|5%29Debug|GetHeightmapRender) (Variables|/Landmass/Landscape/BlueprintBrushes/LandmassBrushManager.LandmassBrushManager_C|GetHeightmap%20Render%20RT (Variables|5%29Debug|GetBrushManager)) (Variables|/Landmass/Landscape/BlueprintBrushes/LandmassBrushManager.LandmassBrushManager_C|GetWeightmap%20RT (Variables|5%29Debug|GetBrushManager))))))
      (CallFunction|CacheDistanceField))
    (else
      _)))

(fn UpdateCurves
  (bind _drawbrushmid (Variables|5%29Debug|GetDrawBrushMID))
  (Utilities|IsValid (Variables|2%29Effects|Curves|GetElevation%20Curve%20Asset)
    (:Is%20Valid
      (Rendering|ClearRenderTarget2D (Variables|2%29Effects|Curves|GetCurve%20RT))
      (bind _canvas (Rendering|BeginDrawCanvastoRenderTarget (Variables|2%29Effects|Curves|GetCurve%20RT)))
      (Utilities|IsValid _canvas
        (:Is%20Valid
          (for _index (range 256)
            (Canvas|DrawBox _canvas (Math|Vector2D|MakeVector2D (Math|Conversions|ToFloat(Integer) _index)) "(X=1.000000,Y=2.000000)" 1.0 (Math|Conversions|ToLinearColor(Float) (Math|Curves|GetFloatValue (Variables|2%29Effects|Curves|GetElevation%20Curve%20Asset) (/ (Math|Conversions|ToFloat(Integer) _index) 255.0)))))
          (Variables|5%29Debug|SetCurveTextureRendered true)
          (Rendering|Material|SetTextureParameterValue _drawbrushmid "CurveRT" (Variables|2%29Effects|Curves|GetCurve%20RT))
          (Rendering|Material|SetScalarParameterValue _drawbrushmid "ChannelEdgeOffset" (Variables|2%29Effects|Curves|GetChannel%20Edge%20Offset) :!class /Script/Engine.MaterialInstanceDynamic)
          (Rendering|Material|SetScalarParameterValue _drawbrushmid "CurveChannelDepth" (* (Variables|2%29Effects|Curves|GetCurve%20Channel%20Depth) (Math|Conversions|ToFloat(Boolean) (Variables|2%29Effects|Curves|GetUse%20Curve%20Channel))) :!class /Script/Engine.MaterialInstanceDynamic)
          (Rendering|Material|SetScalarParameterValue _drawbrushmid "CurveRampWidth" (Variables|2%29Effects|Curves|GetCurve%20Ramp%20Width) :!class /Script/Engine.MaterialInstanceDynamic))))))

(fn ConstructionScript
  (Variables|5%29Debug|SetDistanceFieldIsCached false)
  (CallFunction|BrushSetup))
