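; Landmass landscape brush Blueprint that applies a user-set material to the landscape's heightmap or weightmap.
; On initialize it spawns or updates the LandmassBrushManager; RenderLayer prepares MIDs, feeding the incoming render target as HeightRT.
; It draws the material into the manager's render target matching the layer type and returns it.
; Spline shape, jump-flood, curve texture and custom parameter helpers are defined but not called by this render path.
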(event EventBeginPlay)

(event Collision|EventActorBeginOverlap
  (params (OtherActor :type Actor%20Object%20Reference)))

(event EventInitialize
  (params (InLandscapeTransform :type Transform%20%28by%20ref%29) (InLandscapeSize :type Int%20Point%20Structure%20%28by%20ref%29) (InLandscapeRenderTargetSize :type Int%20Point%20Structure%20%28by%20ref%29))
  (Development|PrintString "Brush Initialize")
  (Utilities|SpawnOrUpdateManager InLandscapeSize InLandscapeRenderTargetSize))

(fn ApplyCustomParams
  (bind _custom_scalar_params (Variables|1%29BrushSettings|Material|GetCustom%20Scalar%20Params))
  (bind _custom_vector_params (Variables|1%29BrushSettings|Material|GetCustom%20Vector%20Params))
  (bind _custom_texture_params (Variables|1%29BrushSettings|Material|GetCustom%20Texture%20Params))
  (Utilities|FlowControl|Sequence
    (:then_0
      (bind _keys (Utilities|Map|Keys _custom_scalar_params))
      (for _array_element _keys
        (if (.ReturnValue (Utilities|Map|Find _custom_scalar_params _array_element))
          (Rendering|Material|SetScalarParameterValue (Variables|5%29Debug|GetDrawBrushMID) _array_element (Utilities|Map|Find _custom_scalar_params _array_element) :!class /Script/Engine.MaterialInstanceDynamic))))
    (:then_1
      (bind _keys_2 (Utilities|Map|Keys _custom_vector_params))
      (for _array_element_2 _keys_2
        (if (.ReturnValue (Utilities|Map|Find _custom_vector_params _array_element_2))
          (Rendering|Material|SetVectorParameterValue (Variables|5%29Debug|GetDrawBrushMID) _array_element_2 (Utilities|Map|Find _custom_vector_params _array_element_2) :!class /Script/Engine.MaterialInstanceDynamic))))
    (:then_2
      (bind _keys_3 (Utilities|Map|Keys _custom_texture_params))
      (for _array_element_3 _keys_3
        (if (.ReturnValue (Utilities|Map|Find _custom_texture_params _array_element_3))
          (Rendering|Material|SetTextureParameterValue (Variables|5%29Debug|GetDrawBrushMID) _array_element_3 (Utilities|Map|Find _custom_texture_params _array_element_3)))))))

(fn BrushSetup
  (params (Landscape%20RT :type Texture%20Render%20Target%202D%20Object%20Reference))
  (bind _brush_material (Variables|1%29BrushSettings|Material|GetBrush%20Material))
  (Class|LandmassBrushManager|SetMPCParams (Variables|5%29Debug|GetBrushManager))
  (Utilities|CreateMIDIfNeeded (Variables|5%29Debug|GetDrawBrushMID) _brush_material
    (:out
      (Utilities|CreateMIDIfNeeded (Variables|5%29Debug|GetWeightmapMID) _brush_material
        (:out
          (Rendering|Material|SetTextureParameterValue (Variables|5%29Debug|GetDrawBrushMID) "HeightRT" Landscape%20RT)
          (Utilities|SetMIDParams))))))

(fn DrawBrushMaterial
  (bind _brushmanager (Variables|5%29Debug|GetBrushManager))
  (bind _heightmap_render_rt (Variables|/Landmass/Landscape/BlueprintBrushes/LandmassBrushManager.LandmassBrushManager_C|GetHeightmap%20Render%20RT _brushmanager))
  (bind _brushmanager_2 (Variables|5%29Debug|GetBrushManager))
  (bind _weightmap_rt (Variables|/Landmass/Landscape/BlueprintBrushes/LandmassBrushManager.LandmassBrushManager_C|GetWeightmap%20RT _brushmanager_2))
  (if (Variables|5%29Debug|GetHeightmapRender)
    (Rendering|ClearRenderTarget2D _heightmap_render_rt)
    (Rendering|DrawMaterialtoRenderTarget _heightmap_render_rt (Variables|5%29Debug|GetDrawBrushMID))
    (else
      (Rendering|ClearRenderTarget2D _weightmap_rt)
      (Rendering|DrawMaterialtoRenderTarget _weightmap_rt (Variables|5%29Debug|GetWeightmapMID)))))

(fn DrawCustomShape)

(fn DrawCanvasShape
  (bind (_r _g _b _a) (Utilities|Struct|BreakLinearColor (Variables|Default|GetDynamicPreviewPoint)))
  (bind (_x _y) (Utilities|Struct|BreakIntPoint (Variables|/Landmass/Landscape/BlueprintBrushes/LandmassBrushManager.LandmassBrushManager_C|GetLandscape%20RT%20Res (Variables|5%29Debug|GetBrushManager))))
  (bind (_x_2 _y_2) (Utilities|Struct|BreakIntPoint (Variables|/Landmass/Landscape/BlueprintBrushes/LandmassBrushManager.LandmassBrushManager_C|GetLandscape%20RT%20Res (Variables|5%29Debug|GetBrushManager))))
  (bind _brushmanager (Variables|5%29Debug|GetBrushManager))
  (bind _depth_and_shape_rt_a (Variables|/Landmass/Landscape/BlueprintBrushes/LandmassBrushManager.LandmassBrushManager_C|GetDepth%20and%20Shape%20RT%20A _brushmanager))
  (Utilities|Array|Clear (Variables|Default|GetCanvasUVTris))
  (for _index (range (- (Spline|GetNumberOfSplinePoints (Variables|Default|GetSpline)) 2))
    (Utilities|Array|Add (Variables|Default|GetCanvasUVTris) (Utilities|Struct|MakeCanvasUVTri (Math|Conversions|ToVector2D(Vector) (* (+ (/ (Spline|GetLocationatSplinePoint (Variables|Default|GetSpline) 0 "World") (Variables|/Landmass/Landscape/BlueprintBrushes/LandmassBrushManager.LandmassBrushManager_C|GetWorld%20Size (Variables|5%29Debug|GetBrushManager))) 0.5) (Math|Vector|MakeVector (Math|Conversions|ToFloat(Integer) _x_2) (Math|Conversions|ToFloat(Integer) _y_2) 1.0))) "(X=0.000000,Y=0.000000)" "(R=1.000000,G=0.000000,B=0.000000,A=1.000000)" (Math|Conversions|ToVector2D(Vector) (* (+ (/ (Spline|GetLocationatSplinePoint (Variables|Default|GetSpline) (+ (+ _index 1) 1) "World") (Variables|/Landmass/Landscape/BlueprintBrushes/LandmassBrushManager.LandmassBrushManager_C|GetWorld%20Size (Variables|5%29Debug|GetBrushManager))) 0.5) (Math|Vector|MakeVector (Math|Conversions|ToFloat(Integer) _x_2) (Math|Conversions|ToFloat(Integer) _y_2) 1.0))) "(X=0.000000,Y=0.000000)" "(R=0.000000,G=1.000000,B=0.245000,A=1.000000)" (Math|Conversions|ToVector2D(Vector) (* (+ (/ (Spline|GetLocationatSplinePoint (Variables|Default|GetSpline) (+ _index 1) "World") (Variables|/Landmass/Landscape/BlueprintBrushes/LandmassBrushManager.LandmassBrushManager_C|GetWorld%20Size (Variables|5%29Debug|GetBrushManager))) 0.5) (Math|Vector|MakeVector (Math|Conversions|ToFloat(Integer) _x_2) (Math|Conversions|ToFloat(Integer) _y_2) 1.0))) "(X=0.000000,Y=0.000000)" "(R=0.000306,G=0.000000,B=1.000000,A=1.000000)")))
  (if (== _a 1.0)
    (Utilities|FlowControl|Sequence
      (:then_0
        (Utilities|Array|Add (Variables|Default|GetCanvasUVTris) (Utilities|Struct|MakeCanvasUVTri (Math|Conversions|ToVector2D(Vector) (* (+ (/ (Spline|GetLocationatSplinePoint (Variables|Default|GetSpline) 0 "World") (Variables|/Landmass/Landscape/BlueprintBrushes/LandmassBrushManager.LandmassBrushManager_C|GetWorld%20Size (Variables|5%29Debug|GetBrushManager))) 0.5) (Math|Vector|MakeVector (Math|Conversions|ToFloat(Integer) _x) (Math|Conversions|ToFloat(Integer) _y) 1.0))) "(X=0.000000,Y=0.000000)" "(R=1.000000,G=0.000000,B=0.000000,A=1.000000)" (Math|Conversions|ToVector2D(Vector) (* (+ (/ (Math|Conversions|ToVector(LinearColor) (Variables|Default|GetDynamicPreviewPoint)) (Variables|/Landmass/Landscape/BlueprintBrushes/LandmassBrushManager.LandmassBrushManager_C|GetWorld%20Size (Variables|5%29Debug|GetBrushManager))) 0.5) (Math|Vector|MakeVector (Math|Conversions|ToFloat(Integer) _x) (Math|Conversions|ToFloat(Integer) _y) 1.0))) "(X=0.000000,Y=0.000000)" "(R=0.000000,G=1.000000,B=0.245000,A=1.000000)" (Math|Conversions|ToVector2D(Vector) (* (+ (/ (Spline|GetLocationatSplinePoint (Variables|Default|GetSpline) (- (Spline|GetNumberOfSplinePoints (Variables|Default|GetSpline)) 1) "World") (Variables|/Landmass/Landscape/BlueprintBrushes/LandmassBrushManager.LandmassBrushManager_C|GetWorld%20Size (Variables|5%29Debug|GetBrushManager))) 0.5) (Math|Vector|MakeVector (Math|Conversions|ToFloat(Integer) _x) (Math|Conversions|ToFloat(Integer) _y) 1.0))) "(X=0.000000,Y=0.000000)" "(R=0.000306,G=0.000000,B=1.000000,A=1.000000)")))
      (:then_1
        _)))
  (Rendering|ClearRenderTarget2D _depth_and_shape_rt_a)
  (bind (_canvas _size _context) (Rendering|BeginDrawCanvastoRenderTarget _depth_and_shape_rt_a))
  (Utilities|IsValid _canvas
    (:Is%20Valid
      (Utilities|IsValid (Variables|5%29Debug|GetShapeMID)
        (:Is%20Valid
          _)
        (:Is%20Not%20Valid
          (bind _returnvalue (Rendering|Material|CreateDynamicMaterialInstance "/Landmass/Landscape/OldPrototype_BP/Mats/Debug/Intersections.Intersections" "None" "None" :!class /Script/Engine.KismetMaterialLibrary))
          (bind _output_get (Variables|5%29Debug|SetShapeMID _returnvalue))))
      (Rendering|Material|SetScalarParameterValue _output_get "Z" (+ (.z (Transformation|GetActorLocation)) 1.0) :!class /Script/Engine.MaterialInstanceDynamic)
      (Canvas|DrawMaterialTriangles _canvas (Variables|5%29Debug|GetShapeMID) (Variables|Default|GetCanvasUVTris))
      (Rendering|EndDrawCanvastoRenderTarget _context))))

(fn ExportRT
  (Rendering|ExportRenderTarget (Variables|5%29Debug|GetDistance%20Field%20Cache) (Variables|5%29Debug|Getpath) (Variables|5%29Debug|GetName)))

(fn ForceUpdate
  (bind _splinemeshesforrender (Variables|Default|GetSplineMeshesForRender))
  ;[1 1 1] Spline Mesh
  (for _array_element _splinemeshesforrender
    (Rendering|SetVisibility _array_element true false))
  (Class|LandmassBrushManager|CaptureMeshDepth (Variables|5%29Debug|GetBrushManager) _splinemeshesforrender)
  (for _array_element_2 (Variables|Default|GetSplineMeshesForRender)
    (Rendering|SetVisibility _array_element_2 (select (Utilities|IsValid 0 :!class /Script/Engine.KismetSystemLibrary) false false) false))
  (Class|JumpFloodComponent2D|JumpFlood (Variables|/Landmass/Landscape/BlueprintBrushes/LandmassBrushManager.LandmassBrushManager_C|GetJumpFlood_Component2D (Variables|5%29Debug|GetBrushManager)) (Variables|/Landmass/Landscape/BlueprintBrushes/LandmassBrushManager.LandmassBrushManager_C|GetDepth%20and%20Shape%20RT%20A (Variables|5%29Debug|GetBrushManager)) 5e+04 (Variables|5%29Debug|GetCurlparams) true 0.0 :!class /Landmass/DistanceFields/Blueprints/JumpFlood_Component2D.JumpFlood_Component2D_C))

(fn RenderLayer
  (params (InParameters :type Landscape%20Brush%20Parameters%20Structure%20%28by%20ref%29))
  (bind (_renderareaworldtransform _renderareasize _combinedresult _layertype) (Utilities|Struct|BreakLandscapeBrushParameters InParameters))
  (Variables|5%29Debug|SetHeightmapRender (Utilities|Enum|Equal(Enum) _layertype "Heightmap"))
  (CallFunction|BrushSetup _combinedresult)
  (if false
    (Development|PrintString "Material Only Draw"))
  ;[1 1 1] Draw the Material
  (CallFunction|DrawBrushMaterial)
  ;[0.1699 0.01189 0.25] Return the result to Landscape
  (return (select (Variables|5%29Debug|GetHeightmapRender) (Variables|/Landmass/Landscape/BlueprintBrushes/LandmassBrushManager.LandmassBrushManager_C|GetHeightmap%20Render%20RT (Variables|5%29Debug|GetBrushManager)) (Variables|/Landmass/Landscape/BlueprintBrushes/LandmassBrushManager.LandmassBrushManager_C|GetWeightmap%20RT (Variables|5%29Debug|GetBrushManager)))))

(fn UpdateCurves
  (Utilities|IsValid (Variables|2%29Effects|Curves|GetElevation%20Curve%20Asset)
    (:Is%20Valid
      (Rendering|ClearRenderTarget2D (Variables|2%29Effects|Curves|GetCurve%20RT))
      (bind _canvas (Rendering|BeginDrawCanvastoRenderTarget (Variables|2%29Effects|Curves|GetCurve%20RT)))
      (Utilities|IsValid _canvas
        (:Is%20Valid
          (for _index (range 256)
            (Canvas|DrawBox _canvas (Math|Vector2D|MakeVector2D (Math|Conversions|ToFloat(Integer) _index)) "(X=1.000000,Y=2.000000)" 1.0 (Math|Conversions|ToLinearColor(Float) (Math|Curves|GetFloatValue (Variables|2%29Effects|Curves|GetElevation%20Curve%20Asset) (/ (Math|Conversions|ToFloat(Integer) _index) 255.0)))))
          (Variables|5%29Debug|SetCurveTextureRendered true))))))

(fn ConstructionScript
  (Variables|5%29Debug|SetDistanceFieldIsCached false))
