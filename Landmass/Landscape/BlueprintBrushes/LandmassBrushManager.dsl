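; Editor-side Landmass brush manager that composites brush actors into landscape heightmap and weightmap render targets.
; A quadtree, native or legacy Blueprint, limits updates to brushes and extents touched by an edit.
; Fast preview mode draws through scratch targets and displays via landscape material parameters or an instanced preview mesh.
; It also handles spline shape drawing, rasterised shape masks and scene-capture mesh depth with temporary material swaps.
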(fn AddAllBrushestoTree
  (for _array_element (Variables|Landmass|GetLandmassBrushes)
    (Utilities|IsValid _array_element
      (:Is%20Valid
        (Variables|Debug|SetBrush%20Requesting%20Update _array_element)
        (Quadtree|AddBrushToTree)))))

(fn AddBrushtoTree
  (bind (_min _max) (Default|BreakExtents (Variables|Quadtree|GetModifiedNodeExtents)))
  (bind _brush_requesting_update (Variables|Debug|GetBrush%20Requesting%20Update))
  (if (Variables|Quadtree|GetUse%20Nativized%20Tree)
    (Landmass|AddBrushtoArray (Variables|Debug|GetBrush%20Requesting%20Update))
    (bind (_modifiedextents _invalidatedbrushes _modifiednodes) (Landmass|AddBrushtoTree (Variables|Debug|GetBrush%20Requesting%20Update) (Variables|/Script/LandmassEditor.LandmassActor|GetBrushExtents (Variables|Debug|GetBrush%20Requesting%20Update)) (Variables|/Script/LandmassEditor.LandmassActor|GetDrawToEntireLandscape (Variables|Debug|GetBrush%20Requesting%20Update))))
    (Variables|Quadtree|SetModifiedNodeExtents _modifiedextents)
    (Variables|Quadtree|SetBrushesRequiringUpdate _invalidatedbrushes)
    (Variables|Quadtree|SetCurrentModifiedNodes _modifiednodes)
    (bind _returnvalue (Landmass|SortBrushes (Variables|Landmass|GetLandmassBrushes) (Variables|Quadtree|GetBrushesRequiringUpdate)))
    (Variables|Quadtree|SetBrushesRequiringUpdate _returnvalue)
    (else
      ; Old BP Based Tree
      (Utilities|Array|AddUnique (Variables|Landmass|GetLandmassBrushesBP) (Variables|Debug|GetBrush%20Requesting%20Update))
      (if (> (Utilities|Array|Length (Utilities|Struct|BreakLandmassQuadtreeData (Variables|Quadtree|GetQuadTreeData_Internal))) 0)
        (bind _removed (Quadtree|RemoveActorFromQuadtree))
        (LocalVariables|SetPreviousNodes _removed)
        (if (Utilities|DoesObjectImplementInterface (Variables|Debug|GetBrush%20Requesting%20Update) "/Landmass/Landscape/BlueprintBrushes/Interface/Landmass_Interface.Landmass_Interface_C")
          (Utilities|FlowControl|Sequence
            (:then_0
              (LocalVariables|SetApplyToWholeLandscape (Variables|/Script/LandmassEditor.LandmassActor|GetDrawToEntireLandscape _brush_requesting_update))
              (LocalVariables|SetCurBrushExtents (Variables|/Script/LandmassEditor.LandmassActor|GetBrushExtents _brush_requesting_update))
              (if (LocalVariables|GetApplyToWholeLandscape)
                (LocalVariables|SetOverlappedNodes (Utilities|Array|MakeArray 0))
                (else
                  (bind _nodes (Quadtree|GetNodesForExtents (LocalVariables|GetCurBrushExtents)))
                  (LocalVariables|SetOverlappedNodes _nodes))))
            (:then_1
              (for _array_element (LocalVariables|GetOverlappedNodes)
                (Utilities|Array|AddUnique (Utilities|Struct|BreakLandmassQuadtree (Utilities|Array|Get(aref) (Utilities|Struct|BreakLandmassQuadtreeData (Variables|Quadtree|GetQuadTreeData_Internal)) _array_element)) (Variables|Debug|GetBrush%20Requesting%20Update))))
            (:then_2
              (Quadtree|CountChildNodes)
              (Utilities|Array|Clear (Variables|Quadtree|GetCurrentModifiedNodes))
              (for _array_element_2 (LocalVariables|GetOverlappedNodes)
                (Utilities|Array|Add (Variables|Quadtree|GetCurrentModifiedNodes) _array_element_2))
              (for _array_element_3 (LocalVariables|GetPreviousNodes)
                (Utilities|Array|AddUnique (Variables|Quadtree|GetCurrentModifiedNodes) _array_element_3)))
            (:then_3
              (Variables|Quadtree|SetModifiedNodeExtents (Math|Vector4|MakeVector4 1e+18 1e+18 -1e+18 -1e+18))
              (for _array_element_4 (Variables|Quadtree|GetCurrentModifiedNodes)
                (Variables|Quadtree|SetModifiedNodeExtents (Default|MakeExtents (Default|MinV2 (Default|BreakExtents (.Extents_29_CFC7F4FD40A19AB21C4959B310ECE410 (Utilities|Struct|BreakLandmassQuadtree (Utilities|Array|Get(aref) (Utilities|Struct|BreakLandmassQuadtreeData (Variables|Quadtree|GetQuadTreeData_Internal)) _array_element_4)))) _min) (Default|MaxV2 (.Max (Default|BreakExtents (.Extents_29_CFC7F4FD40A19AB21C4959B310ECE410 (Utilities|Struct|BreakLandmassQuadtree (Utilities|Array|Get(aref) (Utilities|Struct|BreakLandmassQuadtreeData (Variables|Quadtree|GetQuadTreeData_Internal)) _array_element_4))))) _max)))))
            (:then_4
              (bind _nodes_2 (Quadtree|GetNodesForExtents (Variables|Quadtree|GetModifiedNodeExtents)))
              (bind _output_get (Variables|Quadtree|SetCurrentModifiedNodes _nodes_2))
              (Quadtree|GetBrushesinModifiedNodes))))))))

(fn AddDebugBrushToTree
  (bind _aslandmass_actor (Utilities|Casting|CastToLandmassActor (Variables|Quadtree|GetSpecificActorDebug)))
  (bind (_modifiedextents _invalidatedbrushes _modifiednodes) (Landmass|AddBrushtoTree (Variables|Debug|GetBrush%20Requesting%20Update) (Variables|/Script/LandmassEditor.LandmassActor|GetBrushExtents _aslandmass_actor) (Variables|/Script/LandmassEditor.LandmassActor|GetDrawToEntireLandscape _aslandmass_actor)))
  (Variables|Quadtree|SetModifiedNodeExtents _modifiedextents)
  (Variables|Quadtree|SetBrushesRequiringUpdate _invalidatedbrushes)
  (Variables|Quadtree|SetCurrentModifiedNodes _modifiednodes))

(fn AllocateAllBrushWeightmaps
  (bind (_landscapetransform _landscapequads _rendertargetresolution) (Utilities|Struct|BreakLandmassLandscapeInfo (Variables|Landmass|GetLandscapeInformation)))
  (bind (_x _y) (Utilities|Struct|BreakIntPoint _rendertargetresolution))
  (bind _affectedweightmaplayers (Variables|Settings|GetAffectedWeightmapLayers))
  (bind _weightmapslastidx (LocalVariables|GetWeightmapsLastIdx))
  (for _array_element (Variables|Landmass|GetLandmassBrushes)
    (Utilities|IsValid _array_element
      (:Is%20Valid
        (for _array_element_2 (Variables|/Script/LandmassEditor.LandmassActor|GetWeightmapLayers _array_element)
          (if (!= _array_element_2 "")
            (Utilities|Array|AddUnique (Variables|Settings|GetAffectedWeightmapLayers) _array_element_2)
            (bind _output_get (LocalVariables|SetTempRTRef (Utilities|Map|Find (Variables|RenderTargets|GetWeightmap%20RTs) _array_element_2)))
            (Rendering|CreateAndAssignRenderTarget _output_get _x _y "RTF_R8"
              (:Out
                (Utilities|Map|Add (Variables|RenderTargets|GetWeightmap%20RTs) _array_element_2 (LocalVariables|GetTempRTRef)))))))))
  (bind _output_get_2 (LocalVariables|SetWeightmapsLastIdx (Utilities|Array|LastIndex (Variables|Settings|GetAffectedWeightmapLayers))))
  (for _index (range (+ _output_get_2 1))
    (LocalVariables|SetFoundLayer false)
    (Utilities|FlowControl|Sequence
      (:then_0
        (bind _array_element_3 (Utilities|Array|ForEachLoopwithBreak (Variables|Landmass|GetLandmassBrushes) :!label _label)
          (:LoopBody
            (Utilities|IsValid _array_element_3
              (:Is%20Valid
                (if (>= (Utilities|Array|FindItem (Variables|/Script/LandmassEditor.LandmassActor|GetWeightmapLayers _array_element_3) (Utilities|Array|Get(aref) _affectedweightmaplayers (- _weightmapslastidx _index))) 0)
                  (LocalVariables|SetFoundLayer true)
                  (goto _label Break)))))))
      (:then_1
        (if (not (LocalVariables|GetFoundLayer))
          (Utilities|Map|Remove (Variables|RenderTargets|GetWeightmap%20RTs) (Utilities|Array|Get(aref) _affectedweightmaplayers (- _weightmapslastidx _index)))
          (Class|LandmassPreviewLandscape|UpdateMIDWeight (Variables|Debug|GetPreview%20Landscape%20Mesh) (Utilities|Array|Get(aref) _affectedweightmaplayers (- _weightmapslastidx _index)) "")
          (Utilities|Array|RemoveItem (Variables|Settings|GetAffectedWeightmapLayers) (Utilities|Array|Get(aref) _affectedweightmaplayers (- _weightmapslastidx _index))))))))

(fn AllocateBrushWeightmaps
  (bind (_landscapetransform _landscapequads _rendertargetresolution) (Utilities|Struct|BreakLandmassLandscapeInfo (Variables|Landmass|GetLandscapeInformation)))
  (bind (_x _y) (Utilities|Struct|BreakIntPoint _rendertargetresolution))
  (bind _affectedweightmaplayers (Variables|Settings|GetAffectedWeightmapLayers))
  (bind _weightmapslastidx (LocalVariables|GetWeightmapsLastIdx))
  (for _array_element (Variables|/Script/LandmassEditor.LandmassActor|GetWeightmapLayers (Variables|Debug|GetBrush%20Requesting%20Update))
    (if (!= _array_element "")
      (Utilities|Array|AddUnique (Variables|Settings|GetAffectedWeightmapLayers) _array_element)
      (bind _output_get (LocalVariables|SetTempRTRef (Utilities|Map|Find (Variables|RenderTargets|GetWeightmap%20RTs) _array_element)))
      (Rendering|CreateAndAssignRenderTarget _output_get _x _y "RTF_R8"
        (:Out
          (Utilities|Map|Add (Variables|RenderTargets|GetWeightmap%20RTs) _array_element (LocalVariables|GetTempRTRef))))))
  (bind _output_get_2 (LocalVariables|SetWeightmapsLastIdx (Utilities|Array|LastIndex (Variables|Settings|GetAffectedWeightmapLayers))))
  (for _index (range (+ _output_get_2 1))
    (LocalVariables|SetFoundLayer false)
    (Utilities|FlowControl|Sequence
      (:then_0
        (bind _array_element_2 (Utilities|Array|ForEachLoopwithBreak (Variables|Landmass|GetLandmassBrushes) :!label _label)
          (:LoopBody
            (Utilities|IsValid _array_element_2
              (:Is%20Valid
                (if (>= (Utilities|Array|FindItem (Variables|/Script/LandmassEditor.LandmassActor|GetWeightmapLayers _array_element_2) (Utilities|Array|Get(aref) _affectedweightmaplayers (- _weightmapslastidx _index))) 0)
                  (LocalVariables|SetFoundLayer true)
                  (goto _label Break)))))))
      (:then_1
        (if (not (LocalVariables|GetFoundLayer))
          (Utilities|Map|Remove (Variables|RenderTargets|GetWeightmap%20RTs) (Utilities|Array|Get(aref) _affectedweightmaplayers (- _weightmapslastidx _index)))
          (Class|LandmassPreviewLandscape|UpdateMIDWeight (Variables|Debug|GetPreview%20Landscape%20Mesh) (Utilities|Array|Get(aref) _affectedweightmaplayers (- _weightmapslastidx _index)) "")
          (Utilities|Array|RemoveItem (Variables|Settings|GetAffectedWeightmapLayers) (Utilities|Array|Get(aref) _affectedweightmaplayers (- _weightmapslastidx _index))))))))

(fn AllocateRTs
  (bind (_x _y) (Utilities|Struct|BreakIntPoint (Variables|Debug|LandscapeInformation|GetLandscape%20RT%20Res)))
  (bind (_x_2 _y_2) (Utilities|Struct|BreakIntPoint (Variables|Debug|LandscapeInformation|GetLandscape%20RT%20Res)))
  (bind (_x_3 _y_3) (Utilities|Struct|BreakIntPoint (Variables|Debug|LandscapeInformation|GetLandscape%20RT%20Res)))
  (Utilities|FlowControl|Sequence
    (:then_0
      (Rendering|CreateAndAssignRenderTarget (Variables|RenderTargets|GetHeightmap%20Render%20RT) _x _y
        (:Out
          (Rendering|CreateAndAssignRenderTarget (Variables|RenderTargets|GetWeightmap%20RT) _x_2 _y_2 "RTF_R8"))))
    (:then_1
      (if (Variables|Landmass|GetEnable%20Fast%20Preview%20Mode)
        (Rendering|CreateAndAssignRenderTarget (Variables|RenderTargets|GetScratch%20Heightmap%20RT) _x_3 _y_3)
        (else
          (Variables|RenderTargets|SetScratch%20Heightmap%20RT ""))))
    (:then_3
      (if (Variables|Landmass|GetEnable%20Fast%20Preview%20Mode)
        _))))

(fn BeginShape
  (bind _curcustombrush (Variables|Default|GetCurCustomBrush))
  (bind _spline (Variables|/Landmass/Landscape/BlueprintBrushes/CustomBrush_Landmass.CustomBrush_Landmass_C|GetSpline _curcustombrush))
  (bind _returnvalue (Spline|GetNumberOfSplinePoints _spline))
  (bind _curcustombrush_2 (Variables|Default|GetCurCustomBrush))
  (if (Variables|Default|GetDrawing%20Mode)
    (bind (_cameralocation _rayorigin _raydirection) (Rendering|GetCursorWorldRay :!class /Script/LandmassEditor.LandmassBlueprintFunctionLibrary))
    (Spline|AddSplinePoint _spline (+ _rayorigin (* _raydirection (/ (- (.z _rayorigin) (+ (Variables|ShapeDrawing|GetDrawing%20Plane%20Z) 0.015)) (Math|Float|Absolute(Float) (.z _raydirection))))) "World" true)
    (if (Math|Boolean|ORBoolean (Utilities|Enum|Equal(Enum) (Variables|ShapeDrawing|GetNew%20Shape%20Type) "NewEnumerator0") (Utilities|Enum|Equal(Enum) (Variables|ShapeDrawing|GetNew%20Shape%20Type) "NewEnumerator6") (Utilities|Enum|Equal(Enum) (Variables|ShapeDrawing|GetNew%20Shape%20Type) "NewEnumerator7"))
      (Spline|SetTangentsatSplinePoint _spline (- _returnvalue 1) (Math|Vector|MakeVector 0.0) (Math|Vector|MakeVector 0.0) "World" true))
    (if (> _returnvalue 2)
      (switch Utilities|FlowControl|Switch|SwitchonShapeDrawingMode (Variables|ShapeDrawing|GetNew%20Shape%20Type)
        (:NewEnumerator1
          _)
        (:NewEnumerator5
          _))
      (Class|CustomBrushLandmass|BrushSetup _curcustombrush_2)
      (Variables|/Landmass/Landscape/BlueprintBrushes/CustomBrush_Landmass.CustomBrush_Landmass_C|SetDistanceFieldIsCached false _curcustombrush_2))
    (else
      (Landmass|CreateAndAssignMID (Variables|Default|GetDebug%20MID) "/Landmass/Landscape/BlueprintBrushes/Materials/Debug/DebugDrawing_CustomBrush.DebugDrawing_CustomBrush"
        (:Out
          (Rendering|Material|SetTextureParameterValue (Variables|Default|GetDebug%20MID) "Shape" (Variables|RenderTargets|GetDepth%20and%20Shape%20RT%20A))
          (bind _returnvalue_2 (Game|SpawnActorfromClass "/Landmass/Landscape/BlueprintBrushes/CustomBrush_Landmass.CustomBrush_Landmass_C" (Math|Transform|MakeTransform "0, 0, 0") "Undefined" "MultiplyWithRoot" 0 (Utilities|Select :Index (Variables|ShapeDrawing|GetNew%20Shape%20Type) :NewEnumerator0 "NewEnumerator0" :NewEnumerator6 "NewEnumerator0" :NewEnumerator1 "NewEnumerator1" :NewEnumerator5 "NewEnumerator1" :NewEnumerator7 "NewEnumerator0") (not (Utilities|Enum|Equal(Enum) (Variables|ShapeDrawing|GetNew%20Shape%20Type) "NewEnumerator6")) (Utilities|Select :Index (Variables|ShapeDrawing|GetNew%20Shape%20Type) :NewEnumerator0 (Variables|ShapeDrawing|RoadSetup|GetNew%20Road%20Spline%20Mesh) :NewEnumerator6 (Variables|ShapeDrawing|RoadSetup|GetNew%20Road%20Spline%20Mesh) :NewEnumerator1 (Variables|ShapeDrawing|RoadSetup|GetNew%20Road%20Spline%20Mesh) :NewEnumerator5 (Variables|ShapeDrawing|RiverSetup|GetNew%20River%20Spline%20Mesh) :NewEnumerator7 (Variables|ShapeDrawing|RoadSetup|GetNew%20Road%20Spline%20Mesh)) (Utilities|Select :Index (Variables|ShapeDrawing|GetNew%20Shape%20Type) :NewEnumerator0 0.0 :NewEnumerator6 0.0 :NewEnumerator1 -32.0 :NewEnumerator5 -128.0 :NewEnumerator7 0.0) (Utilities|Select :Index (Variables|ShapeDrawing|GetNew%20Shape%20Type) :NewEnumerator1 (Variables|ShapeDrawing|RoadSetup|GetRoad%20Material) :NewEnumerator5 (Variables|ShapeDrawing|RiverSetup|GetRiver%20Material)) 0))
          (bind _output_get (Variables|Default|SetCurCustomBrush _returnvalue_2))
          (Variables|Settings|SetAffectHeightmap (not (Utilities|Enum|Equal(Enum) (Variables|ShapeDrawing|GetNew%20Shape%20Type) "NewEnumerator7")) _output_get)
          (EditorScripting|ActorEditing|SetActorLabel (Variables|Default|GetCurCustomBrush) (Utilities|String|Append "BP_LandmassBrush_" (Utilities|Select :Index (Variables|ShapeDrawing|GetNew%20Shape%20Type) :NewEnumerator0 "Landmass" :NewEnumerator6 "Mountain" :NewEnumerator1 "RoadSpline" :NewEnumerator5 "RiverSpline" :NewEnumerator7 "Weightmap")) true)
          (if (Utilities|Enum|Equal(Enum) (Variables|ShapeDrawing|GetNew%20Shape%20Type) "NewEnumerator7")
            (Utilities|Map|Add (Variables|/Landmass/Landscape/BlueprintBrushes/CustomBrush_Landmass.CustomBrush_Landmass_C|GetPaint%20Layer%20Target%20Settings (Variables|Default|GetCurCustomBrush)) (Utilities|String|StringToName (Variables|ShapeDrawing|WeightmapShape|GetWeightmap%20Layer%20Name)) (Utilities|Struct|MakeBPWaterBodyWeightmapSettings 512.0))))))))

(fn CaptureMeshDepth
  (params (Primitive :type Array%20of%20Primitive%20Component%20Object%20References))
  (bind (_x _y) (Utilities|Struct|BreakIntPoint (Variables|Debug|LandscapeInformation|GetLandscape%20RT%20Res)))
  (bind _scenecapturecomponent2d (Variables|Default|GetSceneCaptureComponent2D))
  (Rendering|SceneCapture|ClearShowOnlyComponents _scenecapturecomponent2d)
  (for _array_element Primitive
    (Rendering|SceneCapture|ShowOnlyComponent _scenecapturecomponent2d _array_element))
  (Rendering|CreateAndAssignRenderTarget (Variables|RenderTargets|GetDepth%20and%20Shape%20RT%20A) _x _y "RTF_R32f"
    (:Out
      (Variables|/Script/Engine.SceneCaptureComponent2D|SetTextureTarget (Variables|RenderTargets|GetDepth%20and%20Shape%20RT%20A) (Variables|Default|GetSceneCaptureComponent2D))
      (Variables|/Script/Engine.SceneCaptureComponent|SetCaptureSource "SCS_SceneDepth" (Variables|Default|GetSceneCaptureComponent2D))
      (Rendering|SceneCapture|CaptureScene (Variables|Default|GetSceneCaptureComponent2D) :!class /Script/Engine.SceneCaptureComponent2D))))

(fn CaptureMeshDepthBounded
  (params (InPrimitive :type Primitive%20Component%20Object%20Reference) (InRenderTarget :type Texture%20Render%20Target%202D%20Object%20Reference) (InCaptureSize :type Float%20%28double-precision%29) (InCaptureTransform :type Transform))
  (bind _meshdepthmid (LocalVariables|GetMeshDepthMID))
  (bind _scenecapturecomponent2d (Variables|Default|GetSceneCaptureComponent2D))
  (Utilities|IsValid InPrimitive
    (:Is%20Valid
      (Utilities|IsValid (Variables|GetInRenderTarget)
        (:Is%20Valid
          (Rendering|SceneCapture|ClearShowOnlyComponents (Variables|Default|GetSceneCaptureComponent2D))
          (Rendering|SceneCapture|ShowOnlyComponent (Variables|Default|GetSceneCaptureComponent2D) (Variables|GetInPrimitive))
          (Rendering|CreateAndAssignMID _meshdepthmid "/Landmass/Landscape/BlueprintBrushes/Helpers/M_CaptureMeshDepth.M_CaptureMeshDepth"
            (:Out
              (Rendering|Material|SetScalarParameterValue _meshdepthmid "CaptureSize" (Variables|GetInCaptureSize) :!class /Script/Engine.MaterialInstanceDynamic)
              (Utilities|FlowControl|Sequence
                (:then_0
                  ;[1 1 1] Save Material list and Assign Capture Mat
                  (Utilities|Array|Clear (LocalVariables|GetPrimMaterials))
                  (bind _asstatic_mesh_component (Utilities|Casting|CastToStaticMeshComponent (Variables|GetInPrimitive)))
                  (LocalVariables|SetPrevDisallowNanite (Variables|/Script/Engine.StaticMeshComponent|GetbDisallowNanite _asstatic_mesh_component))
                  (LocalVariables|SetPrevForceLOD (Variables|/Script/Engine.StaticMeshComponent|GetForcedLodModel _asstatic_mesh_component))
                  (LocalVariables|SetPrevVisibility (Variables|/Script/Engine.SceneComponent|GetbVisible _asstatic_mesh_component))
                  (Variables|/Script/Engine.StaticMeshComponent|SetbDisallowNanite true _asstatic_mesh_component)
                  (LocalVariables|SetPrevHiddenGame (Variables|/Script/Engine.SceneComponent|GetbHiddenInGame _asstatic_mesh_component))
                  (Rendering|LOD|SetForcedLodModel _asstatic_mesh_component 1)
                  (Rendering|SetVisibility _asstatic_mesh_component true false)
                  (Development|SetHiddeninGame _asstatic_mesh_component false false)
                  (for _index (range (Rendering|Material|GetNumMaterials (Variables|GetInPrimitive)))
                    (Utilities|Array|SetArrayElem (LocalVariables|GetPrimMaterials) _index (Rendering|Material|GetMaterial (Variables|GetInPrimitive) _index) true)
                    (Rendering|Material|SetMaterial (Variables|GetInPrimitive) _index (LocalVariables|GetMeshDepthMID))))
                (:then_1
                  ;[1 1 1] Setup Scene Capture & Capture
                  (Variables|/Script/Engine.SceneCaptureComponent2D|SetOrthoWidth (Variables|GetInCaptureSize) (Variables|Default|GetSceneCaptureComponent2D))
                  (Variables|/Script/Engine.SceneCaptureComponent2D|SetTextureTarget (Variables|GetInRenderTarget) (Variables|Default|GetSceneCaptureComponent2D))
                  (Variables|/Script/Engine.SceneCaptureComponent|SetCaptureSource "SCS_SceneColorHDR" (Variables|Default|GetSceneCaptureComponent2D))
                  (Transformation|SetWorldTransform (Variables|Default|GetSceneCaptureComponent2D) (Variables|GetInCaptureTransform) false false)
                  (Rendering|SceneCapture|CaptureScene (Variables|Default|GetSceneCaptureComponent2D) :!class /Script/Engine.SceneCaptureComponent2D))
                (:then_2
                  ;[1 1 1] Restore Materials and Scene Capture setup
                  (Default|RepositionSceneCapture)
                  (Rendering|SceneCapture|ClearShowOnlyComponents _scenecapturecomponent2d)
                  (Variables|/Script/Engine.SceneCaptureComponent2D|SetTextureTarget "" _scenecapturecomponent2d)
                  (bind _asstatic_mesh_component_2 (Utilities|Casting|CastToStaticMeshComponent (Variables|GetInPrimitive)))
                  (Variables|/Script/Engine.StaticMeshComponent|SetbDisallowNanite (LocalVariables|GetPrevDisallowNanite) _asstatic_mesh_component_2)
                  (Rendering|LOD|SetForcedLodModel _asstatic_mesh_component_2 (LocalVariables|GetPrevForceLOD))
                  (Rendering|SetVisibility _asstatic_mesh_component_2 (LocalVariables|GetPrevVisibility) false)
                  (Development|SetHiddeninGame _asstatic_mesh_component_2 (LocalVariables|GetPrevHiddenGame) false)
                  (for (_array_element _array_index) (LocalVariables|GetPrimMaterials)
                    (Rendering|Material|SetMaterial (Variables|GetInPrimitive) _array_index _array_element)))))))))))

(fn CompleteShape
  (Variables|Default|SetDrawing%20Mode false)
  (Variables|Default|SetPreviouslyDrawing false)
  (Class|CustomBrushLandmass|CenterActorsonSplinePoints (Variables|Default|GetCurCustomBrush))
  (Variables|Default|SetCurCustomBrush 0))

(fn CopyTargettoCBRT
  (bind _copy_export_target (Variables|DebugExport|GetCopy/Export%20Target))
  (bind _weightmap_rts (Variables|RenderTargets|GetWeightmap%20RTs))
  (bind (_value _returnvalue) (Utilities|Map|Find _weightmap_rts _copy_export_target))
  (if (== (Variables|DebugExport|GetCopy/Export%20Target) "Heightmap")
    (Rendering|Material|SetTextureParameterValue (Variables|Debug|GetCopyRT_MID) "RT" (Variables|RenderTargets|GetHeightmap%20Render%20RT))
    (elif _returnvalue
      (Rendering|Material|SetTextureParameterValue (Variables|Debug|GetCopyRT_MID) "RT" _value))))

(fn CopyWMLayersIfNeeded
  (bind _currentrenderparams (Variables|Debug|GetCurrentRenderParams))
  (bind (_renderareaworldtransform _renderareasize _combinedresult _layertype _weightmaplayername) (Utilities|Struct|BreakLandscapeBrushParameters _currentrenderparams))
  (bind _weightmap_rts (Variables|RenderTargets|GetWeightmap%20RTs))
  (bind (_value _returnvalue) (Utilities|Map|Find _weightmap_rts _weightmaplayername))
  (if (Variables|Debug|GetKeepWeightmapRTsResident)
    (if (Variables|Debug|GetHeightmapPass)
      _
      (elif _returnvalue
        (Rendering|Material|SetTextureParameterValue (Variables|Debug|GetCopyRT_MID) "RT" (Variables|RenderTargets|GetCombinedResultRT))
        (Rendering|DrawMaterialtoRenderTarget _value (Variables|Debug|GetCopyRT_MID))))))

(fn CountChildNodes
  (bind _quadtreedata_internal (Variables|Quadtree|GetQuadTreeData_Internal))
  (bind _quadtree_13_e8ca8c804ef0247f756fe3b8e015cfe7 (Utilities|Struct|BreakLandmassQuadtreeData _quadtreedata_internal))
  (bind (_actors_31_dd7b465a42cf64ce22945992a180dd0d _childdatacount_24_488e28b5486c9f3f0d44a2a50447b62f) (Utilities|Struct|BreakLandmassQuadtree (Utilities|Array|Get(aref) _quadtree_13_e8ca8c804ef0247f756fe3b8e015cfe7 (LocalVariables|GetParentToUpdate))))
  (for (_array_element _array_index) (Utilities|Struct|BreakLandmassQuadtreeData (Variables|Quadtree|GetQuadTreeData_Internal))
    (LocalVariables|SetCurrentActorCount (Utilities|Array|Length (.Actors_31_DD7B465A42CF64CE22945992A180DD0D (Utilities|Struct|BreakLandmassQuadtree _array_element))))
    (Utilities|Struct|SetmembersinLandmassQuadtree (Utilities|Array|Get(aref) (Utilities|Struct|BreakLandmassQuadtreeData (Variables|Quadtree|GetQuadTreeData_Internal)) _array_index) 0)
    (if (and (> (LocalVariables|GetCurrentActorCount) 0) (>= (.ParentIndex_26_2769B7C54152BAA5828C46ABCEF62BA9 (Utilities|Struct|BreakLandmassQuadtree _array_element)) 0))
      (LocalVariables|SetCurrentLevel (Utilities|Struct|BreakLandmassQuadtree _array_element))
      (LocalVariables|SetParentToUpdate (.ParentIndex_26_2769B7C54152BAA5828C46ABCEF62BA9 (Utilities|Struct|BreakLandmassQuadtree _array_element)))
      (for _ (range (LocalVariables|GetCurrentLevel))
        (Utilities|Struct|SetmembersinLandmassQuadtree (Utilities|Array|Get(aref) _quadtree_13_e8ca8c804ef0247f756fe3b8e015cfe7 (LocalVariables|GetParentToUpdate)) (+ _childdatacount_24_488e28b5486c9f3f0d44a2a50447b62f (LocalVariables|GetCurrentActorCount)))
        (LocalVariables|SetParentToUpdate (Utilities|Struct|BreakLandmassQuadtree (Utilities|Array|Get(aref) _quadtree_13_e8ca8c804ef0247f756fe3b8e015cfe7 (LocalVariables|GetParentToUpdate))))))))

(fn DrawMaterialWholeLandscape
  (params (In%20Material :type Material%20Interface%20Object%20Reference))
  (bind _scratch_heightmap_rt (Variables|RenderTargets|GetScratch%20Heightmap%20RT))
  (Rendering|ClearRenderTarget2D _scratch_heightmap_rt)
  (Rendering|DrawMaterialtoRenderTarget _scratch_heightmap_rt In%20Material))

(fn DrawBrushMaterial
  (params (Brush :type Landmass%20Actor%20Object%20Reference) (BrushMaterial :type Material%20Interface%20Object%20Reference))
  (bind _brush (Variables|GetBrush))
  (bind _brushrenderparameters (Variables|/Script/LandmassEditor.LandmassActor|GetBrushRenderParameters _brush))
  (bind (_renderareaworldtransform _renderareasize _weightmaplayername) (Utilities|Struct|BreakLandscapeBrushParameters _brushrenderparameters))
  (bind (_renderareaworldtransform_2 _renderareasize_2 _layertype) (Utilities|Struct|BreakLandscapeBrushParameters (Variables|/Script/LandmassEditor.LandmassActor|GetBrushRenderParameters (Variables|GetBrush))))
  (bind _weightmap_rts (Variables|RenderTargets|GetWeightmap%20RTs))
  (bind (_value _returnvalue) (Utilities|Map|Find _weightmap_rts _weightmaplayername))
  (Utilities|FlowControl|Sequence
    (:then_0
      (LocalVariables|SetExtentsToUse (select (Variables|/Script/LandmassEditor.LandmassActor|GetDrawToEntireLandscape (Variables|GetBrush)) (Variables|Quadtree|GetModifiedNodeExtents) (Variables|/Script/LandmassEditor.LandmassActor|GetBrushExtents (Variables|GetBrush))))
      (if (Utilities|Enum|Equal(Enum) _layertype "Heightmap")
        (if (Variables|Landmass|GetEnable%20Fast%20Preview%20Mode)
          (Utilities|BrushDrawToRenderTargetOverlappingExtents (Variables|RenderTargets|GetScratch%20Heightmap%20RT) (Variables|GetBrushMaterial) (LocalVariables|GetExtentsToUse) (Variables|Quadtree|GetModifiedNodeExtents) (Variables|Landmass|GetLandscapeInformation))
          (Utilities|BrushCopyRenderTargetOverlappingExtents (Variables|RenderTargets|GetScratch%20Heightmap%20RT) (Variables|RenderTargets|GetHeightmap%20Render%20RT) self (LocalVariables|GetExtentsToUse) (Variables|Quadtree|GetModifiedNodeExtents) (Variables|Landmass|GetLandscapeInformation))
          (else
            (Utilities|BrushDrawToRenderTargetExtents (Variables|RenderTargets|GetHeightmap%20Render%20RT) (Variables|GetBrushMaterial) (LocalVariables|GetExtentsToUse) (Variables|Landmass|GetLandscapeInformation)) ; Render Brush into the Original Source\nThis version draws to the entire current brushes bounds rather than restricting to the edited brush bounds\nIt doesn't know about which brush was edited as the call comes from the Landscape itself.
            (Utilities|BrushCopyRenderTargetExtents (Variables|RenderTargets|GetHeightmap%20Render%20RT) (Variables|RenderTargets|GetCombinedResultRT) self (LocalVariables|GetExtentsToUse) (Variables|Landmass|GetLandscapeInformation)) ; First Copy the area to read to a 2nd RT
          ))
        (elif (Variables|Landmass|GetEnable%20Fast%20Preview%20Mode)
          (if _returnvalue
            (LocalVariables|SetWeightmap%20RenderTarget _value)
            (Utilities|BrushDrawToRenderTargetOverlappingExtents (Variables|RenderTargets|GetWeightmap%20RT) (Variables|GetBrushMaterial) (LocalVariables|GetExtentsToUse) (Variables|Quadtree|GetModifiedNodeExtents) (Variables|Landmass|GetLandscapeInformation))
            (Utilities|BrushCopyRenderTargetOverlappingExtents (Variables|RenderTargets|GetWeightmap%20RT) (LocalVariables|GetWeightmap%20RenderTarget) self (LocalVariables|GetExtentsToUse) (Variables|Quadtree|GetModifiedNodeExtents) (Variables|Landmass|GetLandscapeInformation))
            (else
              (Development|PrintString "LandscapeActor requested weightmap render but no matching Weighmap RT found.")))
          (else
            (Utilities|BrushDrawToRenderTargetExtents (Variables|RenderTargets|GetWeightmap%20RT) (Variables|GetBrushMaterial) (LocalVariables|GetExtentsToUse) (Variables|Landmass|GetLandscapeInformation)) ; Render Brush into the Original Source\nThis version draws to the entire current brushes bounds rather than restricting to the edited brush bounds\nIt doesn't know about which brush was edited as the call comes from the Landscape itself.
            (Utilities|BrushCopyRenderTargetExtents (Variables|RenderTargets|GetWeightmap%20RT) (Variables|RenderTargets|GetCombinedResultRT) self (LocalVariables|GetExtentsToUse) (Variables|Landmass|GetLandscapeInformation)) ; First Copy the area to read to a 2nd RT
          ))))
    (:then_1
      (return))))

(fn EnablePreviewLandscape
  (if (Variables|Landmass|GetPreview%20Mode%20use%20Landscape%20MID)
    (Variables|Debug|SetFastPreview_UseMID_Previous true)
    (bind _returnvalue (Landmass|GetLandscape))
    (Landscape|Runtime|Material|SetLandscapeMaterialTextureParameterValue _returnvalue "HeightRT" (Variables|RenderTargets|GetHeightmap%20Render%20RT))
    (Landscape|Runtime|Material|SetLandscapeMaterialScalarParameterValue _returnvalue "EnableLandmassPreview" 1.0)
    (Landscape|Runtime|Material|SetLandscapeMaterialScalarParameterValue _returnvalue "RG8Height" 1.0)
    (Landscape|Runtime|Material|SetLandscapeMaterialScalarParameterValue _returnvalue "UseDerivedNormal" 1.0)
    (bind _keys (Utilities|Map|Keys (Variables|RenderTargets|GetWeightmap%20RTs)))
    (for _array_element _keys
      (Landscape|Runtime|Material|SetLandscapeMaterialTextureParameterValue _returnvalue _array_element (Utilities|Map|Find (Variables|RenderTargets|GetWeightmap%20RTs) _array_element)))
    (Rendering|SetVisibility (Variables|/Landmass/Landscape/BlueprintBrushes/Previewing/Landmass_PreviewLandscape.Landmass_PreviewLandscape_C|GetLandscape%20ISMC (Variables|Debug|GetPreview%20Landscape%20Mesh)) false false)
    (Development|ExecuteConsoleCommand "showflag.landscape 2")
    (else
      (Variables|Debug|SetFastPreview_UseMID_Previous false)
      (Utilities|IsValid (Variables|Debug|GetPreview%20Landscape%20Mesh)
        (:Is%20Valid
          (Transformation|AttachActorToActor (Variables|Debug|GetPreview%20Landscape%20Mesh) self "None" "KeepRelative" "KeepRelative" "KeepRelative" true))
        (:Is%20Not%20Valid
          (bind _returnvalue_2 (Game|SpawnActorfromClass "/Landmass/Landscape/BlueprintBrushes/Previewing/Landmass_PreviewLandscape.Landmass_PreviewLandscape_C" (Math|Transform|MakeTransform "0, 0, 0") "Undefined" "MultiplyWithRoot" 0 0))
          (bind _output_get (Variables|Debug|SetPreview%20Landscape%20Mesh _returnvalue_2))
          (Transformation|AttachActorToActor _output_get self "None" "KeepRelative" "KeepRelative" "KeepRelative" true)))
      (Default|SpawnInstanceMeshes (Variables|Debug|GetPreview%20Landscape%20Mesh) self)
      (Rendering|SetVisibility (Variables|/Landmass/Landscape/BlueprintBrushes/Previewing/Landmass_PreviewLandscape.Landmass_PreviewLandscape_C|GetLandscape%20ISMC (Variables|Debug|GetPreview%20Landscape%20Mesh)) true false)
      (Development|ExecuteConsoleCommand "showflag.landscape 0")
      (Landmass|UpdateBrushes))))

(fn ExportRenderTarget
  (Rendering|ExportRenderTarget (Variables|RenderTargets|GetHeightmap%20Render%20RT) (Variables|DebugExport|GetFile%20Path) (Variables|DebugExport|GetFilename)))

(fn FindALandscape
  (bind _outactors (Actor|GetAllActorsOfClass "/Script/Landscape.Landscape"))
  (if (> (Utilities|Array|Length _outactors) 0)
    (Landmass|SetTargetLandscape (Utilities|Array|Get(acopy) _outactors))))

(fn FindBrushInTree
  (bind _quadtreedata_internal (Variables|Quadtree|GetQuadTreeData_Internal))
  (bind _quadtree_13_e8ca8c804ef0247f756fe3b8e015cfe7 (Utilities|Struct|BreakLandmassQuadtreeData _quadtreedata_internal))
  (Utilities|Array|Clear (LocalVariables|GetNodesFoundIn))
  (LocalVariables|SetNodesToSearch (Utilities|Array|MakeArray 0))
  (for _index (range (Variables|Quadtree|GetBrushTreeDepth))
    (LocalVariables|SetCurIndex _index)
    (if (>= (Utilities|Array|FindItem (.Actors_31_DD7B465A42CF64CE22945992A180DD0D (Utilities|Struct|BreakLandmassQuadtree (Utilities|Array|Get(aref) _quadtree_13_e8ca8c804ef0247f756fe3b8e015cfe7 _index))) (Variables|Debug|GetBrush%20Requesting%20Update)) 0)
      (Utilities|Array|Add (LocalVariables|GetNodesFoundIn) (LocalVariables|GetCurIndex)))
    (if (> (.ChildDataCount_24_488E28B5486C9F3F0D44A2A50447B62F (Utilities|Struct|BreakLandmassQuadtree (Utilities|Array|Get(aref) _quadtree_13_e8ca8c804ef0247f756fe3b8e015cfe7 _index))) 0)
      _)))

(fn GatherBrushes
  (Utilities|FlowControl|Sequence
    (:then_0
      (if (Variables|Default|GetGatherBrushesAutomatically)
        (bind _outactors (Actor|GetAllActorsOfClass "/Script/LandmassEditor.LandmassActor"))
        (for _array_element _outactors
          (Utilities|Array|AddUnique (Variables|Landmass|GetLandmassBrushes) _array_element))))))

(fn GenerateShapeMaskFromSpline
  (params (In%20Brush :type Actor%20Object%20Reference) (Fixed%20Step%20Size :type Float%20%28double-precision%29) (Flip%20Sign :type Boolean) (InTexturePath :type String))
  (bind _landscape_rt_res (Variables|Debug|LandscapeInformation|GetLandscape%20RT%20Res))
  (bind _x (Utilities|Struct|BreakIntPoint _landscape_rt_res))
  (bind (_x_2 _y) (Utilities|Struct|BreakIntPoint (Variables|Debug|LandscapeInformation|GetLandscape%20Quads)))
  (bind _intexturepath (Variables|GetInTexturePath))
  (bind _outpath (Default|ObjectPathAddOuter _intexturepath))
  (bind (_returnvalue _object) (Default|IsPathValidSoftObj _outpath))
  (Default|RasterizeSplineShape (Actor|GetComponentbyClass In%20Brush "/Script/Engine.SplineComponent") 0.0 Fixed%20Step%20Size Flip%20Sign)
  (bind _aslandmass_actor (Utilities|Casting|CastToLandmassActor In%20Brush))
  (LocalVariables|SetDrawEntireLandscape (Variables|/Script/LandmassEditor.LandmassActor|GetDrawToEntireLandscape _aslandmass_actor))
  (if (> (Math|Float|Ceil (/ (* _x (Math|Vector2D|GetMax2D (- (Default|WorldToNormalizedLandscapeCoord (Math|Conversions|ToVector(Vector2D) (Math|Vector2D|MakeVector2D (.Z (Math|Vector4|BreakVector4 (Variables|/Script/LandmassEditor.LandmassActor|GetBrushExtents _aslandmass_actor))) (.W (Math|Vector4|BreakVector4 (Variables|/Script/LandmassEditor.LandmassActor|GetBrushExtents _aslandmass_actor)))))) (Default|WorldToNormalizedLandscapeCoord (Math|Conversions|ToVector(Vector2D) (Math|Vector2D|MakeVector2D (Math|Vector4|BreakVector4 (Variables|/Script/LandmassEditor.LandmassActor|GetBrushExtents _aslandmass_actor)) (.Y (Math|Vector4|BreakVector4 (Variables|/Script/LandmassEditor.LandmassActor|GetBrushExtents _aslandmass_actor))))))))) 4.0)) 2)
    (if (LocalVariables|GetDrawEntireLandscape)
      (Rendering|CreateAndAssignRenderTarget (Variables|RenderTargets|GetCroppedRT) _x_2 _y "RTF_R8"
        (:Out
          _))
      (else
        (Rendering|CreateAndAssignRenderTarget (Variables|RenderTargets|GetCroppedRT) (Math|Integer|Min(Integer) (* (Math|Float|Ceil (/ (* _x (Math|Vector2D|GetMax2D (- (Default|WorldToNormalizedLandscapeCoord (Math|Conversions|ToVector(Vector2D) (Math|Vector2D|MakeVector2D (.Z (Math|Vector4|BreakVector4 (Variables|/Script/LandmassEditor.LandmassActor|GetBrushExtents _aslandmass_actor))) (.W (Math|Vector4|BreakVector4 (Variables|/Script/LandmassEditor.LandmassActor|GetBrushExtents _aslandmass_actor)))))) (Default|WorldToNormalizedLandscapeCoord (Math|Conversions|ToVector(Vector2D) (Math|Vector2D|MakeVector2D (Math|Vector4|BreakVector4 (Variables|/Script/LandmassEditor.LandmassActor|GetBrushExtents _aslandmass_actor)) (.Y (Math|Vector4|BreakVector4 (Variables|/Script/LandmassEditor.LandmassActor|GetBrushExtents _aslandmass_actor))))))))) 4.0)) 4) 2048) (Math|Integer|Min(Integer) (* (Math|Float|Ceil (/ (* _x (Math|Vector2D|GetMax2D (- (Default|WorldToNormalizedLandscapeCoord (Math|Conversions|ToVector(Vector2D) (Math|Vector2D|MakeVector2D (.Z (Math|Vector4|BreakVector4 (Variables|/Script/LandmassEditor.LandmassActor|GetBrushExtents _aslandmass_actor))) (.W (Math|Vector4|BreakVector4 (Variables|/Script/LandmassEditor.LandmassActor|GetBrushExtents _aslandmass_actor)))))) (Default|WorldToNormalizedLandscapeCoord (Math|Conversions|ToVector(Vector2D) (Math|Vector2D|MakeVector2D (Math|Vector4|BreakVector4 (Variables|/Script/LandmassEditor.LandmassActor|GetBrushExtents _aslandmass_actor)) (.Y (Math|Vector4|BreakVector4 (Variables|/Script/LandmassEditor.LandmassActor|GetBrushExtents _aslandmass_actor))))))))) 4.0)) 4) 2048) "RTF_R8"
          (:Out
            _))))
    (Rendering|Material|SetTextureParameterValue (Variables|Debug|GetCopyRT_MID) "RT" (Variables|RenderTargets|GetDepth%20and%20Shape%20RT%20A))
    (bind (_canvas _size _context) (Rendering|BeginDrawCanvastoRenderTarget (Variables|RenderTargets|GetCroppedRT)))
    (Canvas|DrawMaterial :self _canvas :RenderMaterial (Variables|Debug|GetCopyRT_MID) :ScreenSize _size :CoordinatePosition (.CoordinatePosition (Landmass|WorldExtentstoLandmassCoordinates (select (LocalVariables|GetDrawEntireLandscape) (Default|MakeExtents (Math|Conversions|ToVector2D(Vector) (.location (Variables|Debug|LandscapeInformation|GetLS%20Transform))) (Math|Conversions|ToVector2D(Vector) (+ (.location (Variables|Debug|LandscapeInformation|GetLS%20Transform)) (Variables|Debug|LandscapeInformation|GetWorld%20Size)))) (Default|MakeExtents (Math|Vector2D|MakeVector2D (Math|Vector4|BreakVector4 (Variables|/Script/LandmassEditor.LandmassActor|GetBrushExtents _aslandmass_actor)) (.Y (Math|Vector4|BreakVector4 (Variables|/Script/LandmassEditor.LandmassActor|GetBrushExtents _aslandmass_actor)))) (Math|Vector2D|MakeVector2D (.Z (Math|Vector4|BreakVector4 (Variables|/Script/LandmassEditor.LandmassActor|GetBrushExtents _aslandmass_actor))) (.W (Math|Vector4|BreakVector4 (Variables|/Script/LandmassEditor.LandmassActor|GetBrushExtents _aslandmass_actor)))))) (Variables|Landmass|GetLandscapeInformation))) :CoordinateSize (.CoordinateSize (Landmass|WorldExtentstoLandmassCoordinates (select (LocalVariables|GetDrawEntireLandscape) (Default|MakeExtents (Math|Conversions|ToVector2D(Vector) (.location (Variables|Debug|LandscapeInformation|GetLS%20Transform))) (Math|Conversions|ToVector2D(Vector) (+ (.location (Variables|Debug|LandscapeInformation|GetLS%20Transform)) (Variables|Debug|LandscapeInformation|GetWorld%20Size)))) (Default|MakeExtents (Math|Vector2D|MakeVector2D (Math|Vector4|BreakVector4 (Variables|/Script/LandmassEditor.LandmassActor|GetBrushExtents _aslandmass_actor)) (.Y (Math|Vector4|BreakVector4 (Variables|/Script/LandmassEditor.LandmassActor|GetBrushExtents _aslandmass_actor)))) (Math|Vector2D|MakeVector2D (.Z (Math|Vector4|BreakVector4 (Variables|/Script/LandmassEditor.LandmassActor|GetBrushExtents _aslandmass_actor))) (.W (Math|Vector4|BreakVector4 (Variables|/Script/LandmassEditor.LandmassActor|GetBrushExtents _aslandmass_actor)))))) (Variables|Landmass|GetLandscapeInformation))) :Rotation 0.0 :PivotPoint "(X=0.500000,Y=0.500000)")
    (Rendering|EndDrawCanvastoRenderTarget _context)
    (bind _distance_rt (Class|JumpFloodComponent2D|JumpFloodAndComputeDistance (Variables|Default|GetJumpFlood_Component2D) (Variables|RenderTargets|GetCroppedRT) (Math|Color|MakeColor 16.0 0.0 5.0 0.0) 8))
    (Variables|RenderTargets|SetDistanceFieldRT _distance_rt)
    (if (Utilities|String|EqualExactly(String) _intexturepath)
      (bind _returnvalue_2 (Game|ConstructObjectfromClass "/Script/Engine.Texture2D"))
      (EditorScripting|Rendering|ConvertRenderTargettoTexture2DEditorOnly (Variables|RenderTargets|GetDistanceFieldRT) _returnvalue_2)
      (Variables|/Script/Engine.Texture|SetCompressionSettings "TC_SingleFloat" _returnvalue_2)
      (Rendering|ForceUpdateTexture _returnvalue_2)
      (LocalVariables|SetTex _returnvalue_2)
      (elif _returnvalue
        (bind _astexture_2d (Utilities|Casting|CastToTexture2D _object))
        (EditorScripting|Rendering|ConvertRenderTargettoTexture2DEditorOnly (Variables|RenderTargets|GetDistanceFieldRT) _astexture_2d)
        (Transactions|TransactObject _astexture_2d)
        (Variables|/Script/Engine.Texture|SetCompressionSettings "TC_SingleFloat" _astexture_2d)
        (Rendering|ForceUpdateTexture _astexture_2d)
        (LocalVariables|SetTex _astexture_2d)
        (else
          (bind _returnvalue_3 (EditorScripting|Rendering|RenderTarget2DCreateStaticTexture2DEditorOnly (Variables|RenderTargets|GetDistanceFieldRT) _intexturepath "TC_SingleFloat"))
          (LocalVariables|SetTex _returnvalue_3))))
    (Class|LandmassInterface|SetShapeMask(Message) (Variables|GetIn%20Brush) (LocalVariables|GetTex) (Utilities|Array|MakeArray "0, 0, 0") (Default|MakeExtents))
    (CallFunction|UpdateBrushes)
    (else
      (Development|PrintString "Shape Finalization Abandoned because resulting shape was too small."))))

(fn GetBrushesinModifiedNodes
  (bind (_actors_31_dd7b465a42cf64ce22945992a180dd0d _childdatacount_24_488e28b5486c9f3f0d44a2a50447b62f) (Utilities|Struct|BreakLandmassQuadtree (Utilities|Array|Get(aref) (Utilities|Struct|BreakLandmassQuadtreeData (Variables|Quadtree|GetQuadTreeData_Internal)) (LocalVariables|GetCurrentNode))))
  (bind _currentnode (LocalVariables|GetCurrentNode))
  (bind _quadtreedata_internal (Variables|Quadtree|GetQuadTreeData_Internal))
  (bind _quadtree_13_e8ca8c804ef0247f756fe3b8e015cfe7 (Utilities|Struct|BreakLandmassQuadtreeData _quadtreedata_internal))
  (bind _output (Utilities|Array|Get(aref) _quadtree_13_e8ca8c804ef0247f756fe3b8e015cfe7 _currentnode))
  (bind _parentindex_26_2769b7c54152baa5828c46abcef62ba9 (Utilities|Struct|BreakLandmassQuadtree _output))
  (bind (_actors_31_dd7b465a42cf64ce22945992a180dd0d_2 _index_x0y0_7_da9d91e24f764d7dbd4d3f9fb8c3c746 _index_x1y0_8_6dad9ca94fe643e04cc00c80d2755731 _index_x0y1_10_e78cc5ea43dcd1ef149c1b8b4baf36b4 _index_x1y1_12_04d22aeb4fa65d7b52d777a53440e336) (Utilities|Struct|BreakLandmassQuadtree (Utilities|Array|Get(aref) (Utilities|Struct|BreakLandmassQuadtreeData (Variables|Quadtree|GetQuadTreeData_Internal)) (LocalVariables|GetCurrentNode))))
  (bind _currentnode_2 (LocalVariables|GetCurrentNode))
  (bind _quadtreedata_internal_2 (Variables|Quadtree|GetQuadTreeData_Internal))
  (bind _quadtree_13_e8ca8c804ef0247f756fe3b8e015cfe7_2 (Utilities|Struct|BreakLandmassQuadtreeData _quadtreedata_internal_2))
  (bind _output_2 (Utilities|Array|Get(aref) _quadtree_13_e8ca8c804ef0247f756fe3b8e015cfe7_2 _currentnode_2))
  (bind (_currentlevel_20_876bc8b8418658909b9774b3700dc438 _actors_31_dd7b465a42cf64ce22945992a180dd0d_3) (Utilities|Struct|BreakLandmassQuadtree _output_2))
  (bind _landmassbrushesbp (Variables|Landmass|GetLandmassBrushesBP))
  (bind _landmassbrushesbp_2 (Variables|Landmass|GetLandmassBrushesBP))
  (Utilities|Array|Clear (Variables|Quadtree|GetBrushesRequiringUpdate))
  (bind _output_get (LocalVariables|SetNodesToCheck (Variables|Quadtree|GetCurrentModifiedNodes)))
  (LocalVariables|SetParentsToCheck _output_get)
  (Utilities|FlowControl|Sequence
    (:then_0
      ;[0.0697 1 0.4721] Check Each Level of Quadtree
      (for _index (range (+ (Variables|Quadtree|GetBrushTreeDepth) 1))
        (Utilities|FlowControl|Sequence
          (:then_0
            ;[0.09567 0.6308 1] Check all the Nodes in our current list
            (Utilities|Array|Clear (LocalVariables|GetNodesToCheckNext))
            (for _array_element (LocalVariables|GetNodesToCheck)
              (LocalVariables|SetCurrentNode _array_element)
              (Utilities|FlowControl|Sequence
                (:then_0
                  ;[0.06453 0.2914 1] Add Actors from Current Node
                  (LocalVariables|SetCurrentLevel _currentlevel_20_876bc8b8418658909b9774b3700dc438)
                  (if (> (Utilities|Array|Length _actors_31_dd7b465a42cf64ce22945992a180dd0d_3) 0)
                    (for _array_element_2 _actors_31_dd7b465a42cf64ce22945992a180dd0d_3
                      (if (>= (Utilities|Array|FindItem _landmassbrushesbp_2 _array_element_2) 0)
                        (bind _aslandmass_actor (Utilities|Casting|CastToLandmassActor _array_element_2))
                        (Utilities|Array|SetArrayElem (Variables|Quadtree|GetBrushesRequiringUpdate) (Utilities|Array|FindItem _landmassbrushesbp_2 _array_element_2) _aslandmass_actor true)))))
                (:then_1
                  ;[0.2339 1 0.8497] Set Next Loop to check the Child indices of the current node if Child Data found
                  (if (< (LocalVariables|GetCurrentLevel) (Variables|Quadtree|GetBrushTreeDepth))
                    (if (> _childdatacount_24_488e28b5486c9f3f0d44a2a50447b62f 0)
                      (Utilities|Array|AppendArray (LocalVariables|GetNodesToCheckNext) (Utilities|Array|MakeArray _index_x0y0_7_da9d91e24f764d7dbd4d3f9fb8c3c746 _index_x1y0_8_6dad9ca94fe643e04cc00c80d2755731 _index_x0y1_10_e78cc5ea43dcd1ef149c1b8b4baf36b4 _index_x1y1_12_04d22aeb4fa65d7b52d777a53440e336)) ; If child node data exists, add children check next iteration
                    )))))
            (LocalVariables|SetNodesToCheck (LocalVariables|GetNodesToCheckNext)))
          (:then_1
            (Utilities|Array|Clear (LocalVariables|GetParentsToCheckNext))
            (for _array_element_3 (LocalVariables|GetParentsToCheck)
              (bind _output_get_2 (LocalVariables|SetCurrentNode _array_element_3))
              ;[1 0.7242 0.1504] Always include Actors from Higher in Tree
              (bind _output_get_3 (LocalVariables|SetCurrentParent _parentindex_26_2769b7c54152baa5828c46abcef62ba9))
              (if (>= _output_get_3 0)
                (for _array_element_4 (Utilities|Struct|BreakLandmassQuadtree (Utilities|Array|Get(aref) (Utilities|Struct|BreakLandmassQuadtreeData (Variables|Quadtree|GetQuadTreeData_Internal)) _parentindex_26_2769b7c54152baa5828c46abcef62ba9))
                  (if (>= (Utilities|Array|FindItem _landmassbrushesbp _array_element_4) 0)
                    (bind _aslandmass_actor_2 (Utilities|Casting|CastToLandmassActor _array_element_4))
                    (Utilities|Array|SetArrayElem (Variables|Quadtree|GetBrushesRequiringUpdate) (Utilities|Array|FindItem _landmassbrushesbp _array_element_4) _aslandmass_actor_2 true)))
                (Utilities|Array|AddUnique (LocalVariables|GetParentsToCheckNext) (LocalVariables|GetCurrentParent))))
            (LocalVariables|SetParentsToCheck (LocalVariables|GetParentsToCheckNext))))))
    (:then_1
      (Utilities|FlowControl|Sequence
        (:then_1
          (if (Variables|Quadtree|GetDebug%20Text)
            (Development|PrintString "Completed Finding Brushes in modified Nodes")
            (for _array_element_5 (Variables|Quadtree|GetBrushesRequiringUpdate)
              (Utilities|IsValid _array_element_5
                (:Is%20Valid
                  (Development|PrintString (Utilities|String|Append "Brush requires Update: " (Utilities|GetDisplayName _array_element_5))))))))))))

(fn GetNodesForExtents
  (params (In%20Extents :type Vector%204%20Structure))
  (bind _quadtreedata_internal (Variables|Quadtree|GetQuadTreeData_Internal))
  (bind _quadtree_13_e8ca8c804ef0247f756fe3b8e015cfe7 (Utilities|Struct|BreakLandmassQuadtreeData _quadtreedata_internal))
  (bind (_min _max) (Default|BreakExtents (Variables|GetIn%20Extents)))
  (bind (_x _y) (Math|Vector2D|BreakVector2D _max))
  (bind _in_extents (Variables|GetIn%20Extents))
  (bind (_min_2 _max_2) (Default|BreakExtents _in_extents))
  (bind (_x_2 _y_2) (Math|Vector2D|BreakVector2D (LocalVariables|GetCurNodeCenter)))
  (bind (_x_3 _y_3) (Math|Vector2D|BreakVector2D (LocalVariables|GetCurNodeCenter)))
  (bind (_x_4 _y_4) (Math|Vector2D|BreakVector2D (LocalVariables|GetCurNodeCenter)))
  (bind (_x_5 _y_5) (Math|Vector2D|BreakVector2D (LocalVariables|GetCurNodeCenter)))
  (bind (_min_3 _max_3) (Default|BreakExtents (Variables|GetIn%20Extents)))
  (bind (_x_6 _y_6) (Math|Vector2D|BreakVector2D _max_3))
  (bind (_x_7 _y_7) (Math|Vector2D|BreakVector2D _min_2))
  (bind (_x_8 _y_8) (Math|Vector2D|BreakVector2D (Default|BreakExtents (Variables|GetIn%20Extents))))
  (bind _quadtreedata_internal_2 (Variables|Quadtree|GetQuadTreeData_Internal))
  (bind _quadtree_13_e8ca8c804ef0247f756fe3b8e015cfe7_2 (Utilities|Struct|BreakLandmassQuadtreeData _quadtreedata_internal_2))
  (bind _quadtreedata_internal_3 (Variables|Quadtree|GetQuadTreeData_Internal))
  (bind _quadtree_13_e8ca8c804ef0247f756fe3b8e015cfe7_3 (Utilities|Struct|BreakLandmassQuadtreeData _quadtreedata_internal_3))
  (bind _overlappednodes (LocalVariables|GetOverlappedNodes))
  (Utilities|FlowControl|Sequence
    (:then_0
      ;[0.2926 1 0.8393] Check each level of Tree
      (LocalVariables|SetOverlappedNodes (Utilities|Array|MakeArray 0))
      (for _ (range (Variables|Quadtree|GetBrushTreeDepth))
        ;[0.06105 0.3733 1] Check if Brush Overlaps Current Node
        (Utilities|Array|Clear (LocalVariables|GetNodesToCheck_NextLevel))
        (for _array_element (LocalVariables|GetOverlappedNodes)
          (LocalVariables|SetCurNodeCenter (/ (+ (Default|BreakExtents (.Extents_29_CFC7F4FD40A19AB21C4959B310ECE410 (Utilities|Struct|BreakLandmassQuadtree (Utilities|Array|Get(aref) _quadtree_13_e8ca8c804ef0247f756fe3b8e015cfe7 _array_element)))) (.Max (Default|BreakExtents (.Extents_29_CFC7F4FD40A19AB21C4959B310ECE410 (Utilities|Struct|BreakLandmassQuadtree (Utilities|Array|Get(aref) _quadtree_13_e8ca8c804ef0247f756fe3b8e015cfe7 _array_element)))))) 2.0))
          (Utilities|FlowControl|Sequence
            (:then_0
              (if (and (< _x_7 _x_2) (< _y_7 _y_2))
                (Utilities|Array|Add (LocalVariables|GetNodesToCheck_NextLevel) (.index_x0y0_7_DA9D91E24F764D7DBD4D3F9FB8C3C746 (Utilities|Struct|BreakLandmassQuadtree (Utilities|Array|Get(aref) _quadtree_13_e8ca8c804ef0247f756fe3b8e015cfe7 _array_element))))))
            (:then_1
              (if (and (> (Math|Vector2D|BreakVector2D _max_2) _x_3) (< _y_8 _y_3))
                (Utilities|Array|Add (LocalVariables|GetNodesToCheck_NextLevel) (.index_x1y0_8_6DAD9CA94FE643E04CC00C80D2755731 (Utilities|Struct|BreakLandmassQuadtree (Utilities|Array|Get(aref) _quadtree_13_e8ca8c804ef0247f756fe3b8e015cfe7 _array_element))))))
            (:then_2
              (if (and (< (Math|Vector2D|BreakVector2D _min_3) _x_5) (> _y_6 _y_5))
                (Utilities|Array|Add (LocalVariables|GetNodesToCheck_NextLevel) (.index_x0y1_10_E78CC5EA43DCD1EF149C1B8B4BAF36B4 (Utilities|Struct|BreakLandmassQuadtree (Utilities|Array|Get(aref) _quadtree_13_e8ca8c804ef0247f756fe3b8e015cfe7 _array_element))))))
            (:then_3
              (if (and (> _x _x_4) (> _y _y_4))
                (Utilities|Array|Add (LocalVariables|GetNodesToCheck_NextLevel) (.index_x1y1_12_04D22AEB4FA65D7B52D777A53440E336 (Utilities|Struct|BreakLandmassQuadtree (Utilities|Array|Get(aref) _quadtree_13_e8ca8c804ef0247f756fe3b8e015cfe7 _array_element))))))))
        (LocalVariables|SetOverlappedNodes (LocalVariables|GetNodesToCheck_NextLevel))))
    (:then_1
      ;[1 1 1] Consolidate Parent Nodes and Add Brush Actor to Data
      ;; Consoldate Nodes using Hierarchy
      (for _ (range (Variables|Quadtree|GetBrushTreeDepth))
        (Utilities|Array|Clear (LocalVariables|GetFullyOccupiedParentNodes))
        (for _array_element_2 (LocalVariables|GetOverlappedNodes)
          (if (Math|Boolean|ANDBoolean (>= (Utilities|Array|FindItem (LocalVariables|GetOverlappedNodes) (.index_x0y0_7_DA9D91E24F764D7DBD4D3F9FB8C3C746 (Utilities|Struct|BreakLandmassQuadtree (Utilities|Array|Get(aref) _quadtree_13_e8ca8c804ef0247f756fe3b8e015cfe7_2 (.ParentIndex_26_2769B7C54152BAA5828C46ABCEF62BA9 (Utilities|Struct|BreakLandmassQuadtree (Utilities|Array|Get(aref) _quadtree_13_e8ca8c804ef0247f756fe3b8e015cfe7_2 _array_element_2))))))) 0) (>= (Utilities|Array|FindItem (LocalVariables|GetOverlappedNodes) (.index_x1y0_8_6DAD9CA94FE643E04CC00C80D2755731 (Utilities|Struct|BreakLandmassQuadtree (Utilities|Array|Get(aref) _quadtree_13_e8ca8c804ef0247f756fe3b8e015cfe7_2 (.ParentIndex_26_2769B7C54152BAA5828C46ABCEF62BA9 (Utilities|Struct|BreakLandmassQuadtree (Utilities|Array|Get(aref) _quadtree_13_e8ca8c804ef0247f756fe3b8e015cfe7_2 _array_element_2))))))) 0) (>= (Utilities|Array|FindItem (LocalVariables|GetOverlappedNodes) (.index_x0y1_10_E78CC5EA43DCD1EF149C1B8B4BAF36B4 (Utilities|Struct|BreakLandmassQuadtree (Utilities|Array|Get(aref) _quadtree_13_e8ca8c804ef0247f756fe3b8e015cfe7_2 (.ParentIndex_26_2769B7C54152BAA5828C46ABCEF62BA9 (Utilities|Struct|BreakLandmassQuadtree (Utilities|Array|Get(aref) _quadtree_13_e8ca8c804ef0247f756fe3b8e015cfe7_2 _array_element_2))))))) 0) (>= (Utilities|Array|FindItem (LocalVariables|GetOverlappedNodes) (.index_x1y1_12_04D22AEB4FA65D7B52D777A53440E336 (Utilities|Struct|BreakLandmassQuadtree (Utilities|Array|Get(aref) _quadtree_13_e8ca8c804ef0247f756fe3b8e015cfe7_2 (.ParentIndex_26_2769B7C54152BAA5828C46ABCEF62BA9 (Utilities|Struct|BreakLandmassQuadtree (Utilities|Array|Get(aref) _quadtree_13_e8ca8c804ef0247f756fe3b8e015cfe7_2 _array_element_2))))))) 0))
            (Utilities|Array|AddUnique (LocalVariables|GetFullyOccupiedParentNodes) (.ParentIndex_26_2769B7C54152BAA5828C46ABCEF62BA9 (Utilities|Struct|BreakLandmassQuadtree (Utilities|Array|Get(aref) _quadtree_13_e8ca8c804ef0247f756fe3b8e015cfe7_2 _array_element_2))))))
        (for _array_element_3 (LocalVariables|GetFullyOccupiedParentNodes)
          (Utilities|Array|RemoveItem _overlappednodes (.index_x0y0_7_DA9D91E24F764D7DBD4D3F9FB8C3C746 (Utilities|Struct|BreakLandmassQuadtree (Utilities|Array|Get(aref) _quadtree_13_e8ca8c804ef0247f756fe3b8e015cfe7_3 _array_element_3))))
          (Utilities|Array|RemoveItem _overlappednodes (.index_x1y0_8_6DAD9CA94FE643E04CC00C80D2755731 (Utilities|Struct|BreakLandmassQuadtree (Utilities|Array|Get(aref) _quadtree_13_e8ca8c804ef0247f756fe3b8e015cfe7_3 _array_element_3))))
          (Utilities|Array|RemoveItem _overlappednodes (.index_x0y1_10_E78CC5EA43DCD1EF149C1B8B4BAF36B4 (Utilities|Struct|BreakLandmassQuadtree (Utilities|Array|Get(aref) _quadtree_13_e8ca8c804ef0247f756fe3b8e015cfe7_3 _array_element_3))))
          (Utilities|Array|RemoveItem _overlappednodes (.index_x1y1_12_04D22AEB4FA65D7B52D777A53440E336 (Utilities|Struct|BreakLandmassQuadtree (Utilities|Array|Get(aref) _quadtree_13_e8ca8c804ef0247f756fe3b8e015cfe7_3 _array_element_3))))
          (Utilities|Array|Insert (LocalVariables|GetOverlappedNodes) _array_element_3))))
    (:then_2
      (return (LocalVariables|GetOverlappedNodes)))))

(fn GetShapeMask
  (return ""))

(fn GetBlueprintRenderDependencies
  (Utilities|Array|Clear (LocalVariables|GetRenderDependencies))
  (for _array_element (Variables|Landmass|GetLandmassBrushes)
    (Utilities|IsValid _array_element
      (:Is%20Valid
        (if (Utilities|DoesObjectImplementInterface _array_element "/Landmass/Landscape/BlueprintBrushes/Interface/Landmass_Interface.Landmass_Interface_C")
          (bind _objects (Class|LandmassInterface|GetRenderDependencies(Message) _array_element))
          (Utilities|Array|AppendArray (LocalVariables|GetRenderDependencies) _objects)))))
  (return (LocalVariables|GetRenderDependencies)))

(fn GetRenderDependencies
  (return 0))

(fn LaunchLandmassEditor
  (bind _returnvalue (Game|SpawnActorfromClass "/Landmass/Landscape/ErosionBrush/BP/Widgets/EUB_SpawnErosionTab.EUB_SpawnErosionTab_C" (Math|Conversions|ToTransform(Vector) (Math|Vector|MakeVector 0.0)) "Undefined" "MultiplyWithRoot" 0 0))
  (Class|EUBSpawnErosionTab|SpawnLandmassEditor _returnvalue self 0)
  (Actor|DestroyActor _returnvalue))

(fn LaunchLandmassEditor
  (params (BrushRequestingEditor :type Landmass%20Actor%20Object%20Reference))
  (bind _returnvalue (Game|SpawnActorfromClass "/Landmass/Landscape/ErosionBrush/BP/Widgets/EUB_SpawnErosionTab.EUB_SpawnErosionTab_C" (Math|Conversions|ToTransform(Vector) (Math|Vector|MakeVector 0.0)) "Undefined" "MultiplyWithRoot" 0 0))
  (Class|EUBSpawnErosionTab|SpawnLandmassEditor _returnvalue self (Variables|GetBrushRequestingEditor))
  (Actor|DestroyActor _returnvalue))

(fn ListChildren
  (for (_array_element _array_index) (Utilities|Struct|BreakLandmassQuadtreeData (Variables|Quadtree|GetQuadTreeData_Internal))
    (if (or (> (Utilities|Array|Length (.Actors_31_DD7B465A42CF64CE22945992A180DD0D (Utilities|Struct|BreakLandmassQuadtree _array_element))) 0) (> (.ChildDataCount_24_488E28B5486C9F3F0D44A2A50447B62F (Utilities|Struct|BreakLandmassQuadtree _array_element)) 0))
      (Development|PrintString (Utilities|String|Append "Node Index: " (Utilities|String|ToString(Integer) _array_index) " had " (Utilities|String|ToString(Integer) (Utilities|Array|Length (.Actors_31_DD7B465A42CF64CE22945992A180DD0D (Utilities|Struct|BreakLandmassQuadtree _array_element)))) "  actors and then: " (Utilities|String|ToString(Integer) (.ChildDataCount_24_488E28B5486C9F3F0D44A2A50447B62F (Utilities|Struct|BreakLandmassQuadtree _array_element))) " children")))))

(fn PopulateTree
  (bind (_x _y) (Math|Vector2D|BreakVector2D (+ (+ (* (Math|Conversions|ToVector2D(IntPoint) (Math|Conversions|Indices|Converta1DIndextoa2DIndex (LocalVariables|GetCur%20Node) (LocalVariables|GetSize%20of%20Level))) (LocalVariables|GetWorldSize%20of%20Current%20Node)) (Math|Conversions|ToVector2D(Vector) (.location (Utilities|Struct|BreakLandmassLandscapeInfo (Variables|Landmass|GetLandscapeInformation))))) (LocalVariables|GetWorldSize%20of%20Current%20Node))))
  (bind (_x_2 _y_2) (Math|Vector2D|BreakVector2D (+ (* (Math|Conversions|ToVector2D(IntPoint) (Math|Conversions|Indices|Converta1DIndextoa2DIndex (LocalVariables|GetCur%20Node) (LocalVariables|GetSize%20of%20Level))) (LocalVariables|GetWorldSize%20of%20Current%20Node)) (Math|Conversions|ToVector2D(Vector) (.location (Utilities|Struct|BreakLandmassLandscapeInfo (Variables|Landmass|GetLandscapeInformation)))))))
  (if (Variables|Quadtree|GetUse%20Nativized%20Tree)
    (Landmass|PopulateNodeTree)
    (for _array_element (Variables|Landmass|GetLandmassBrushes)
      (Utilities|IsValid _array_element
        (:Is%20Valid
          (Landmass|AddBrushtoTree _array_element (Variables|/Script/LandmassEditor.LandmassActor|GetBrushExtents _array_element) (Variables|/Script/LandmassEditor.LandmassActor|GetDrawToEntireLandscape _array_element)))))
    (else
      (Utilities|Array|Clear (Utilities|Struct|BreakLandmassQuadtreeData (Variables|Quadtree|GetQuadTreeData_Internal)))
      (Variables|Quadtree|SetQuadtree%20Leaf%20Node%20Size (/ (.x (Variables|Debug|LandscapeInformation|GetWorld%20Size)) (Math|Float|Power 2.0 (Math|Conversions|ToFloat(Integer) (Variables|Quadtree|GetBrushTreeDepth)))))
      (for _index (range (+ (Variables|Quadtree|GetBrushTreeDepth) 1))
        (LocalVariables|SetCurrentLevel _index)
        (bind _output_get (LocalVariables|SetSize%20of%20Level (Math|Float|Truncate (Math|Float|Power 2.0 (Math|Conversions|ToFloat(Integer) (LocalVariables|GetCurrentLevel))))))
        (LocalVariables|SetNodes%20in%20Level (* _output_get _output_get))
        (LocalVariables|SetSize%20of%20Next%20Level (Math|Float|Truncate (Math|Float|Power 2.0 (Math|Conversions|ToFloat(Integer) (+ (LocalVariables|GetCurrentLevel) 1)))))
        (LocalVariables|SetNode%20Index%20Offset (+ (LocalVariables|GetNodes%20in%20Level) (LocalVariables|GetNode%20Index%20Offset)))
        (for _index_2 (range (LocalVariables|GetNodes%20in%20Level))
          (LocalVariables|SetCur%20Node _index_2)
          (LocalVariables|SetWorldSize%20of%20Current%20Node (/ (.x (Variables|Debug|LandscapeInformation|GetWorld%20Size)) (LocalVariables|GetSize%20of%20Level)))
          (LocalVariables|SetExtents (Math|Vector4|MakeVector4 _x_2 _y_2 _x _y))
          (Utilities|Array|Add (Utilities|Struct|BreakLandmassQuadtreeData (Variables|Quadtree|GetQuadTreeData_Internal)) (select (< (LocalVariables|GetCurrentLevel) (Variables|Quadtree|GetBrushTreeDepth)) (Utilities|Struct|MakeLandmassQuadtree :CurrentLevel_20_876BC8B8418658909B9774B3700DC438 (LocalVariables|GetCurrentLevel) :ParentIndex_26_2769B7C54152BAA5828C46ABCEF62BA9 -1 :index_x0y0_7_DA9D91E24F764D7DBD4D3F9FB8C3C746 (+ (Math|Conversions|Indices|Converta2DIndextoa1DIndex (+ (Utilities|Struct|MakeIntPoint 0) (* (Math|Conversions|Indices|Converta1DIndextoa2DIndex (LocalVariables|GetCur%20Node) (LocalVariables|GetSize%20of%20Level)) (Utilities|Struct|MakeIntPoint 2 2))) (LocalVariables|GetSize%20of%20Next%20Level)) (LocalVariables|GetNode%20Index%20Offset)) :index_x1y0_8_6DAD9CA94FE643E04CC00C80D2755731 (+ (Math|Conversions|Indices|Converta2DIndextoa1DIndex (+ (Utilities|Struct|MakeIntPoint 1) (* (Math|Conversions|Indices|Converta1DIndextoa2DIndex (LocalVariables|GetCur%20Node) (LocalVariables|GetSize%20of%20Level)) (Utilities|Struct|MakeIntPoint 2 2))) (LocalVariables|GetSize%20of%20Next%20Level)) (LocalVariables|GetNode%20Index%20Offset)) :index_x0y1_10_E78CC5EA43DCD1EF149C1B8B4BAF36B4 (+ (Math|Conversions|Indices|Converta2DIndextoa1DIndex (+ (Utilities|Struct|MakeIntPoint 0 1) (* (Math|Conversions|Indices|Converta1DIndextoa2DIndex (LocalVariables|GetCur%20Node) (LocalVariables|GetSize%20of%20Level)) (Utilities|Struct|MakeIntPoint 2 2))) (LocalVariables|GetSize%20of%20Next%20Level)) (LocalVariables|GetNode%20Index%20Offset)) :index_x1y1_12_04D22AEB4FA65D7B52D777A53440E336 (+ (Math|Conversions|Indices|Converta2DIndextoa1DIndex (+ (Utilities|Struct|MakeIntPoint 1 1) (* (Math|Conversions|Indices|Converta1DIndextoa2DIndex (LocalVariables|GetCur%20Node) (LocalVariables|GetSize%20of%20Level)) (Utilities|Struct|MakeIntPoint 2 2))) (LocalVariables|GetSize%20of%20Next%20Level)) (LocalVariables|GetNode%20Index%20Offset)) :ChildDataCount_24_488E28B5486C9F3F0D44A2A50447B62F 0 :Extents_29_CFC7F4FD40A19AB21C4959B310ECE410 (LocalVariables|GetExtents)) (Utilities|Struct|MakeLandmassQuadtree :CurrentLevel_20_876BC8B8418658909B9774B3700DC438 (LocalVariables|GetCurrentLevel) :ParentIndex_26_2769B7C54152BAA5828C46ABCEF62BA9 -1 :index_x0y0_7_DA9D91E24F764D7DBD4D3F9FB8C3C746 -1 :index_x1y0_8_6DAD9CA94FE643E04CC00C80D2755731 -1 :index_x0y1_10_E78CC5EA43DCD1EF149C1B8B4BAF36B4 -1 :index_x1y1_12_04D22AEB4FA65D7B52D777A53440E336 -1 :ChildDataCount_24_488E28B5486C9F3F0D44A2A50447B62F 0 :Extents_29_CFC7F4FD40A19AB21C4959B310ECE410 (LocalVariables|GetExtents))))))
      ;[1 1 1] Initialize Parent Indices
      (for (_array_element_2 _array_index) (Utilities|Struct|BreakLandmassQuadtreeData (Variables|Quadtree|GetQuadTreeData_Internal))
        (for _array_element_3 (Utilities|Array|MakeArray (.index_x0y0_7_DA9D91E24F764D7DBD4D3F9FB8C3C746 (Utilities|Struct|BreakLandmassQuadtree _array_element_2)) (.index_x1y0_8_6DAD9CA94FE643E04CC00C80D2755731 (Utilities|Struct|BreakLandmassQuadtree _array_element_2)) (.index_x0y1_10_E78CC5EA43DCD1EF149C1B8B4BAF36B4 (Utilities|Struct|BreakLandmassQuadtree _array_element_2)) (.index_x1y1_12_04D22AEB4FA65D7B52D777A53440E336 (Utilities|Struct|BreakLandmassQuadtree _array_element_2)))
          (if (>= _array_element_3 0)
            (bind _structout (Utilities|Struct|SetmembersinLandmassQuadtree (Utilities|Array|Get(aref) (Utilities|Struct|BreakLandmassQuadtreeData (Variables|Quadtree|GetQuadTreeData_Internal)) _array_element_3) _array_index))
            (Utilities|Array|SetArrayElem (Utilities|Struct|BreakLandmassQuadtreeData (Variables|Quadtree|GetQuadTreeData_Internal)) _array_element_3 _structout)))))))

(fn PrintNoteDebug
  (for (_array_element _array_index) (Variables|Quadtree|GetBrushNodeData)
    (LocalVariables|SetCurrentActorCount (Utilities|Array|Length (Utilities|Struct|BreakBrushDataTree _array_element)))
    (if (> (LocalVariables|GetCurrentActorCount) 0)
      (Development|PrintString (Utilities|String|Append "Node Index: " (Utilities|String|ToString(Integer) _array_index) " had " (Utilities|String|ToString(Integer) (LocalVariables|GetCurrentActorCount)) "  actors and " (Utilities|String|ToString(Integer) (.ChildDataCount (Utilities|Struct|BreakBrushDataTree _array_element))) " total child elements"))
      (for _array_element_2 (Utilities|Struct|BreakBrushDataTree _array_element)
        (Development|PrintString (Utilities|String|Append "child: " (Utilities|GetDisplayName _array_element_2)))))))

(fn RasterizeSplineShape
  (params (Spline :type Spline%20Component%20Object%20Reference) (Z%20Position :type Float%20%28double-precision%29) (Fixed%20Sample%20Length :type Float%20%28double-precision%29) (FlipSign :type Boolean))
  (bind (_x _y) (Utilities|Struct|BreakIntPoint (Variables|Debug|LandscapeInformation|GetLandscape%20RT%20Res)))
  (bind _rastershapemid (Variables|Debug|GetRasterShapeMID))
  (bind _depth_and_shape_rt_a (Variables|RenderTargets|GetDepth%20and%20Shape%20RT%20A))
  (bind _newparam (Class|LandmassFunctionLibrary|SplinePointsToCanvasUVTris Spline (Variables|Landmass|GetLandscapeInformation) Fixed%20Sample%20Length))
  (LocalVariables|SetCanvas%20Tris _newparam)
  (Rendering|CreateAndAssignRenderTarget (Variables|RenderTargets|GetDepth%20and%20Shape%20RT%20A) _x _y "RTF_R32f"
    (:Out
      (Rendering|ClearRenderTarget2D _depth_and_shape_rt_a)
      (bind (_canvas _size _context) (Rendering|BeginDrawCanvastoRenderTarget _depth_and_shape_rt_a))
      (Rendering|CreateAndAssignMID _rastershapemid "/Landmass/Landscape/OldPrototype_BP/Mats/Debug/Intersections.Intersections"
        (:Out
          (Rendering|Material|SetScalarParameterValue _rastershapemid "Z" (Variables|GetZ%20Position) :!class /Script/Engine.MaterialInstanceDynamic)
          (Rendering|Material|SetScalarParameterValue _rastershapemid "Sign" (Math|Float|SelectFloat -1.0 1.0 (Variables|GetFlipSign)) :!class /Script/Engine.MaterialInstanceDynamic)
          (Canvas|DrawMaterialTriangles _canvas (Variables|Debug|GetRasterShapeMID) (LocalVariables|GetCanvas%20Tris))
          (Rendering|EndDrawCanvastoRenderTarget _context))))))

(fn RemoveActorFromNode
  (params (In_Brush :type Actor%20Object%20Reference) (In_NodeIndex :type Integer))
  (if (>= (Utilities|Array|FindItem (Utilities|Struct|BreakLandmassQuadtree (Utilities|Array|Get(aref) (Utilities|Struct|BreakLandmassQuadtreeData (Variables|Quadtree|GetQuadTreeData_Internal)) (Variables|GetIn_NodeIndex))) (Variables|GetIn_Brush)) 0)
    (bind _returnvalue (Utilities|Array|RemoveItem (Utilities|Struct|BreakLandmassQuadtree (Utilities|Array|Get(aref) (Utilities|Struct|BreakLandmassQuadtreeData (Variables|Quadtree|GetQuadTreeData_Internal)) (Variables|GetIn_NodeIndex))) (Variables|GetIn_Brush)))
    (LocalVariables|SetWasRemoved _returnvalue)
    (else
      (LocalVariables|SetWasRemoved false)))
  (return (LocalVariables|GetWasRemoved)))

(fn RemoveActorFromQuadtree
  (bind (_actors_31_dd7b465a42cf64ce22945992a180dd0d _index_x0y0_7_da9d91e24f764d7dbd4d3f9fb8c3c746 _index_x1y0_8_6dad9ca94fe643e04cc00c80d2755731 _index_x0y1_10_e78cc5ea43dcd1ef149c1b8b4baf36b4 _index_x1y1_12_04d22aeb4fa65d7b52d777a53440e336) (Utilities|Struct|BreakLandmassQuadtree (Utilities|Array|Get(aref) (Utilities|Struct|BreakLandmassQuadtreeData (Variables|Quadtree|GetQuadTreeData_Internal)) (LocalVariables|GetCurrentNodeIndex))))
  (bind (_actors_31_dd7b465a42cf64ce22945992a180dd0d_2 _childdatacount_24_488e28b5486c9f3f0d44a2a50447b62f) (Utilities|Struct|BreakLandmassQuadtree (Utilities|Array|Get(aref) (Utilities|Struct|BreakLandmassQuadtreeData (Variables|Quadtree|GetQuadTreeData_Internal)) (LocalVariables|GetCurrentNodeIndex))))
  (bind _quadtreedata_internal (Variables|Quadtree|GetQuadTreeData_Internal))
  (bind _quadtree_13_e8ca8c804ef0247f756fe3b8e015cfe7 (Utilities|Struct|BreakLandmassQuadtreeData _quadtreedata_internal))
  (Utilities|Array|Clear (LocalVariables|GetRemovedFromNodes))
  (LocalVariables|SetNodesToCheck (Utilities|Array|MakeArray 0))
  (for _ (range (+ (Variables|Quadtree|GetBrushTreeDepth) 1))
    (Utilities|Array|Clear (LocalVariables|GetNodesToCheck_Next))
    (for _array_element (LocalVariables|GetNodesToCheck)
      (bind _output_get (LocalVariables|SetCurrentNodeIndex _array_element))
      (Utilities|FlowControl|Sequence
        (:then_0
          (bind _was_removed (Quadtree|RemoveActorfromNode (Variables|Debug|GetBrush%20Requesting%20Update) (LocalVariables|GetCurrentNodeIndex)))
          (if _was_removed
            (Utilities|Array|Add (LocalVariables|GetRemovedFromNodes) (LocalVariables|GetCurrentNodeIndex))))
        (:then_1
          (if (> _childdatacount_24_488e28b5486c9f3f0d44a2a50447b62f 0)
            (Utilities|Array|AppendArray (LocalVariables|GetNodesToCheck_Next) (Utilities|Array|MakeArray _index_x0y0_7_da9d91e24f764d7dbd4d3f9fb8c3c746 _index_x1y0_8_6dad9ca94fe643e04cc00c80d2755731 _index_x0y1_10_e78cc5ea43dcd1ef149c1b8b4baf36b4 _index_x1y1_12_04d22aeb4fa65d7b52d777a53440e336))))))
    (LocalVariables|SetNodesToCheck (LocalVariables|GetNodesToCheck_Next)))
  (return (LocalVariables|GetRemovedFromNodes)))

(fn RenderErosionLayer
  (params (In%20Parameters :type Landscape%20Brush%20Parameters%20Structure) (Extents :type Vector%204%20Structure) (BrushManager :type Landmass%20Brush%20Manager%20Object%20Reference))
  (return false))

(fn RenderLayer
  (params (InParameters :type Landscape%20Brush%20Parameters%20Structure%20%28by%20ref%29))
  (bind _currentrenderparams (Variables|Debug|GetCurrentRenderParams))
  (bind (_renderareaworldtransform _renderareasize _combinedresult _layertype) (Utilities|Struct|BreakLandscapeBrushParameters _currentrenderparams))
  (bind _output_get (Variables|Debug|SetCurrentRenderParams InParameters))
  (if (Variables|Landmass|GetEnable%20Fast%20Preview%20Mode)
    (CallFunction|EnablePreviewLandscape)
    (else
      (Variables|Debug|SetHeightmapPass (Utilities|Enum|Equal(Enum) _layertype "Heightmap"))
      (Variables|RenderTargets|SetCombinedResultRT _combinedresult)
      (if (Variables|Debug|GetHeightmapPass)
        (Variables|RenderTargets|SetInputLandscapeHeightRT _combinedresult))
      (Rendering|Material|SetTextureParameterValue (Variables|Debug|GetCopyRT_MID) "RT" (Variables|RenderTargets|GetCombinedResultRT))
      (Rendering|DrawMaterialtoRenderTarget (select (Variables|Debug|GetHeightmapPass) (Variables|RenderTargets|GetHeightmap%20Render%20RT) (Variables|RenderTargets|GetWeightmap%20RT)) (Variables|Debug|GetCopyRT_MID))
      (Landmass|UpdateBrushesInternal)
      (CallFunction|CopyWMLayersifNeeded) ; Optionally copy Weightmap layers to individual RTs\nOnly needed in special cases such as re-using this data externally
      (Default|CallLandmassManagerRenderComplete)
      (return (Variables|RenderTargets|GetCombinedResultRT)))))

(fn RepositionSceneCapture
  (bind (_x _y) (Math|Vector2D|BreakVector2D (Variables|Debug|LandscapeInformation|GetRT%20World%20Size)))
  (Variables|/Script/Engine.SceneCaptureComponent2D|SetOrthoWidth (Math|Float|Max(Float) _x _y) (Variables|Default|GetSceneCaptureComponent2D))
  (Transformation|SetWorldLocation (Variables|Default|GetSceneCaptureComponent2D) (Math|Vector|MakeVector (.x (+ (.location (Variables|Debug|LandscapeInformation|GetLS%20Transform)) (Math|Transform|TransformLocation (Math|Transform|MakeTransform "0, 0, 0" (.rotation (Variables|Debug|LandscapeInformation|GetLS%20Transform))) (* (- (Math|Conversions|ToVector(Vector2D) (Variables|Debug|LandscapeInformation|GetRT%20World%20Size) 0.512) (.scale (Variables|Debug|LandscapeInformation|GetLS%20Transform))) 0.5)))) (.y (+ (.location (Variables|Debug|LandscapeInformation|GetLS%20Transform)) (Math|Transform|TransformLocation (Math|Transform|MakeTransform "0, 0, 0" (.rotation (Variables|Debug|LandscapeInformation|GetLS%20Transform))) (* (- (Math|Conversions|ToVector(Vector2D) (Variables|Debug|LandscapeInformation|GetRT%20World%20Size) 0.512) (.scale (Variables|Debug|LandscapeInformation|GetLS%20Transform))) 0.5)))) 5e+04) false false))

(fn RequestUpdateFromBrush
  (params (BrushRequestingUpdate :type Landmass%20Actor%20Object%20Reference))
  (bind _output_get (Variables|Debug|SetBrush%20Requesting%20Update BrushRequestingUpdate))
  (Variables|/Script/LandmassEditor.LandmassActor|SetBrushManager self _output_get)
  (if (Variables|Default|GetInitialized)
    (CallFunction|AllocateBrushWeightmaps)
    (if (Variables|Landmass|GetEnable%20Fast%20Preview%20Mode)
      (Quadtree|AddBrushToTree) ; For Preview Mesh Mode, Add Brush to Tree to setup for drawing only brushes in modified Extents
      (Utilities|FlowControl|Sequence
        (:then_0
          ; Prevent Multiple Draws
          (if (or (> (- (Utilities|Time|GetRealTimeSeconds) (Variables|Debug|GetLast%20Render%20Time)) 0.05) (== (Variables|Debug|GetLast%20Updated%20Brush) (Variables|Debug|GetBrush%20Requesting%20Update)))
            (if (Variables|Quadtree|GetNextFrameFullRedraw)
              (Landmass|UpdateBrushesInternal)
              (Variables|Quadtree|SetNextFrameFullRedraw false)
              (else
                (Variables|Quadtree|SetNextFrameFullRedraw false)
                (Variables|Debug|SetLast%20Render%20Time (Utilities|Time|GetRealTimeSeconds))
                (Variables|Debug|SetLast%20Updated%20Brush (Variables|Debug|GetBrush%20Requesting%20Update))))
            (else
              (Variables|Quadtree|SetNextFrameFullRedraw true)))
          (Default|UpdateSpecifiedBrushesForLayer (Variables|Quadtree|GetBrushesRequiringUpdate) true))
        (:then_1
          ;[1 1 1] Update Brushes: Weightmaps
          (for _array_element (Variables|Settings|GetAffectedWeightmapLayers)
            (Default|UpdateSpecifiedBrushesForLayer (Variables|Quadtree|GetBrushesRequiringUpdate) false _array_element))))
      (else
        (Landmass|AddBrushtoArray (Variables|Debug|GetBrush%20Requesting%20Update))
        ;[1 0.5825 0.06679] Set Extents for Whole Landscape
        (Variables|Quadtree|SetModifiedNodeExtents (Default|MakeExtents (Math|Conversions|ToVector2D(Vector) (.location (Variables|Debug|LandscapeInformation|GetLS%20Transform))) (+ (Math|Conversions|ToVector2D(Vector) (.location (Variables|Debug|LandscapeInformation|GetLS%20Transform))) (Math|Conversions|ToVector2D(Vector) (Variables|Debug|LandscapeInformation|GetWorld%20Size)))))
        ;[0.11 0.5168 1] Update all Brushes
        (Landscape|RequestLandscapeUpdate true)))
    (else
      (Development|PrintString "Brush Requested Render:  Manager was NOT Initiliazed!!!" false)
      (Development|PrintString (Utilities|String|Append "Render Requested by:  " (Utilities|GetDisplayName (Variables|Debug|GetBrush%20Requesting%20Update))) false))))

(fn SetLandscapeInfo
  (params (Landscape%20Information :type Landmass%20Landscape%20Info%20Structure))
  (bind _output_get (Variables|Landmass|SetLandscapeInformation Landscape%20Information))
  (Variables|Debug|LandscapeInformation|SetLS%20Transform (Utilities|Struct|BreakLandmassLandscapeInfo _output_get))
  (Variables|Debug|LandscapeInformation|SetLandscape%20Quads (.LandscapeQuads (Utilities|Struct|BreakLandmassLandscapeInfo _output_get)))
  (Variables|Debug|LandscapeInformation|SetLandscape%20RT%20Res (.RenderTargetResolution (Utilities|Struct|BreakLandmassLandscapeInfo _output_get)))
  (CallFunction|UpdateTransform)
  (Quadtree|PopulateTree))

(fn SetMPCParams
  (bind _landscape_rt_res (Variables|Debug|LandscapeInformation|GetLandscape%20RT%20Res))
  (bind (_x _y) (Utilities|Struct|BreakIntPoint _landscape_rt_res))
  (bind _landscape_quads (Variables|Debug|LandscapeInformation|GetLandscape%20Quads))
  (bind (_x_2 _y_2) (Utilities|Struct|BreakIntPoint _landscape_quads))
  (bind (_x_3 _y_3) (Math|Vector2D|BreakVector2D (Variables|Debug|LandscapeInformation|GetRT%20World%20Size)))
  (bind (_x_4 _y_4) (Math|Vector2D|BreakVector2D (Variables|Debug|LandscapeInformation|GetRT%20World%20Center)))
  (bind _world_size (Variables|Debug|LandscapeInformation|GetWorld%20Size))
  (bind _ls_transform (Variables|Debug|LandscapeInformation|GetLS%20Transform))
  (bind _ls_transform_2 (Variables|Debug|LandscapeInformation|GetLS%20Transform))
  (Rendering|Material|SetScalarParameterValue "/Landmass/Landscape/BlueprintBrushes/MPC/MPC_Landscape.MPC_Landscape" "RTResX" (Math|Conversions|ToFloat(Integer) _x) :!class /Script/Engine.KismetMaterialLibrary)
  (Rendering|Material|SetScalarParameterValue "/Landmass/Landscape/BlueprintBrushes/MPC/MPC_Landscape.MPC_Landscape" "RTResY" (Math|Conversions|ToFloat(Integer) _y) :!class /Script/Engine.KismetMaterialLibrary)
  (Rendering|Material|SetScalarParameterValue "/Landmass/Landscape/BlueprintBrushes/MPC/MPC_Landscape.MPC_Landscape" "LSQuadsX" (Math|Conversions|ToFloat(Integer) _x_2) :!class /Script/Engine.KismetMaterialLibrary)
  (Rendering|Material|SetScalarParameterValue "/Landmass/Landscape/BlueprintBrushes/MPC/MPC_Landscape.MPC_Landscape" "LSQuadsY" (Math|Conversions|ToFloat(Integer) _y_2) :!class /Script/Engine.KismetMaterialLibrary)
  (Rendering|Material|SetScalarParameterValue "/Landmass/Landscape/BlueprintBrushes/MPC/MPC_Landscape.MPC_Landscape" "WorldSizeX" (.x _world_size) :!class /Script/Engine.KismetMaterialLibrary)
  (Rendering|Material|SetScalarParameterValue "/Landmass/Landscape/BlueprintBrushes/MPC/MPC_Landscape.MPC_Landscape" "WorldSizeY" (.y _world_size) :!class /Script/Engine.KismetMaterialLibrary)
  (Rendering|Material|SetScalarParameterValue "/Landmass/Landscape/BlueprintBrushes/MPC/MPC_Landscape.MPC_Landscape" "LandscapeZLocation" (.z (.location _ls_transform)) :!class /Script/Engine.KismetMaterialLibrary)
  (Rendering|Material|SetScalarParameterValue "/Landmass/Landscape/BlueprintBrushes/MPC/MPC_Landscape.MPC_Landscape" "LandscapeZScale" (/ (.z (.scale _ls_transform)) 128.0) :!class /Script/Engine.KismetMaterialLibrary)
  (Rendering|Material|SetVectorParameterValue "/Landmass/Landscape/BlueprintBrushes/MPC/MPC_Landscape.MPC_Landscape" "LandscapeLocation" (Math|Conversions|ToLinearColor(Vector) (.location _ls_transform_2)) :!class /Script/Engine.KismetMaterialLibrary)
  (Rendering|Material|SetVectorParameterValue "/Landmass/Landscape/BlueprintBrushes/MPC/MPC_Landscape.MPC_Landscape" "LandscapeScale" (Math|Conversions|ToLinearColor(Vector) (.scale _ls_transform_2)) :!class /Script/Engine.KismetMaterialLibrary)
  (Rendering|Material|SetVectorParameterValue "/Landmass/Landscape/BlueprintBrushes/MPC/MPC_Landscape.MPC_Landscape" "RTWorldLocation" (Math|Color|MakeColor _x_4 _y_4) :!class /Script/Engine.KismetMaterialLibrary)
  (Rendering|Material|SetVectorParameterValue "/Landmass/Landscape/BlueprintBrushes/MPC/MPC_Landscape.MPC_Landscape" "RTWorldSize" (Math|Color|MakeColor _x_3 _y_3) :!class /Script/Engine.KismetMaterialLibrary)
  (Rendering|Material|SetScalarParameterValue "/Landmass/Landscape/BlueprintBrushes/MPC/MPC_Landscape.MPC_Landscape" "ComponentRes" 256.0 :!class /Script/Engine.KismetMaterialLibrary))

(fn TogglePreviewMode
  (params (bEnablePreviewMode :type Boolean))
  (bind _preview_landscape_mesh (Variables|Debug|GetPreview%20Landscape%20Mesh))
  ;[1 1 1] Handle Toggling Mesh Preview Mode
  (bind _output_get (Variables|Landmass|SetEnable%20Fast%20Preview%20Mode bEnablePreviewMode))
  (if _output_get
    ;[1 1 1] Enable Mesh Preview: Cache Brush states and disable them affecting landscape
    (CallFunction|AllocateRTs)
    (Variables|Debug|SetFastPreviewMode_Previous true)
    (Variables|Settings|SetAffectHeightmap false)
    (Variables|Settings|SetAffectWeightmap false)
    (Variables|Settings|SetAffectVisibilityLayer false)
    (CallFunction|EnablePreviewLandscape)
    (CallFunction|PreviewModeTimer)
    (else
      ;[1 1 1] Disable Mesh Preview: Restore Brush states
      (Rendering|SetVisibility (Variables|/Landmass/Landscape/BlueprintBrushes/Previewing/Landmass_PreviewLandscape.Landmass_PreviewLandscape_C|GetLandscape%20ISMC (Variables|Debug|GetPreview%20Landscape%20Mesh)) false false)
      (Variables|Debug|SetFastPreviewMode_Previous false)
      (Variables|Settings|SetAffectHeightmap true)
      (Variables|Settings|SetAffectWeightmap true)
      (Variables|Settings|SetAffectVisibilityLayer true)
      (Development|ExecuteConsoleCommand "showflag.landscape 2")
      (Utilities|IsValid _preview_landscape_mesh
        (:Is%20Valid
          (Actor|DestroyActor _preview_landscape_mesh))))))

(fn UpdateBrushes
  (if (Variables|Landmass|GetEnable%20Fast%20Preview%20Mode)
    (Landmass|UpdateBrushesInternal)
    (else
      (Landscape|RequestLandscapeUpdate true) ; This allows UpdateBrushes_Internal to be called from the RenderLayer function
    )))

(fn UpdateBrushesInternal
  (bind _currentrenderparams (Variables|Debug|GetCurrentRenderParams))
  (bind (_renderareaworldtransform _renderareasize _weightmaplayername) (Utilities|Struct|BreakLandscapeBrushParameters _currentrenderparams))
  ;[0.11 0.5168 1] Get All brushes
  (Landmass|GatherBrushes)
  ;[1 0.5825 0.06679] Set Extents for Whole Landscape
  (Utilities|GetWholeLandscapeExtents)
  (Default|AllocateAllBrushWeightmaps)
  (if (Variables|Landmass|GetEnable%20Fast%20Preview%20Mode)
    (Utilities|FlowControl|Sequence
      (:then_0
        ;[0.11 0.5168 1] Update all Brushes: Height
        (Default|UpdateSpecifiedBrushesForLayer (Variables|Landmass|GetLandmassBrushes) true))
      (:then_1
        ;[1 1 1] Update Brushes: Weightmaps
        (for _array_element (Variables|Settings|GetAffectedWeightmapLayers)
          (Default|UpdateSpecifiedBrushesForLayer (Variables|Landmass|GetLandmassBrushes) false _array_element))))
    (elif (Variables|Debug|GetHeightmapPass)
      (for _array_element_2 (Variables|Landmass|GetLandmassBrushes)
        (bind _output_get (Variables|Debug|SetBrush%20Requesting%20Update _array_element_2))
        (Utilities|IsValid _output_get
          (:Is%20Valid
            ;[0.09423 0.5411 1] Render Heightmap
            (if (Variables|/Script/LandmassEditor.LandmassActor|GetAffectsHeightmap (Variables|Debug|GetBrush%20Requesting%20Update))
              (Default|RenderLayerNative (Variables|Debug|GetBrush%20Requesting%20Update) (Utilities|Struct|MakeLandscapeBrushParameters "0.000000,0.000000,0.000000|0.000000,0.000000,-0.000000|1.000000,1.000000,1.000000" "(X=2147483647,Y=2147483647)" (Variables|RenderTargets|GetCombinedResultRT) "Heightmap"))))))
      (else
        (for _array_element_3 (Variables|Landmass|GetLandmassBrushes)
          (bind _output_get_2 (Variables|Debug|SetBrush%20Requesting%20Update _array_element_3))
          (Utilities|IsValid _output_get_2
            (:Is%20Valid
              ;[0.3003 0.04666 0.4479] Render Heightmap
              (if (and (Utilities|Array|ContainsItem (Variables|/Script/LandmassEditor.LandmassActor|GetWeightmapLayers (Variables|Debug|GetBrush%20Requesting%20Update)) _weightmaplayername) (Variables|/Script/LandmassEditor.LandmassActor|GetAffectsWeightmaps (Variables|Debug|GetBrush%20Requesting%20Update)))
                (Default|RenderLayerNative (Variables|Debug|GetBrush%20Requesting%20Update) (Utilities|Struct|MakeLandscapeBrushParameters "0.000000,0.000000,0.000000|0.000000,0.000000,-0.000000|1.000000,1.000000,1.000000" "(X=2147483647,Y=2147483647)" (Variables|RenderTargets|GetCombinedResultRT) "Weightmap" _weightmaplayername))))))))))

(fn UpdateSpecifiedBrushesforLayer
  (params (Specified%20Brushes :type Array%20of%20Landmass%20Actor%20Object%20References) (Heightmap :type Boolean) (Weightmap%20Name :type Name))
  (bind _weightmap_name (Variables|GetWeightmap%20Name))
  (bind _weightmap_rts (Variables|RenderTargets|GetWeightmap%20RTs))
  (bind (_value _returnvalue) (Utilities|Map|Find _weightmap_rts _weightmap_name))
  (bind (_x _y _z _w) (Math|Vector4|BreakVector4 (Variables|Quadtree|GetModifiedNodeExtents)))
  (Variables|Debug|SetHeight%20Idx 0)
  ; Extents Print Debug
  (if (Variables|Quadtree|GetDebug%20Text)
    (Development|PrintString (Utilities|String|Append "Current Edited Extents" (Utilities|String|ToString(LinearColor) (Utilities|Struct|MakeLinearColor _x _y _z _w)))))
  (if Heightmap
    (if (Variables|Landmass|GetEnable%20Fast%20Preview%20Mode)
      (Utilities|BrushCopyRenderTargetExtents (Variables|RenderTargets|GetInputLandscapeHeightRT) (Variables|RenderTargets|GetHeightmap%20Render%20RT) self (Variables|Quadtree|GetModifiedNodeExtents) (Variables|Landmass|GetLandscapeInformation)) ; First Copy the area to read to a 2nd RT
      (Utilities|BrushCopyRenderTargetExtents (Variables|RenderTargets|GetInputLandscapeHeightRT) (Variables|RenderTargets|GetScratch%20Heightmap%20RT) self (Variables|Quadtree|GetModifiedNodeExtents) (Variables|Landmass|GetLandscapeInformation)) ; First Copy the area to read to a 2nd RT
      (else
        ;[1 1 1] Clear the Render Targets in the Modified Region
        (Utilities|BrushDrawToRenderTargetExtents (Variables|RenderTargets|GetHeightmap%20Render%20RT) "/Landmass/Landscape/BlueprintBrushes/Materials/Internal/ClearRTHeight.ClearRTHeight" (Variables|Quadtree|GetModifiedNodeExtents) (Variables|Landmass|GetLandscapeInformation))
        (Utilities|BrushDrawToRenderTargetExtents (Variables|RenderTargets|GetScratch%20Heightmap%20RT) "/Landmass/Landscape/BlueprintBrushes/Materials/Internal/ClearRTHeight.ClearRTHeight" (Variables|Quadtree|GetModifiedNodeExtents) (Variables|Landmass|GetLandscapeInformation))))
    (for _array_element (Variables|GetSpecified%20Brushes)
      (bind _output_get (Variables|Debug|SetBrush%20Requesting%20Update _array_element))
      (Utilities|IsValid _output_get
        (:Is%20Valid
          ;[0.09423 0.5411 1] Render Heightmap
          (if (Variables|/Script/LandmassEditor.LandmassActor|GetAffectsHeightmap (Variables|Debug|GetBrush%20Requesting%20Update))
            (Default|RenderLayerNative (Variables|Debug|GetBrush%20Requesting%20Update) (Utilities|Struct|MakeLandscapeBrushParameters "0.000000,0.000000,0.000000|0.000000,0.000000,-0.000000|1.000000,1.000000,1.000000" "(X=2147483647,Y=2147483647)" (Variables|RenderTargets|GetHeightmap%20Render%20RT) "Heightmap"))))))
    (Class|LandmassPreviewLandscape|UpdateMIDHeight (Variables|Debug|GetPreview%20Landscape%20Mesh) (Variables|RenderTargets|GetHeightmap%20Render%20RT))
    (elif _returnvalue
      (Utilities|IsValid _value
        (:Is%20Valid
          (LocalVariables|SetTarget%20Weightmap _value)
          ;[1 1 1] Clear the Render Targets in the Modified Region
          (Utilities|BrushDrawToRenderTargetExtents (LocalVariables|GetTarget%20Weightmap) "/Landmass/Landscape/BlueprintBrushes/Materials/Internal/ClearRTWeight.ClearRTWeight" (Variables|Quadtree|GetModifiedNodeExtents) (Variables|Landmass|GetLandscapeInformation))
          (Utilities|BrushDrawToRenderTargetExtents (Variables|RenderTargets|GetWeightmap%20RT) "/Landmass/Landscape/BlueprintBrushes/Materials/Internal/ClearRTWeight.ClearRTWeight" (Variables|Quadtree|GetModifiedNodeExtents) (Variables|Landmass|GetLandscapeInformation))
          (for _array_element_2 (Variables|GetSpecified%20Brushes)
            (bind _output_get_2 (Variables|Debug|SetBrush%20Requesting%20Update _array_element_2))
            (Utilities|IsValid _output_get_2
              (:Is%20Valid
                ;[0.3003 0.04666 0.4479] Render Weightmap
                (if (and (Utilities|Array|ContainsItem (Variables|/Script/LandmassEditor.LandmassActor|GetWeightmapLayers (Variables|Debug|GetBrush%20Requesting%20Update)) (Variables|GetWeightmap%20Name)) (Variables|/Script/LandmassEditor.LandmassActor|GetAffectsWeightmaps (Variables|Debug|GetBrush%20Requesting%20Update)))
                  (Default|RenderLayerNative (Variables|Debug|GetBrush%20Requesting%20Update) (Utilities|Struct|MakeLandscapeBrushParameters "0.000000,0.000000,0.000000|0.000000,0.000000,-0.000000|1.000000,1.000000,1.000000" "(X=2147483647,Y=2147483647)" (LocalVariables|GetTarget%20Weightmap) "Weightmap" (Variables|GetWeightmap%20Name)))))))
          (Class|LandmassPreviewLandscape|UpdateMIDWeight (Variables|Debug|GetPreview%20Landscape%20Mesh) (Variables|GetWeightmap%20Name) (LocalVariables|GetTarget%20Weightmap)))))))

(fn UpdateTransform
  (bind _landscape_quads (Variables|Debug|LandscapeInformation|GetLandscape%20Quads))
  (bind (_x _y) (Utilities|Struct|BreakIntPoint _landscape_quads))
  (bind (_x_2 _y_2) (Utilities|Struct|BreakIntPoint (Variables|Debug|LandscapeInformation|GetLandscape%20Quads)))
  (Variables|Debug|LandscapeInformation|SetWorld%20Size (Math|Vector|MakeVector (* (Math|Conversions|ToFloat(Integer) _x_2) (.x (.scale (Variables|Debug|LandscapeInformation|GetLS%20Transform)))) (* (Math|Conversions|ToFloat(Integer) _y_2) (.y (.scale (Variables|Debug|LandscapeInformation|GetLS%20Transform)))) 0.512))
  (Variables|Debug|LandscapeInformation|SetRT%20World%20Size (* (Math|Conversions|ToVector2D(IntPoint) (Variables|Debug|LandscapeInformation|GetLandscape%20RT%20Res)) (Math|Conversions|ToVector2D(Vector) (.scale (Variables|Debug|LandscapeInformation|GetLS%20Transform)))))
  (Variables|Debug|LandscapeInformation|SetRT%20World%20Center (+ (Math|Conversions|ToVector2D(Vector) (.location (Variables|Debug|LandscapeInformation|GetLS%20Transform))) (* (* (Math|Conversions|ToVector2D(Vector) (.scale (Variables|Debug|LandscapeInformation|GetLS%20Transform))) (Variables|Debug|LandscapeInformation|GetLandscape%20RT%20Res)) 0.5)))
  (CallFunction|AllocateRTs)
  (Variables|/Script/Engine.SceneCaptureComponent2D|SetTextureTarget (Variables|RenderTargets|GetDepth%20and%20Shape%20RT%20A) (Variables|Default|GetSceneCaptureComponent2D))
  (CallFunction|SetMPCParams)
  (Default|RepositionSceneCapture)
  (Variables|Debug|SetSections%20X (/ _x 256))
  (Variables|Debug|SetSections%20Y (/ _y 256)))

(fn ConstructionScript
  (CallFunction|SetLandscapeInfo (Variables|Landmass|GetLandscapeInformation))
  (CallFunction|AllocateRTs)
  (CallFunction|SetMPCParams)
  (Quadtree|PopulateTree)
  (Variables|Default|SetInitialized true)
  (if (== (Variables|Landmass|GetEnable%20Fast%20Preview%20Mode) (Variables|Debug|GetFastPreviewMode_Previous))
    (if (Variables|Landmass|GetEnable%20Fast%20Preview%20Mode)
      (Landmass|UpdateBrushes)
      (if (== (Variables|Landmass|GetPreview%20Mode%20use%20Landscape%20MID) (Variables|Debug|GetFastPreview_UseMID_Previous))
        _
        (else
          _))))
  (Default|TogglePreviewMode (Variables|Landmass|GetEnable%20Fast%20Preview%20Mode)))
