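# Landmass material function for brushes that outputs fractal 3D gradient noise with adjustable octaves, persistence and optional turbulence.
# A Custom HLSL node sums octaves, doubling position and repeat size each step and wrapping lattice seeds in XY so the noise tiles.
# Cells near the low X/Y tile edges take a shared Edge Seed, in a band that widens per octave by the threshold.
# The Tiling? input is unused because tiling is hard-coded on.
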MaterialFunction /Landmass/Landscape/BlueprintBrushes/MF/LandmassBrush_NoiseTiling
graph {
 properties {
  description="This function allows more advanced sampling of Noise including control of Persistence and Octaves as well as blending to a common Edge Seed if desired. "
  libraryCategoriesText [Landmass]
 }
 nodes [
  MaterialExpressionNamedRerouteDeclaration_0 = NamedRerouteDeclaration(None=FunctionInput() {bUsePreviewValueAsDefault=true, inputName=Position}) {name=Position, nodeColor=<rgba 0 0 0 1>}
  n1 = FunctionInput() {bUsePreviewValueAsDefault=true, inputName=Octaves, inputType=FunctionInput_Scalar, sortPriority=1, previewValue=<10 0 0 1>}
  MaterialExpressionNamedRerouteDeclaration_1 = NamedRerouteDeclaration(None=@n1) {name=Octaves, nodeColor=<rgba 0 0 0 1>}
  n2 = FunctionInput() {bUsePreviewValueAsDefault=true, inputName=Repeat Size, inputType=FunctionInput_Scalar, sortPriority=2, previewValue=<5 0 0 1>}
  MaterialExpressionNamedRerouteDeclaration_2 = NamedRerouteDeclaration(None=@n2) {name=Repeat Size, nodeColor=<rgba 0 0 0 1>}
  n3 = FunctionInput() {bUsePreviewValueAsDefault=true, inputName=Persistence, inputType=FunctionInput_Scalar, sortPriority=3, previewValue=<0.4 0 0 1>}
  MaterialExpressionNamedRerouteDeclaration_3 = NamedRerouteDeclaration(None=@n3) {name=Persistence, nodeColor=<rgba 0 0 0 1>}
  MaterialExpressionNamedRerouteDeclaration_4 = NamedRerouteDeclaration(None=FunctionInput(Preview=Constant()) {bUsePreviewValueAsDefault=true, inputName=Turbulence, inputType=FunctionInput_Scalar, sortPriority=4}) {name=Turbulence, nodeColor=<rgba 0 0 0 1>}
  MaterialExpressionNamedRerouteDeclaration_5 = NamedRerouteDeclaration(None=FunctionInput() {bUsePreviewValueAsDefault=true, inputName=Edge Seed, inputType=FunctionInput_Scalar, sortPriority=5}) {name=Edge Seed, nodeColor=<rgba 0 0 0 1>}
  n4 = FunctionInput(Preview=Constant() {r=1}) {bUsePreviewValueAsDefault=true, inputName=Tiling?, inputType=FunctionInput_Scalar, sortPriority=6, previewValue=<1 0 0 1>}
  MaterialExpressionNamedRerouteDeclaration_6 = NamedRerouteDeclaration(None=@n4) {name=Tiling?, nodeColor=<rgba 0 0 0 1>}
  n5 = FunctionInput() {bUsePreviewValueAsDefault=true, inputName=Edge Seed Threshold, inputType=FunctionInput_Scalar, sortPriority=7, previewValue=<1 0 0 1>}
  MaterialExpressionNamedRerouteDeclaration_7 = NamedRerouteDeclaration(None=@n5) {name=Edge Seed Threshold, nodeColor=<rgba 0 0 0 1>}
  n6 = Custom!(
   Position=@MaterialExpressionNamedRerouteDeclaration_0
   Octaves=@MaterialExpressionNamedRerouteDeclaration_1
   RepeatSize=@MaterialExpressionNamedRerouteDeclaration_2
   Persistence=@MaterialExpressionNamedRerouteDeclaration_3
   bTurbulence=@MaterialExpressionNamedRerouteDeclaration_4
   EdgeSeed=@MaterialExpressionNamedRerouteDeclaration_5
   Tiling=@MaterialExpressionNamedRerouteDeclaration_6
   EdgeThreshold=@MaterialExpressionNamedRerouteDeclaration_7
  ) {
   code="//in v3 Position\r\n//in scalar RepeatSize\r\n//in scalar Persistance\r\n//in scalar Octaves\r\n//in constant bTurbulence\r\n//in constant EdgeSeed\r\n\r\n\r\nstruct NoiseFunctions\r\n{\r\n\r\n\tfloat thisEdgeSeed;\r\n\tfloat iteration;\r\n\tint EdgeSeedThreshold;\r\n\r\n\tfloat3 NoiseTileWrap(float3 v,  bool bTiling, float RepeatSize)\r\n\t{\t\r\n\t\tif (floor(v.x) <0 || floor(v.y) <0)\r\n\t\t{\r\n         //v.xy= 0;\r\n\t\t\t//v.z = thisEdgeSeed;\r\n\t\t}\r\n\t\t\r\n\t\tv.xy = bTiling ? (frac(v.xy / RepeatSize) * RepeatSize) : v.xy;\r\n\t\t\r\n\t\tif (floor(v.x) <= iteration * EdgeSeedThreshold || floor(v.y) <= iteration * EdgeSeedThreshold)\r\n\t\t{\r\n         //v.xy= 0;\r\n\t\t\tv.z = thisEdgeSeed;\r\n\t\t}\r\n\t\t\r\n\t\treturn v;\r\n\t\t\r\n\t}\r\n\r\n\tfloat4 PerlinRamp(float4 t)\r\n\t{\r\n\t\treturn t * t * t * (t * (t * 6 - 15) + 10); \r\n\t}\r\n\t\r\n\t// Analytical derivative of the PerlinRamp polynomial\r\n\t// only needed by Perlin functions in this file\r\n\t// scalar(per component): 2 add, 5 mul\r\n\tfloat4 PerlinRampDerivative(float4 t)\r\n\t{\r\n\t\treturn t * t * (t * (t * 30 - 60) + 30);\r\n\t}\r\n\r\n\tfloat4 MGradient(int seed, float3 offset)\r\n\t{\r\n\t\tuint rand = Rand3DPCG16(int3(seed,0,0)).x;\r\n\t\tfloat3 direction = float3(rand.xxx && MGradientMask) * MGradientScale - 1;\r\n\t\treturn float4(direction, dot(direction, offset));\r\n\t}\r\n\r\n\tfloat3 NoiseSeeds(float3 v, bool bTiling, float RepeatSize, out float seed000, out float seed001, out float seed010, out float seed011, out float seed100, out float seed101, out float seed110, out float seed111)\r\n\t{\r\n\t\tfloat3 fv = frac(v);\r\n\t\tfloat3 iv = floor(v);\r\n\t\r\n\t\tconst float3 primes = float3(19, 47, 101);\r\n\t\r\n\t\tif (bTiling)\r\n\t\t{\t// can't algebraically combine with primes\r\n\t\t\tseed000 = dot(primes, NoiseTileWrap(iv, true, RepeatSize));\r\n\t\t\tseed100 = dot(primes, NoiseTileWrap(iv + float3(1, 0, 0), true, RepeatSize));\r\n\t\t\tseed010 = dot(primes, NoiseTileWrap(iv + float3(0, 1, 0), true, RepeatSize));\r\n\t\t\tseed110 = dot(primes, NoiseTileWrap(iv + float3(1, 1, 0), true, RepeatSize));\r\n\t\t\tseed001 = dot(primes, NoiseTileWrap(iv + float3(0, 0, 1), true, RepeatSize));\r\n\t\t\tseed101 = dot(primes, NoiseTileWrap(iv + float3(1, 0, 1), true, RepeatSize));\r\n\t\t\tseed011 = dot(primes, NoiseTileWrap(iv + float3(0, 1, 1), true, RepeatSize));\r\n\t\t\tseed111 = dot(primes, NoiseTileWrap(iv + float3(1, 1, 1), true, RepeatSize));\r\n\t\t}\r\n\t\telse\r\n\t\t{\t// get to combine offsets with multiplication by primes in this case\r\n\t\t\tseed000 = dot(iv, primes);\r\n\t\t\tseed100 = seed000 + primes.x;\r\n\t\t\tseed010 = seed000 + primes.y;\r\n\t\t\tseed110 = seed100 + primes.y;\r\n\t\t\tseed001 = seed000 + primes.z;\r\n\t\t\tseed101 = seed100 + primes.z;\r\n\t\t\tseed011 = seed010 + primes.z;\r\n\t\t\tseed111 = seed110 + primes.z;\r\n\t\t}\r\n\t\r\n\t\treturn fv;\r\n\t}\r\n\r\n\r\n\tfloat GradientNoise3D(float3 v, bool bTiling, float RepeatSize)\r\n\t{\r\n\t\tfloat seed000, seed001, seed010, seed011, seed100, seed101, seed110, seed111;\r\n\t\tfloat3 fv = NoiseSeeds(v, bTiling, RepeatSize, seed000, seed001, seed010, seed011, seed100, seed101, seed110, seed111);\r\n\t\r\n\t\tfloat rand000 = MGradient(int(seed000), fv - float3(0, 0, 0)).w;\r\n\t\tfloat rand100 = MGradient(int(seed100), fv - float3(1, 0, 0)).w;\r\n\t\tfloat rand010 = MGradient(int(seed010), fv - float3(0, 1, 0)).w;\r\n\t\tfloat rand110 = MGradient(int(seed110), fv - float3(1, 1, 0)).w;\r\n\t\tfloat rand001 = MGradient(int(seed001), fv - float3(0, 0, 1)).w;\r\n\t\tfloat rand101 = MGradient(int(seed101), fv - float3(1, 0, 1)).w;\r\n\t\tfloat rand011 = MGradient(int(seed011), fv - float3(0, 1, 1)).w;\r\n\t\tfloat rand111 = MGradient(int(seed111), fv - float3(1, 1, 1)).w;\r\n\t\r\n\t\tfloat3 Weights = PerlinRamp(float4(fv, 0)).xyz;\r\n\t\r\n\t\tfloat i = lerp(lerp(rand000, rand100, Weights.x), lerp(rand010, rand110, Weights.x), Weights.y);\r\n\t\tfloat j = lerp(lerp(rand001, rand101, Weights.x), lerp(rand011, rand111, Weights.x), Weights.y);\r\n\t\treturn lerp(i, j, Weights.z).x;\r\n\t}\r\n};\r\n\r\nNoiseFunctions NoiseHelper;\r\nNoiseHelper.thisEdgeSeed = EdgeSeed;\r\nNoiseHelper.iteration = 0;\r\nNoiseHelper.EdgeSeedThreshold = EdgeThreshold;\r\n\r\nfloat Out = 0.0f;\r\nfloat OctaveIntensity = 1.0f;\r\n\r\nfloat LevelScale = 2.0;\r\n\r\nfor(uint i = 0; i < Octaves; ++i)\r\n{\r\n\r\n\tif(bTurbulence)\r\n\t{\r\n\t\tOut += abs(NoiseHelper.GradientNoise3D(Position, true, RepeatSize)) * OctaveIntensity;\r\n\t\tNoiseHelper.iteration += 1.0;\r\n\t}\r\n\telse\r\n\t{\r\n\t\tOut += NoiseHelper.GradientNoise3D(Position, true, RepeatSize) * OctaveIntensity;\r\n\t\tNoiseHelper.iteration += 1.0;\r\n\t}\r\n\r\n\tPosition *= LevelScale;\r\n\tRepeatSize *= LevelScale;\r\n\tOctaveIntensity *= Persistence;\r\n}\r\n\r\nif(!bTurbulence)\r\n{\r\n\t// bring -1..1 to 0..1 range\r\n\tOut = Out * 0.5f + 0.5f;\r\n}\r\n\r\n// Out is in 0..1 range\r\nreturn lerp(0, 1, Out);\r\n\t"
   description=Custom Noise
   outputType=CMOT_Float1
   inputs [
    {
     inputName=Position
     input {expression=&self:MaterialExpressionNamedRerouteDeclaration_0}
    }
    {
     inputName=Octaves
     input {expression=&self:MaterialExpressionNamedRerouteDeclaration_1}
    }
    {
     inputName=RepeatSize
     input {expression=&self:MaterialExpressionNamedRerouteDeclaration_2}
    }
    {
     inputName=Persistence
     input {expression=&self:MaterialExpressionNamedRerouteDeclaration_3}
    }
    {
     inputName=bTurbulence
     input {expression=&self:MaterialExpressionNamedRerouteDeclaration_4}
    }
    {
     inputName=EdgeSeed
     input {expression=&self:MaterialExpressionNamedRerouteDeclaration_5}
    }
    {
     inputName=Tiling
     input {expression=&self:MaterialExpressionNamedRerouteDeclaration_6}
    }
    {
     inputName=EdgeThreshold
     input {expression=&self:MaterialExpressionNamedRerouteDeclaration_7}
    }
   ]
  }
  FunctionOutput(None=@n6)
 ]
}
