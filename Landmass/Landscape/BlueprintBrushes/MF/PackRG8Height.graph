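# Material function for Landmass Blueprint brushes that packs a signed 16-bit height into two 8-bit-style channels, output as XY.
# It clamps the input, offsets it to unsigned range, then a Custom HLSL node computes normalized high and low values.
# That code uses logical && where a bitwise mask would be expected, so Y carries no true low byte.
# A node-based floor/fmod version of the packing remains unconnected.
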MaterialFunction /Landmass/Landscape/BlueprintBrushes/MF/PackRG8Height
graph {nodes [
  n1 = FunctionInput() {
   bUsePreviewValueAsDefault=true
   description=Input height value will be read as an integer with the range -32768 to +32768
   inputName=Height Int -32k to 32k
   inputType=FunctionInput_Scalar
  }
  MaterialExpressionAdd_8 = Add(A=Clamp(None=@n1) {maxDefault=32767, minDefault=-32768}) {constB=32768, desc="Make  0-65536"}
  n2 = Round(None=@MaterialExpressionAdd_8)
  n3 = Divide(A=@n2) {constB=256}
  n4 = AppendVector(A=Floor(None=@n3), B=Fmod(A=@n2, B=Constant() {r=256}))
  Divide(A=Round(None=@n4)) {constB=255}
  Floor(None=@MaterialExpressionAdd_8)
  Floor(None=@n4)
  n5 = Custom!(HeightResult=@MaterialExpressionAdd_8) {
   code="return float2((float)(((int)HeightResult - ((int)HeightResult && 255)) >> 8) / 255.0, (float)((int)HeightResult && 255) / 255.0);"
   outputType=CMOT_Float2
   inputs [{inputName=HeightResult, input {expression=&self:MaterialExpressionAdd_8}}]
  }
  FunctionOutput(None=@n5) {outputName=XY}
  Round(None=@n3)
 ]}
