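# Landmass landscape brush material that caches a distance field packed into 8-bit channels, using an unlit alpha-composite setup.
# SampleDistanceField supplies the distance, optionally blended by UseBlur toward a nine-tap box average with texel offsets from render target resolution.
# The distance is scaled by half the larger world size and packed by PackRG8Height, with a second packed value appended.
# Side branches include a MeshDepth height blend and a normalized test output.
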Material /Landmass/Landscape/BlueprintBrushes/Materials/CacheDistanceField_RG8
graph {
 properties {
  bAllowNegativeEmissiveColor=true
  blendMode=BLEND_AlphaComposite
  pixelDepthOffsetMode=PDOM_Legacy
  shadingModel=MSM_Unlit
  usedShadingModels=Unlit
 }
 nodes [
  n1 = MaterialFunctionCall!(UV Offset) {materialFunction=/Landmass/Landscape/BlueprintBrushes/MF/SampleDistanceField}
  n2 = AppendVector(A=@n1.UV Distance, B=@n1.UV Distance)
  n3 = ScalarParameter() {defaultValue=1, parameterName=Bluroffset}
  n4 = Divide(
   A=@n3
   B=AppendVector(
    A=CollectionParameter() {collection=/Landmass/Landscape/BlueprintBrushes/MPC/MPC_Landscape, parameterName=RTResX}
    B=CollectionParameter() {collection=&MPC_Landscape, parameterName=RTResY}
   )
  )
  n5 = Multiply(A=@n4, B=Constant2Vector() {r=1})
  n6 = MaterialFunctionCall(UV Offset=@n5) {materialFunction=&SampleDistanceField}
  n7 = MaterialFunctionCall(UV Offset=Multiply(A=@n5) {constB=-1}) {materialFunction=&SampleDistanceField}
  n8 = Add(A=AppendVector(A=@n6.UV Distance, B=@n6.UV Distance), B=AppendVector(A=@n7.UV Distance, B=@n7.UV Distance))
  n9 = Multiply(A=@n4, B=Constant2Vector() {g=1})
  n10 = MaterialFunctionCall(UV Offset=@n9) {materialFunction=&SampleDistanceField}
  n11 = MaterialFunctionCall(UV Offset=Multiply(A=@n9) {constB=-1}) {materialFunction=&SampleDistanceField}
  n12 = Add(A=AppendVector(A=@n10.UV Distance, B=@n10.UV Distance), B=AppendVector(A=@n11.UV Distance, B=@n11.UV Distance))
  n13 = Divide(
   A=@n3
   B=AppendVector(
    A=CollectionParameter() {collection=&MPC_Landscape, parameterName=RTResX}
    B=CollectionParameter() {collection=&MPC_Landscape, parameterName=RTResY}
   )
  )
  n14 = Multiply(A=@n13, B=Constant2Vector() {g=1, r=1})
  n15 = MaterialFunctionCall(UV Offset=@n14) {materialFunction=&SampleDistanceField}
  n16 = MaterialFunctionCall(UV Offset=Multiply(A=@n14, B=Constant2Vector() {g=1, r=-1}) {constB=-1}) {materialFunction=&SampleDistanceField}
  n17 = Add(A=AppendVector(A=@n15.UV Distance, B=@n15.UV Distance), B=AppendVector(A=@n16.UV Distance, B=@n16.UV Distance))
  n18 = Multiply(A=@n13, B=Constant2Vector() {g=-1, r=-1})
  n19 = MaterialFunctionCall(UV Offset=@n18) {materialFunction=&SampleDistanceField}
  n20 = MaterialFunctionCall(UV Offset=Multiply(A=@n18, B=Constant2Vector() {g=1, r=-1}) {constB=-1}) {materialFunction=&SampleDistanceField}
  n21 = Add(A=AppendVector(A=@n19.UV Distance, B=@n19.UV Distance), B=AppendVector(A=@n20.UV Distance, B=@n20.UV Distance))
  n22 = Add(A=Add(A=Add(A=@n8, B=@n12), B=Add(A=@n17, B=@n21)), B=@n2)
  n23 = MaterialFunctionCall(Float2=LinearInterpolate(A=@n2, B=Divide(A=@n22) {constB=9}, Alpha=ScalarParameter() {parameterName=UseBlur})) {materialFunction=&BreakOutFloat2Components}
  n24 = Multiply(
   A=@n23.R
   B=Max(
    A=CollectionParameter() {collection=&MPC_Landscape, parameterName=WorldSizeX}
    B=CollectionParameter() {collection=&MPC_Landscape, parameterName=WorldSizeY}
   )
  )
  n25 = MaterialFunctionCall(Float2=MaterialFunctionCall!(Height Int -32k to 32k=@n23.G) {materialFunction=/Landmass/Landscape/BlueprintBrushes/MF/PackRG8Height}.XY) {materialFunction=&BreakOutFloat2Components}
  AppendVector(
   A=MaterialFunctionCall(Height Int -32k to 32k=Multiply(A=@n24) {constB=0.5}) {materialFunction=&PackRG8Height}.XY
   B=@n25.R
  )
  n26 = TextureSampleParameter2D() {parameterName=MeshDepth, texture=/Engine/EngineResources/Black_Low.Black}
  LinearInterpolate(
   A=ScalarParameter() {defaultValue=1394, parameterName=Z}
   B=LinearInterpolate(A=@n23.G, B=Subtract(B=@n26.R) {constA=50000}, Alpha=@n1.Mask)
   Alpha=ScalarParameter() {parameterName=UseMeshDepth}
  )
  Multiply(A=@n23.G) {constB=10000, desc="Testing normalized, Counteracting a /10000 in DetectEdges mat"}
  Multiply(A=@n24, B=CollectionParameter() {collection=/Landmass/Landscape/BlueprintBrushes/MPC/MPC_Erosion, parameterName=GridSize}) {constB=10}
  OneMinus(None=@n25.G)
  ScalarParameter() {defaultValue=1394, parameterName=Z}
 ]
}
