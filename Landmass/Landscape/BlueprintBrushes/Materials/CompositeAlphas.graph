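# Landmass landscape brush material that composites one water shape's velocity, height and depth over the previous CombinedVelocityAndHeight results.
# A cached distance field gives the shape mask, dilated where a custom neighborhood-averaging node finds no valid previous samples nearby.
# The mask blends in velocity and ZOffset-adjusted height as RGB, and normalized depth merged with previous alpha.
# Further branches build a secondary dilation mask and maximum water height from offset neighbor samples.
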Material /Landmass/Landscape/BlueprintBrushes/Materials/CompositeAlphas
graph {
 properties {
  bAllowNegativeEmissiveColor=true
  blendMode=BLEND_AlphaComposite
  pixelDepthOffsetMode=PDOM_Legacy
  shadingModel=MSM_Unlit
  usedShadingModels=Unlit
 }
 nodes [
  Constant()
  n1 = TextureSampleParameter2D() {desc=Previous Combined Results, parameterName=CombinedVelocityAndHeight, texture=&Black}
  n2 = MaterialFunctionCall!(Cached DF, UV Offset) {materialFunction=/Landmass/Landscape/BlueprintBrushes/MF/ReadCachedDistanceField}
  n3 = MaterialFunctionCall!(World Z=ScalarParameter() {parameterName=ZOffset}) {materialFunction=/Landmass/Landscape/BlueprintBrushes/MF/LocalizeLandscapeZValue}
  MaterialExpressionTextureObjectParameter_0 = TextureObjectParameter() {parameterName=CombinedVelocityAndHeight, texture=&Black}
  MaterialExpressionTextureCoordinate_0 = TextureCoordinate()
  n4 = CollectionParameter() {collection=/Landmass/Landscape/BlueprintBrushes/MPC/MPC_Landscape, parameterName=RTResX}
  n5 = CollectionParameter() {collection=&MPC_Landscape, parameterName=RTResY}
  MaterialExpressionMultiply_10 = Multiply(A=Divide(B=AppendVector(A=@n4, B=@n5)), B=ScalarParameter() {defaultValue=2, parameterName=Texel Scale})
  MaterialExpressionScalarParameter_8 = ScalarParameter() {defaultValue=3, parameterName=Radius}
  n6 = Custom!(
   Tex=@MaterialExpressionTextureObjectParameter_0
   UV=@MaterialExpressionTextureCoordinate_0
   TexelSize=@MaterialExpressionMultiply_10
   Radius=@MaterialExpressionScalarParameter_8
  ) {
   code="float2 accum = 0;\r\n\r\nfloat2 curUV = UV;\r\nfloat neighbors = 0;\r\n\r\nfor (int i = -Radius; i <= Radius; i++)\r\n{\r\n\r\n\tfor(int j=-Radius; j <= Radius; j++)\r\n\t{\r\n\t\t\r\n\t\tcurUV = UV + float2(TexelSize.x * (float)i, 0) + float2(0, TexelSize.y * (float)j);\r\n\t\tfloat4 cursample = Tex.SampleLevel(TexSampler, curUV, 0);\r\n\t\t\r\n\t\tfloat depthmask = 1 - floor(cursample.a);\r\n\t\t\r\n\t\tif (depthmask == 1.0)\r\n\t\t{\r\n\t\t\taccum += cursample.xy;\r\n\t\t\tneighbors += 1;\r\n\t\t\r\n\t\t}\r\n\t\t\r\n\t}\r\n\t\r\n}\r\n\r\nif(neighbors > 0)\r\n{\r\n\taccum.xy /= neighbors;\r\n}\r\n\r\nreturn float3( accum.xy, saturate(neighbors) );"
   description=Blur Velocities
   inputs [
    {inputName=Tex, input {expression=&self:MaterialExpressionTextureObjectParameter_0}}
    {inputName=UV, input {expression=&self:MaterialExpressionTextureCoordinate_0}}
    {inputName=TexelSize, input {expression=&self:MaterialExpressionMultiply_10}}
    {inputName=Radius, input {expression=&self:MaterialExpressionScalarParameter_8}}
   ]
  }
  n7 = LinearInterpolate(B=ScalarParameter() {defaultValue=1536, parameterName=Shape Dilation}, Alpha=OneMinus(None=ComponentMask(None=@n6) {b=true})) {constB=0}
  n8 = Saturate(None=Ceil(None=Add(A=@n2.World Distance, B=@n7))) {desc=Mask}
  LinearInterpolate(A=@n1.RGB, B=AppendVector(A=@n2.Velocity, B=Subtract(A=@n2.Water Height, B=@n3.Scale Only)), Alpha=@n8)
  n9 = CollectionParameter() {collection=&MPC_Landscape, parameterName=RTResX}
  n10 = CollectionParameter() {collection=&MPC_Landscape, parameterName=RTResY}
  n11 = Multiply(A=ScalarParameter() {defaultValue=2, parameterName=Dilation Spread}, B=Divide(B=AppendVector(A=@n9, B=@n10)))
  n12 = Multiply(A=Multiply(A=@n11, B=Constant2Vector() {r=1})) {constB=-1}
  n13 = MaterialFunctionCall(UV Offset=@n12) {materialFunction=&ReadCachedDistanceField}
  n14 = Saturate(None=Ceil(None=Multiply(A=@n13.World Distance))) {desc=Mask}
  n15 = MaterialFunctionCall(UV Offset=@n12) {materialFunction=&ReadCachedDistanceField}
  n16 = Saturate(None=Ceil(None=Multiply(A=@n15.World Distance))) {desc=Mask}
  n17 = Multiply(A=@n11, B=Constant2Vector() {g=1})
  n18 = MaterialFunctionCall(UV Offset=@n17) {materialFunction=&ReadCachedDistanceField}
  n19 = Saturate(None=Ceil(None=Multiply(A=@n18.World Distance))) {desc=Mask}
  n20 = MaterialFunctionCall(UV Offset=Multiply(A=@n17) {constB=-1}) {materialFunction=&ReadCachedDistanceField}
  n21 = Saturate(None=Ceil(None=Multiply(A=@n20.World Distance))) {desc=Mask}
  Multiply(A=Max(A=Max(A=@n14, B=@n16), B=Max(A=@n19, B=@n21))) {desc=Any Offsets Hit WaterBody?}
  n22 = TextureSampleParameter2D() {parameterName=CombinedVelocityAndHeight, texture=&Black}
  n23 = MaterialFunctionCall(World Z=ScalarParameter() {parameterName=ZOffset}) {
   desc="Z Offset pushes water but not terrain carving, thus affects Depth"
   materialFunction=&LocalizeLandscapeZValue
  }
  n24 = Saturate(None=Divide(A=Subtract(A=@n2.Water Depth, B=@n23.Scale Only)) {constB=1000})
  n25 = OneMinus(None=Max(A=OneMinus(None=@n22.A), B=@n24))
  n26 = OneMinus(None=@n24)
  n27 = Saturate(None=Floor(None=@n22.A))
  n28 = Multiply(A=@n8, B=@n27)
  OneMinus(None=LinearInterpolate(A=@n1.A, B=LinearInterpolate(A=@n25, B=@n26, Alpha=@n28), Alpha=@n8))
  n29 = LinearInterpolate(B=ScalarParameter() {defaultValue=64, parameterName=Secondary Shape Dilation}, Alpha=@n27) {constB=0}
  Saturate(None=Ceil(None=Add(A=@n2.World Distance, B=@n29))) {desc=Secondary Dilation Mask}
  n30 = LinearInterpolate(B=@n13.Water Height, Alpha=@n14) {constA=-50000}
  n31 = LinearInterpolate(B=@n15.Water Height, Alpha=@n16) {constA=-50000}
  n32 = LinearInterpolate(B=@n18.Water Height, Alpha=@n19) {constA=-50000}
  n33 = LinearInterpolate(B=@n20.Water Height, Alpha=@n21) {constA=-50000}
  Subtract(A=Max(A=Max(A=@n30, B=@n31), B=Max(A=@n32, B=@n33))) {constB=32}
 ]
}
