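# Unlit Landmass brush material that finalizes the CombinedVelocityAndHeight texture, apparently for water rendering on landscape.
# A Custom HLSL box blur averages RG velocities over texels passing an alpha mask and flags whether any contributed.
# Where none do, the blue-channel height drops to at most the unpacked landscape height minus an offset.
# Height is capped at an encodable maximum, with MPC_Landscape supplying texel size and Z transform.
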Material /Landmass/Landscape/BlueprintBrushes/Materials/FinalizeCombinedTexture
graph {
 properties {
  bAllowNegativeEmissiveColor=true
  blendMode=BLEND_AlphaComposite
  pixelDepthOffsetMode=PDOM_Legacy
  shadingModel=MSM_Unlit
  usedShadingModels=Unlit
 }
 nodes [
  n1 = MaterialFunctionCall!(RG=TextureSampleParameter2D() {parameterName=LandscapeHeight, texture=&Black}.RGB) {materialFunction=/Landmass/Landscape/BlueprintBrushes/MF/UnpackRG8Height}
  n2 = Divide(
   A=Add(A=@n1.Result, B=ScalarParameter() {defaultValue=10000, parameterName=MinTerrainZ})
   B=ScalarParameter() {defaultValue=15000, parameterName=MaxTerrainZ}
  )
  Add(
   A=Multiply(A=@n2, B=CollectionParameter() {collection=/Landmass/Landscape/BlueprintBrushes/MPC/MPC_Landscape, parameterName=LandscapeZScale})
   B=CollectionParameter() {collection=&MPC_Landscape, parameterName=LandscapeZLocation}
  )
  MaterialExpressionTextureObjectParameter_0 = TextureObjectParameter() {parameterName=CombinedVelocityAndHeight, texture=&Black}
  MaterialExpressionTextureCoordinate_0 = TextureCoordinate()
  n3 = CollectionParameter() {collection=&MPC_Landscape, parameterName=RTResX}
  n4 = CollectionParameter() {collection=&MPC_Landscape, parameterName=RTResY}
  MaterialExpressionMultiply_1 = Multiply(A=Divide(B=AppendVector(A=@n3, B=@n4)), B=ScalarParameter() {defaultValue=1, parameterName=Texel Scale})
  MaterialExpressionScalarParameter_4 = ScalarParameter() {defaultValue=5, parameterName=Radius}
  n5 = Custom!(
   Tex=@MaterialExpressionTextureObjectParameter_0
   UV=@MaterialExpressionTextureCoordinate_0
   TexelSize=@MaterialExpressionMultiply_1
   Radius=@MaterialExpressionScalarParameter_4
  ) {
   code="float2 accum = 0;\r\n\r\nfloat2 curUV = UV;\r\nfloat neighbors = 0;\r\n\r\nfor (int i = -Radius; i <= Radius; i++)\r\n{\r\n\r\n\tfor(int j=-Radius; j <= Radius; j++)\r\n\t{\r\n\t\t\r\n\t\tcurUV = UV + float2(TexelSize.x * (float)i, 0) + float2(0, TexelSize.y * (float)j);\r\n\t\tfloat4 cursample = Tex.SampleLevel(TexSampler, curUV, 0);\r\n\t\t\r\n\t\tfloat depthmask = 1 - floor(cursample.a);\r\n\t\t\r\n\t\tif (depthmask == 1.0)\r\n\t\t{\r\n\t\t\taccum += cursample.xy;\r\n\t\t\tneighbors += 1;\r\n\t\t\r\n\t\t}\r\n\t\t\r\n\t}\r\n\t\r\n}\r\n\r\nif(neighbors > 0)\r\n{\r\n\taccum.xy /= neighbors;\r\n}\r\n\r\nreturn float3( accum.xy, saturate(neighbors) );"
   description=Blur Velocities
   inputs [
    {
     inputName=Tex
     input {expression=&self:MaterialExpressionTextureObjectParameter_0}
    }
    {
     inputName=UV
     input {expression=&self:MaterialExpressionTextureCoordinate_0}
    }
    {inputName=TexelSize, input {expression=&self:MaterialExpressionMultiply_1}}
    {
     inputName=Radius
     input {expression=&self:MaterialExpressionScalarParameter_4}
    }
   ]
  }
  n6 = TextureSampleParameter2D() {parameterName=CombinedVelocityAndHeight, texture=&Black}
  n7 = MaterialFunctionCall(Float3=AppendVector(A=ComponentMask(None=@n5) {g=true, r=true}, B=@n6.B)) {materialFunction=&BreakOutFloat3Components}
  n8 = Subtract(A=@n1.Result, B=ScalarParameter() {defaultValue=256, parameterName=Water Terrain Offset})
  n9 = ComponentMask(None=@n5) {b=true}
  n10 = Min(
   A=LinearInterpolate(A=@n7.B, B=Min(A=@n7.B, B=@n8), Alpha=OneMinus(None=@n9))
   B=ScalarParameter() {defaultValue=10000, parameterName=Max Encodable Water Level}
  )
  n11 = CollectionParameter() {collection=&MPC_Landscape, parameterName=LandscapeZScale}
  MaterialFunctionCall(
   X=@n7.R
   Y=@n7.R
   Z=Add(
    A=Multiply(A=@n10, B=@n11)
    B=CollectionParameter() {collection=&MPC_Landscape, parameterName=LandscapeZLocation}
   )
  ) {materialFunction=&MakeFloat3}
  n12 = TextureCoordinate()
  MaterialFunctionCall(Number=@n6.B, UVs=Frac(None=Multiply(A=@n12) {constB=16})) {materialFunction=&DebugScalarValues}
  n13 = CollectionParameter() {collection=&MPC_Landscape, parameterName=WaterClearHeight}
  n14 = CollectionParameter() {collection=&MPC_Landscape, parameterName=LandscapeZLocation}
  Max(
   A=@n8
   B=Divide(
    A=Subtract(A=@n13, B=@n14)
    B=CollectionParameter() {collection=&MPC_Landscape, parameterName=LandscapeZScale}
   )
  )
  OneMinus(None=Saturate(None=@n2))
 ]
}
