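# Unlit Landmass brush material that generates procedural noise and adds it to the height read from HeightRT.
# UVs scaled by MPC_Landscape world size, offset by Position, drive a custom HLSL octave loop of gradient noise, with seed as Z.
# ElevationParams handles invert, pre-offset and elevation; the sum with unpacked height is clamped and repacked to RG8.
# Debug pack/unpack math, a Voronoi Noise node and loose scalar parameters sit outside this chain.
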Material /Landmass/Landscape/BlueprintBrushes/Materials/GenNoise_01
graph {
 properties {bAllowNegativeEmissiveColor=true, pixelDepthOffsetMode=PDOM_Legacy, shadingModel=MSM_Unlit, usedShadingModels=Unlit}
 nodes [
  n1 = ScalarParameter() {defaultValue=-32768, parameterName=H}
  n2 = MaterialFunctionCall!(Height Int -32k to 32k=@n1) {materialFunction=/Landmass/Landscape/OldPrototype_BP/Mats/MF/PackRG8Height}
  Divide(A=@n2.XY) {constB=255}
  n3 = Floor(None=Add(A=@n1) {constB=32768, desc="Make  0-65536"})
  n4 = Floor(None=Divide(A=@n3) {constB=256})
  n5 = Fmod(A=@n3, B=Constant() {r=256})
  n6 = Divide(A=Floor(None=AppendVector(A=@n4, B=@n5))) {constB=255}
  MaterialFunctionCall!(Vector2=@n6, MaximumNumberOfDigits=Constant() {r=8}, UVs, DebugTextLocation RG_UpperRight BA_LowerLeft, Component Spacing) {materialFunction=/Engine/Functions/Engine_MaterialFunctions02/Utility/DebugFloat2Values}
  n7 = MaterialFunctionCall!(RG=ComponentMask(None=TextureSampleParameter2D() {parameterName=HeightRT, texture=/Engine/EngineResources/Black_Low.Black}.RGB) {g=true, r=true}) {materialFunction=/Landmass/Landscape/OldPrototype_BP/Mats/MF/UnpackRG8Height}
  n8 = VectorParameter() {desc="R=Elevation Max, G=Elevation Min, B=Pre-Offset, A=Invert", parameterName=ElevationParams}
  n9 = Divide(A=AppendVector(
    A=CollectionParameter() {collection=/Landmass/Landscape/BlueprintBrushes/MPC/MPC_Landscape, parameterName=WorldSizeX}
    B=CollectionParameter() {collection=&MPC_Landscape, parameterName=WorldSizeY}
   )) {constB=1000}
  n10 = Multiply(A=TextureCoordinate(), B=@n9)
  n11 = Divide(A=ComponentMask(None=VectorParameter() {parameterName=Position}.RGB) {g=true, r=true}) {constB=1000}
  MaterialExpressionVectorParameter_0 = VectorParameter() {desc="R=Tiling, G=Seed, B=Octaves, A=Persistence", parameterName=NoiseSeedParams, defaultValue=<rgba 2 0 2 0.5>}
  MaterialExpressionAppendVector_0 = AppendVector(A=Multiply(A=Subtract(A=@n10, B=@n11), B=@MaterialExpressionVectorParameter_0.R), B=@MaterialExpressionVectorParameter_0.G)
  MaterialExpressionConstant_2 = Constant() {desc="3 = Gradient, 5 = Voronoi", r=3}
  MaterialExpressionConstant_3 = Constant()
  MaterialExpressionMultiply_9 = Multiply(A=@n9, B=@MaterialExpressionVectorParameter_0.R)
  n12 = Custom!(
   Position=@MaterialExpressionAppendVector_0
   Function=@MaterialExpressionConstant_2
   bTurbulence=@MaterialExpressionConstant_3
   Octaves=@MaterialExpressionVectorParameter_0.B
   Persistence=@MaterialExpressionVectorParameter_0.B
   RepeatSize=@MaterialExpressionMultiply_9
  ) {
   code="\r\nfloat Out = 0;\r\nfloat OutScale = 1;\r\n\r\n[loop]\r\nfor(int i = 0; i < Octaves; ++i)\r\n{\r\n\r\n\r\n\tif(bTurbulence)\r\n\t{\r\n\t\tOut += abs(Noise3D_Multiplexer(Function, Position, 4, 1, RepeatSize)) * OutScale;\r\n\t}\r\n\telse\r\n\t{\r\n\t\tOut += Noise3D_Multiplexer(Function, Position, 4, 1, RepeatSize) * OutScale;\r\n\t}\r\n\r\n\tPosition *= 2;\r\n\tRepeatSize *= 2;\r\n\tOutScale *= Persistence;\r\n\r\n}\r\n\r\nif(!bTurbulence)\r\n{\r\n\t// bring -1..1 to 0..1 range\r\n\tOut = Out * 0.5f + 0.5f;\r\n}\r\n\r\n// Out is in 0..1 range\r\nreturn lerp(0, 1, Out);"
   description=CustomNoise
   outputType=CMOT_Float1
   inputs [
    {inputName=Position, input {expression=&self:MaterialExpressionAppendVector_0}}
    {inputName=Function, input {expression=&self:MaterialExpressionConstant_2}}
    {inputName=bTurbulence, input {expression=&self:MaterialExpressionConstant_3}}
    {
     inputName=Octaves
     input {expression=&self:MaterialExpressionVectorParameter_0, mask=1, maskB=1, outputIndex=3}
    }
    {
     inputName=Persistence
     input {expression=&self:MaterialExpressionVectorParameter_0, mask=1, maskA=1, outputIndex=4}
    }
    {inputName=RepeatSize, input {expression=&self:MaterialExpressionMultiply_9}}
   ]
  }
  n13 = Subtract(A=LinearInterpolate(A=@n12, B=OneMinus(None=@n12), Alpha=@n8.A), B=@n8.B)
  n14 = Add(B=LinearInterpolate(A=@n8.G, B=@n8.G, Alpha=Max(A=@n13) {constB=0}))
  MaterialFunctionCall(Height Int -32k to 32k=Clamp(None=Add(A=@n7.Result, B=@n14)) {maxDefault=32767, minDefault=-32768}) {materialFunction=&PackRG8Height}
  Multiply(A=@n2.XY) {constB=255}
  Noise!(World Position=@MaterialExpressionAppendVector_0, FilterWidth) {bTurbulence=false, levelScale=3, levels=16, noiseFunction=NOISEFUNCTION_VoronoiALU, outputMin=0, quality=5, repeatSize=2}
  ScalarParameter() {parameterName=Seed}
  ScalarParameter() {defaultValue=2, parameterName=Tiling}
  ScalarParameter() {defaultValue=5, parameterName=NoiseType}
  ScalarParameter() {defaultValue=0.5, parameterName=Persistence}
  ScalarParameter() {defaultValue=0.3, parameterName=Elevation Pre-Offset}
  ScalarParameter() {defaultValue=3, parameterName=Octaves}
  ScalarParameter() {defaultValue=1, parameterName=Invert}
  Sign(None=MaterialFunctionCall(RG=@n6) {materialFunction=&UnpackRG8Height}.Result)
 ]
}
