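# Material function that snaps UVs to texel centers of the landscape render target for Landmass blueprint brushes.
# It scales the texture coordinate by the RTResX and RTResY values from MPC_Landscape, floors it, adds a half-texel offset and divides by the resolution.
# The TexelCenterUVs output gives point-sampled lookups, avoiding blending between neighboring texels.
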MaterialFunction /Landmass/Landscape/BlueprintBrushes/MF/PointSampleLandscapeUVs
graph {nodes [
  n1 = AppendVector(
   A=CollectionParameter() {collection=/Landmass/Landscape/BlueprintBrushes/MPC/MPC_Landscape, parameterName=RTResX}
   B=CollectionParameter() {collection=&MPC_Landscape, parameterName=RTResY}
  )
  n2 = Floor(None=Multiply(A=TextureCoordinate(), B=@n1))
  FunctionOutput(None=Divide(A=Add(A=@n2) {constB=0.5}, B=@n1)) {outputName=TexelCenterUVs}
 ]}
