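# Unlit Landmass brush material that stamps a displacement texture onto landscape height data read from a render target.
# It converts texel UVs to world space via MPC_Landscape, then offsets, scales and rotates them into brush-local texture coordinates.
# The selected channel, scaled by height, gets a Z offset and box-mask edge falloff.
# Alpha, min, max and additive blends against the unpacked height feed BrushBlendModes_Manual, whose result is repacked as RG8.
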Material /Landmass/Landscape/BlueprintBrushes/Materials/M_SimpleBrush
graph {
 properties {pixelDepthOffsetMode=PDOM_Legacy, shadingModel=MSM_Unlit, usedShadingModels=Unlit}
 nodes [
  n1 = MaterialFunctionCall() {materialFunction=/Landmass/Landscape/BlueprintBrushes/MF/PointSampleLandscapeUVs}
  n2 = Multiply(
   A=@n1.TexelCenterUVs
   B=ComponentMask(None=CollectionParameter() {collection=/Landmass/Landscape/BlueprintBrushes/MPC/MPC_Landscape, parameterName=RTWorldSize}) {g=true, r=true}
  )
  n3 = ComponentMask(None=CollectionParameter() {collection=&MPC_Landscape, parameterName=LandscapeLocation}) {g=true, r=true}
  n4 = VectorParameter() {parameterName=P}
  n5 = Multiply(A=VectorParameter() {parameterName=BrushScale, defaultValue=<rgba 1 1 1 0>}.RGB, B=ScalarParameter() {defaultValue=4096, parameterName=Size})
  n6 = Divide(
   A=Subtract(A=Add(A=@n2, B=@n3), B=ComponentMask(None=@n4.RGB) {g=true, r=true})
   B=ComponentMask(None=Multiply(A=@n5) {constB=2}) {g=true, r=true}
  )
  n7 = VectorParameter() {parameterName=BrushRotation, defaultValue=<rgba 1 0 0 0>}
  n8 = ComponentMask(None=@n7.RGB) {g=true, r=true}
  n9 = MaterialFunctionCall(X=Multiply(A=@n7.G) {constB=-1}, Y=@n7.R) {materialFunction=&MakeFloat2}
  n10 = ConstantBiasScale(None=AppendVector(A=DotProduct(A=@n6, B=@n8), B=DotProduct(A=@n6, B=@n9.Result)))
  n11 = VectorParameter() {parameterName=Channel, defaultValue=<rgba 1 0 0 0>}
  n12 = DotProduct(
   A=TextureSampleParameter2D(UVs=Saturate(None=@n10)) {parameterName=Displacement, texture=/Engine/EditorResources/Ai_Spawnpoint}.RGBA
   B=AppendVector(A=@n11.RGB, B=@n11.RGB)
  )
  n13 = Multiply(
   A=Multiply(A=@n12, B=ScalarParameter() {defaultValue=2500, parameterName=Height})
   B=VectorParameter() {parameterName=BrushScale, defaultValue=<rgba 1 1 1 0>}.B
  )
  n14 = ScalarParameter() {defaultValue=0.3, parameterName=EdgeFalloff}
  n15 = MaterialFunctionCall!(A=@n10, B=Constant() {r=0.5}, Bounds=Subtract(A=Constant() {r=0.99}, B=Multiply(A=@n14) {constB=2}), Edge Falloff=@n14) {desc=SafeClip, materialFunction=/Engine/Functions/Engine_MaterialFunctions01/Gradient/BoxMask-2D}
  n16 = Multiply(A=@n15.Result, B=Add(A=ScalarParameter() {parameterName=Z}, B=@n13))
  n17 = Saturate(None=Divide(A=@n12, B=ScalarParameter() {defaultValue=0.01, parameterName=AlphaFalloff}))
  n18 = MaterialFunctionCall!(RG=ComponentMask(None=TextureSampleParameter2D(UVs=@n1.TexelCenterUVs) {parameterName=HeightRT, texture=&Black}.RGB) {g=true, r=true}) {materialFunction=/Landmass/Landscape/BlueprintBrushes/MF/UnpackRG8Height}
  n19 = ScalarParameter() {parameterName=Z}
  n20 = Multiply(A=@n15.Result, B=Add(A=@n13, B=Multiply(A=@n17, B=@n19)))
  MaterialFunctionCall!(Height Int -32k to 32k=MaterialFunctionCall!(
    AlphaBlend=LinearInterpolate(A=@n18.Result, B=@n16, Alpha=@n17)
    Min=Min(A=@n18.Result, B=@n20)
    Max=Max(A=@n18.Result, B=@n20)
    Additive=Add(A=@n18.Result, B=@n20)
   ) {materialFunction=/Landmass/Landscape/OldPrototype_BP/Mats/MF/BrushBlendModes_Manual}.Result) {materialFunction=/Landmass/Landscape/BlueprintBrushes/MF/PackRG8Height}
 ]
}
