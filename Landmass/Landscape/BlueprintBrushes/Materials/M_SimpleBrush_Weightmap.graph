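# M_SimpleBrush_Weightmap is an unlit Landmass brush material that adds a soft circular spot to a landscape render target, by its name a weightmap.
# Texel-center UVs from PointSampleLandscapeUVs are rescaled by render target size over landscape quad count from MPC_Landscape, then centered and converted to landscape units.
# A zero-hardness sphere mask at point P with radius R is scaled by Opacity and added to the red channel of HeightRT.
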Material /Landmass/Landscape/BlueprintBrushes/Materials/M_SimpleBrush_Weightmap
graph {
 properties {pixelDepthOffsetMode=PDOM_Legacy, shadingModel=MSM_Unlit, usedShadingModels=Unlit}
 nodes [
  n1 = MaterialFunctionCall() {materialFunction=/Landmass/Landscape/BlueprintBrushes/MF/PointSampleLandscapeUVs}
  n2 = AppendVector(
   A=CollectionParameter() {collection=/Landmass/Landscape/BlueprintBrushes/MPC/MPC_Landscape, parameterName=LSQuadsX}
   B=CollectionParameter() {collection=&MPC_Landscape, parameterName=LSQuadsY}
  )
  n3 = Divide(
   A=AppendVector(
    A=CollectionParameter() {collection=&MPC_Landscape, parameterName=RTResX}
    B=CollectionParameter() {collection=&MPC_Landscape, parameterName=RTResY}
   )
   B=@n2
  ) {constA=1024, constB=1024, desc=RT Size / LS Resolution}
  n4 = Multiply(A=Subtract(A=Multiply(A=@n1.TexelCenterUVs, B=@n3)) {constB=0.5}, B=Multiply(A=@n2) {constB=100})
  n5 = ComponentMask(None=VectorParameter() {parameterName=P}.RGB) {g=true, r=true}
  n6 = ScalarParameter() {defaultValue=2048, parameterName=R}
  Add(
   A=TextureSampleParameter2D() {parameterName=HeightRT, texture=&Black}.R
   B=Multiply(A=SphereMask(A=@n4, B=@n5, Radius=@n6) {hardnessPercent=0}, B=ScalarParameter() {defaultValue=1, parameterName=Opacity})
  )
  TextureCoordinate()
 ]
}
