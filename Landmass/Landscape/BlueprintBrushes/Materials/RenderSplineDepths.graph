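# Unlit Landmass blueprint brush material for spline meshes that, going by its name, renders spline depths, along with directional velocity data.
# It interpolates velocity and depth between A and B parameters along the U texture coordinate.
# Velocity scales a vector from the TangentBasis function, whose XY is packed with depth into a three-channel value.
# Full float precision and negative emissive color are enabled, allowing signed data.
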Material /Landmass/Landscape/BlueprintBrushes/Materials/RenderSplineDepths
graph {
 properties {
  bAllowNegativeEmissiveColor=true
  bUsedWithSplineMeshes=true
  bUsedWithStaticLighting=true
  floatPrecisionMode=MFPM_Full
  pixelDepthOffsetMode=PDOM_Legacy
  shadingModel=MSM_Unlit
  usedShadingModels=Unlit
 }
 nodes [
  n1 = LinearInterpolate(
   A=ScalarParameter() {defaultValue=128, parameterName=VelA}
   B=ScalarParameter() {defaultValue=128, parameterName=VelB}
   Alpha=ComponentMask(None=TextureCoordinate()) {r=true}
  )
  n2 = MaterialFunctionCall() {materialFunction=&TangentBasis}
  n3 = TextureCoordinate()
  AppendVector(
   A=ComponentMask(None=Multiply(A=@n1, B=@n2.X U)) {g=true, r=true}
   B=LinearInterpolate(
    A=ScalarParameter() {parameterName=DepthA}
    B=ScalarParameter() {defaultValue=0.2, parameterName=DepthB}
    Alpha=ComponentMask(None=@n3) {r=true}
   )
  )
  Divide(A=@n1) {constB=100}
 ]
}
