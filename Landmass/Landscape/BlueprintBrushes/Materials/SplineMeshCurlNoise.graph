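# Unlit, two-sided spline mesh material for Landmass Blueprint brushes, with graphs for a line pattern and curl noise; no material output links are shown.
# Screen-space derivatives of UV V mark integer crossings at LineWidth, giving bright green lines over a dim base, with a constant fallback on ES2/ES3_1.
# Two seeded curl-noise layers sample world XY normalized by MPC_Landscape size, are scaled by the Curl parameter, then summed and masked to XY.
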Material /Landmass/Landscape/BlueprintBrushes/Materials/SplineMeshCurlNoise
graph {
 properties {bUsedWithSplineMeshes=true, pixelDepthOffsetMode=PDOM_Legacy, shadingModel=MSM_Unlit, twoSided=true, usedShadingModels=Unlit}
 nodes [
  n1 = ComponentMask(None=TextureCoordinate()) {g=true}
  n2 = ScalarParameter() {defaultValue=2, parameterName=LineWidth}
  n3 = Multiply(A=DDX(Value=@n1), B=@n2)
  n4 = Multiply(A=DDY(Value=@n1), B=@n2)
  n5 = Add(A=Add(A=@n1, B=@n3), B=@n4)
  n6 = Subtract(A=Subtract(A=@n1, B=@n3), B=@n4)
  MaterialExpressionAbs_0 = Abs(None=Subtract(A=Floor(None=@n5), B=Floor(None=@n6)))
  n7 = Add(A=Subtract(A=@n1, B=@n3), B=@n4)
  n8 = Subtract(A=Add(A=@n1, B=@n3), B=@n4)
  MaterialExpressionAbs_2 = Abs(None=Subtract(A=Floor(None=@n7), B=Floor(None=@n8)))
  n9 = Custom!(A=@MaterialExpressionAbs_0, B=@MaterialExpressionAbs_2) {
   code="max(A,B)"
   description=Max
   outputType=CMOT_Float1
   inputs [
    {inputName=A, input {expression=&self:MaterialExpressionAbs_0}}
    {inputName=B, input {expression=&self:MaterialExpressionAbs_2}}
   ]
  }
  n10 = If(A=@n9, B=Constant() {r=1}, "A > B"=Constant() {r=1}, "A < B"=Constant() {r=0.15})
  n11 = Constant() {r=1}
  FeatureLevelSwitch(Default=@n10, ES2=@n11, ES3_1=@n11, SM4_REMOVED=@n10)
  n12 = Multiply(A=Constant3Vector() {constant=<rgba 0.35 0.9 0 0>}, B=@n10)
  n13 = Constant() {r=1}
  FeatureLevelSwitch(Default=@n12, ES2=@n13, ES3_1=@n13, SM4_REMOVED=@n12)
  n14 = Divide(
   A=ComponentMask(None=WorldPosition().XYZ) {g=true, r=true}
   B=AppendVector(
    A=CollectionParameter() {collection=/Landmass/Landscape/BlueprintBrushes/MPC/MPC_Landscape, parameterName=WorldSizeX}
    B=CollectionParameter() {collection=&MPC_Landscape, parameterName=WorldSizeY}
   )
  )
  n15 = ScalarParameter() {parameterName=Seed1}
  n16 = VectorParameter() {parameterName=Curl, defaultValue=<rgba 3 15000 24 0>}
  n17 = VectorNoise(World Position=Multiply(A=AppendVector(A=@n14, B=@n15), B=@n16.R)) {noiseFunction=VNF_CurlALU}
  n18 = Divide(A=@n16.G) {constB=100}
  n19 = ScalarParameter() {parameterName=Seed2}
  n20 = VectorNoise(World Position=Multiply(A=AppendVector(A=@n14, B=@n19), B=@n16.B)) {noiseFunction=VNF_CurlALU}
  n21 = Divide(A=@n16.A) {constB=100}
  Multiply(A=Add(A=Multiply(A=@n17, B=@n18), B=Multiply(A=@n20, B=@n21)), B=Constant3Vector() {constant=<rgba 1 1 0 0>})
  TextureCoordinate()
 ]
}
