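; Landmass brush child actor Blueprint that builds a chain of spline mesh components along a referenced spline.
; UpdateValues stores the spline, material, mesh, scale and segment length; the construction script calls SetupSplineMeshComponents to clear and rebuild the segments.
; Segment length is adjusted so whole segments evenly fill the spline, with normalized tangents scaled to that length.
; End scales multiply Scale by spline scale, sampled by passing distance as time.
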(event EventBeginPlay)

(event Collision|EventActorBeginOverlap
  (params (OtherActor :type Actor%20Object%20Reference)))

(event EventTick
  (params (DeltaSeconds :type Float%20%28single-precision%29)))

(fn SetupSplineMeshComponents
  (bind _segment_length (Variables|Default|GetSegment%20Length))
  (bind _total_spine_length (Variables|Default|GetTotal%20Spine%20Length))
  (bind _returnvalue (/ _total_spine_length _segment_length))
  (bind _returnvalue_2 (Math|Float|Floor _returnvalue))
  (bind _adjustedlength (LocalVariables|GetAdjustedLength))
  (bind _idx (Variables|Default|GetIdx))
  (bind _returnvalue_3 (+ _idx 1))
  (bind _returnvalue_4 (* _returnvalue_3 _adjustedlength))
  (bind _returnvalue_5 (* _idx _adjustedlength))
  (bind _scale (Variables|Default|GetScale))
  (bind _splineref (Variables|Default|GetSplineRef))
  (bind _components (Variables|Default|GetComponents))
  (for _array_element _components
    (Components|DestroyComponent _array_element))
  (Utilities|Array|Clear _components)
  (Variables|Default|SetTotal%20Spine%20Length (Spline|GetSplineLength (Variables|Default|GetSplineRef)))
  (LocalVariables|SetAdjustedLength (/ _total_spine_length (Math|Conversions|ToFloat(Integer) _returnvalue_2)))
  (for _index (range _returnvalue_2)
    (Variables|Default|SetIdx _index)
    (bind _returnvalue_6 (AddComponent|NODE_AddSplineMeshComponent-1 :RelativeTransform (Math|Transform|MakeTransform "0, 0, 0")))
    (Components|StaticMesh|SetStaticMesh _returnvalue_6 (Variables|Default|GetMesh))
    (Rendering|Material|SetMaterial _returnvalue_6 0 (Variables|Default|GetMaterial))
    (SplineMesh|SetStartScale _returnvalue_6 (Math|Conversions|ToVector2D(Vector) (* _scale (Math|Vector|MakeVector (.y (Spline|GetScaleatTime _splineref _returnvalue_5 true)) (.z (Spline|GetScaleatTime _splineref _returnvalue_5 true))))) false :!class /Script/Engine.SplineMeshComponent)
    (SplineMesh|SetEndScale _returnvalue_6 (Math|Conversions|ToVector2D(Vector) (* _scale (Math|Vector|MakeVector (.y (Spline|GetScaleatTime _splineref _returnvalue_4 true)) (.z (Spline|GetScaleatTime _splineref _returnvalue_4 true))))) false :!class /Script/Engine.SplineMeshComponent)
    (SplineMesh|SetStartandEnd _returnvalue_6 (Spline|GetLocationatDistanceAlongSpline (Variables|Default|GetSplineRef) _returnvalue_5 "Local") (* (Math|Vector|Normalize (Spline|GetTangentatDistanceAlongSpline (Variables|Default|GetSplineRef) _returnvalue_5 "Local")) (LocalVariables|GetAdjustedLength)) (Spline|GetLocationatDistanceAlongSpline (Variables|Default|GetSplineRef) _returnvalue_4 "Local") (* (Math|Vector|Normalize (Spline|GetTangentatDistanceAlongSpline (Variables|Default|GetSplineRef) _returnvalue_4 "Local")) (LocalVariables|GetAdjustedLength)) true :!class /Script/Engine.SplineMeshComponent)
    (Utilities|Array|Add (Variables|Default|GetComponents) _returnvalue_6)))

(fn UpdateValues
  (params (Spline :type Spline%20Component%20Object%20Reference) (Mat :type Material%20Interface%20Object%20Reference) (Mesh :type Static%20Mesh%20Object%20Reference) (Scale :type Vector) (Segment%20Length :type Float%20%28double-precision%29))
  (Variables|Default|SetSplineRef Spline)
  (Variables|Default|SetMaterial Mat)
  (Variables|Default|SetMesh Mesh)
  (Variables|Default|SetScale Scale)
  (Variables|Default|SetSegment%20Length Segment%20Length))

(fn ConstructionScript
  (CallFunction|SetupSplineMeshComponents))
