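; Experimental Landmass landscape brush that tests round-tripping render target pixels through CPU-side data; its RenderLayer override only allocates a render target.
; CopyMaterial to RT draws a material, packs it into a quarter-width float target, and reads it back into a LandscapeTextureData struct.
; Copy Data to RT rebuilds 16-pixel tiles by feeding texels through 64 vector parameters into canvas material draws.
; Other events convert the target to a Texture2D and print index tests.
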(event EventBeginPlay)

(event Collision|EventActorBeginOverlap
  (params (OtherActor :type Actor%20Object%20Reference)))

(event EventTick
  (params (DeltaSeconds :type Float%20%28single-precision%29)))

(event Custom|CopyMaterial%20to%20RT
  (bind _defaultsize (Variables|Default|GetDefaultSize))
  (bind _returnvalue (Math|Conversions|ToFloat(Integer) _defaultsize))
  (bind _returnvalue_2 (/ _returnvalue 4.0))
  (bind _returnvalue_3 (Math|Color|MakeColor 0.0 0.0 _returnvalue_2 _returnvalue))
  (bind _rt_packed (Variables|Default|GetRT_Packed))
  (bind _returnvalue_4 (Rendering|RenderTargetSampleRectangleEditorOnly _rt_packed _returnvalue_3))
  (bind _defaultsize_2 (Variables|Default|GetDefaultSize))
  (bind _returnvalue_5 (Math|Conversions|ToFloat(Integer) _defaultsize_2))
  (bind _rt_packed_2 (Variables|Default|GetRT_Packed))
  (bind _packmid (Variables|Default|GetPackMID))
  (Rendering|CreateAndAssignRenderTarget (Variables|Default|GetRT) (Variables|Default|GetDefaultSize) (Variables|Default|GetDefaultSize) "RTF_RGBA16f"
    (:Out
      (Rendering|DrawMaterialtoRenderTarget (Variables|Default|GetRT) (Variables|Default|GetMaterial))
      (Rendering|CreateAndAssignMID (Variables|Default|GetPackMID) "/Landmass/Landscape/BlueprintBrushes/Materials/Erosion/Erosion_PackHeightData.Erosion_PackHeightData"
        (:Out
          (Rendering|Material|SetTextureParameterValue (Variables|Default|GetPackMID) "RT" (Variables|Default|GetRT))
          (Rendering|Material|SetScalarParameterValue _packmid "SizeX" _returnvalue_5 :!class /Script/Engine.MaterialInstanceDynamic)
          (Rendering|Material|SetScalarParameterValue _packmid "SizeY" _returnvalue_5 :!class /Script/Engine.MaterialInstanceDynamic)
          (Rendering|CreateAndAssignRenderTarget (Variables|Default|GetRT_Packed) (/ (Variables|Default|GetDefaultSize) 4) (Variables|Default|GetDefaultSize) "RTF_RGBA32f"
            (:Out
              (Rendering|ClearRenderTarget2D _rt_packed_2)
              (Rendering|DrawMaterialtoRenderTarget _rt_packed_2 (Variables|Default|GetPackMID))
              (Development|PrintString (Utilities|String|ToString(Integer) (Utilities|Array|Length _returnvalue_4)))
              (Utilities|Struct|SetmembersinLandscapeTextureData (Variables|Default|GetTexture) _returnvalue_4)
              (Variables|Default|SetRendered true))))))))

(event Custom|Copy%20Data%20to%20RT
  (bind _rt_2 (Variables|Default|GetRT_2))
  (Rendering|CreateAndAssignMID (Variables|Default|GetWriteRTMID) "/Landmass/Landscape/BlueprintBrushes/Materials/Debug/ReadTextureData.ReadTextureData"
    (:Out
      (Rendering|CreateAndAssignRenderTarget (Variables|Default|GetRT_2) (Variables|Default|GetDefaultSize) (Variables|Default|GetDefaultSize)
        (:Out
          (Rendering|ClearRenderTarget2D _rt_2)
          (bind (_canvas _size _context) (Rendering|BeginDrawCanvastoRenderTarget _rt_2))
          (for _index (range (/ (Variables|Default|GetDefaultSize) 16))
            (Variables|Default|SetCur_X _index)
            (for _index_2 (range (/ (Variables|Default|GetDefaultSize) 16))
              (Variables|Default|SetCur_Y _index_2)
              (for _index_3 (range 64)
                (Rendering|Material|SetVectorParameterValue (Variables|Default|GetWriteRTMID) (Utilities|String|StringToName (Utilities|String|Append "p" (Utilities|String|ToString(Integer) _index_3))) (Utilities|Get4Texels (Math|Conversions|Indices|Converta2DIndextoa1DIndex (+ (Math|Conversions|Indices|Converta1DIndextoa2DIndex _index_3 4) (Utilities|Struct|MakeIntPoint (* (Variables|Default|GetCur_X) 4) (* (Variables|Default|GetCur_Y) 16))) (/ (Variables|Default|GetDefaultSize) 2))) :!class /Script/Engine.MaterialInstanceDynamic)
                (Canvas|DrawMaterial :self _canvas :RenderMaterial (Variables|Default|GetWriteRTMID) :ScreenPosition (+ (Math|Vector2D|MakeVector2D (Math|Conversions|ToFloat(Integer) (* (Variables|Default|GetCur_X) 16)) (Math|Conversions|ToFloat(Integer) (* (Variables|Default|GetCur_Y) 16))) "(X=0.500000,Y=0.500000)") :ScreenSize "(X=16.000000,Y=16.000000)" :CoordinateSize "(X=1.000000,Y=1.000000)" :Rotation 0.0 :PivotPoint "(X=0.500000,Y=0.500000)"))))
          (Rendering|EndDrawCanvastoRenderTarget _context))))))

(event Custom|Make%20Texture
  (bind _returnvalue (Game|ConstructObjectfromClass "/Script/Engine.Texture2D"))
  (bind _output_get (Variables|Default|SetTex _returnvalue))
  (EditorScripting|Rendering|ConvertRenderTargettoTexture2DEditorOnly (Variables|Default|GetRT) _output_get))

(event Custom|Test%20Index
  (for _index (range 64)
    (Development|PrintString (Utilities|String|ToString(IntPoint) (Math|Conversions|Indices|Converta1DIndextoa2DIndex _index 4)))))

(fn RenderLayer
  (params (InParameters :type Landscape%20Brush%20Parameters%20Structure%20%28by%20ref%29))
  (bind (_renderareaworldtransform _renderareasize _combinedresult) (Utilities|Struct|BreakLandscapeBrushParameters InParameters))
  (Rendering|CreateAndAssignRenderTarget (Variables|Default|GetRT) (Variables|/Script/Engine.TextureRenderTarget2D|GetSizeX _combinedresult) (Variables|/Script/Engine.TextureRenderTarget2D|GetSizeY _combinedresult)))

(fn ConstructionScript
  (Development|PrintString "Construction Script"))
