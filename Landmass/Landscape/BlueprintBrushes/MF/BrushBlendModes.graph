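# Material function for Landmass landscape blueprint brushes that selects one of four blend results: AlphaBlend, Min, Max or Additive.
# It floors a BlendMode scalar parameter, then derives selection weights by successive subtraction and saturation.
# Three chained lerps use those weights as a branchless switch, so modes 0 to 3 pick the inputs in that order.
# The blend results arrive precomputed as function inputs.
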MaterialFunction /Landmass/Landscape/BlueprintBrushes/MF/BrushBlendModes
graph {nodes [
  n1 = Floor(None=ScalarParameter() {parameterName=BlendMode})
  n2 = LinearInterpolate(
   A=FunctionInput() {bUsePreviewValueAsDefault=true, inputName=AlphaBlend, inputType=FunctionInput_Scalar}
   B=FunctionInput() {bUsePreviewValueAsDefault=true, inputName=Min, inputType=FunctionInput_Scalar, sortPriority=1}
   Alpha=Saturate(None=@n1)
  )
  n3 = FunctionInput() {bUsePreviewValueAsDefault=true, inputName=Max, sortPriority=2}
  n4 = Subtract(A=@n1)
  n5 = Saturate(None=@n4)
  n6 = Subtract(A=@n4)
  FunctionOutput(None=LinearInterpolate(
    A=LinearInterpolate(A=@n2, B=@n3, Alpha=@n5)
    B=FunctionInput() {bUsePreviewValueAsDefault=true, inputName=Additive, sortPriority=3}
    Alpha=Saturate(None=@n6)
   ))
 ]}
