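# Material function for Landmass landscape brushes that outputs UVs for placing brush content within the landscape render target.
# It rebuilds world XY from texture coordinates and MPC_Landscape values, then offsets by brush position and projects onto the brush's rotated axes.
# The result is divided by brush scale and size, centered and saturated.
# Scalar parameters blend to landscape-wide UVs (DrawToEntireLandscape) or raw texture coordinates (ForceUseRegularUVs).
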MaterialFunction /Landmass/Landscape/BlueprintBrushes/MF/LandmassBrush_UVs
graph {nodes [
  n1 = Multiply(
   A=TextureCoordinate()
   B=ComponentMask(None=CollectionParameter() {collection=/Landmass/Landscape/BlueprintBrushes/MPC/MPC_Landscape, parameterName=RTWorldSize}) {g=true, r=true}
  )
  n2 = Multiply(A=ComponentMask(None=CollectionParameter() {collection=&MPC_Landscape, parameterName=LandscapeScale}) {g=true, r=true}) {constB=0.5}
  n3 = CollectionParameter() {collection=&MPC_Landscape, parameterName=LandscapeLocation}
  n4 = Subtract(
   A=Add(A=Subtract(A=@n1, B=@n2), B=ComponentMask(None=@n3) {g=true, r=true})
   B=ComponentMask(None=VectorParameter() {parameterName=BrushPosition}.RGB) {g=true, r=true}
  )
  n5 = VectorParameter() {parameterName=BrushXVector, defaultValue=<rgba 1 0 0 0>}
  n6 = Multiply(A=@n5.G) {constB=-1}
  n7 = AppendVector(
   A=DotProduct(A=@n4, B=ComponentMask(None=@n5.RGB) {g=true, r=true})
   B=DotProduct(A=@n4, B=MaterialFunctionCall(X=@n6, Y=@n5.R) {materialFunction=&MakeFloat2}.Result)
  )
  n8 = ComponentMask(None=Multiply(
    A=VectorParameter() {parameterName=BrushScale, defaultValue=<rgba 1 1 1 0>}.RGB
    B=ScalarParameter() {defaultValue=4096, parameterName=BrushSize}
   )) {g=true, r=true}
  n9 = Saturate(None=Add(A=Divide(A=@n7, B=@n8)) {constB=0.5})
  n10 = CollectionParameter() {collection=&MPC_Landscape, parameterName=RTWorldSize}
  n11 = CollectionParameter() {collection=&MPC_Landscape, parameterName=WorldSizeX}
  n12 = CollectionParameter() {collection=&MPC_Landscape, parameterName=WorldSizeY}
  n13 = Multiply(A=TextureCoordinate(), B=Divide(A=ComponentMask(None=@n10) {g=true, r=true}, B=AppendVector(A=@n11, B=@n12)))
  n14 = ScalarParameter() {parameterName=DrawToEntireLandscape}
  FunctionOutput(None=LinearInterpolate(A=LinearInterpolate(A=@n9, B=@n13, Alpha=@n14), B=TextureCoordinate(), Alpha=ScalarParameter() {parameterName=ForceUseRegularUVs})) {outputName=UVs}
 ]}
