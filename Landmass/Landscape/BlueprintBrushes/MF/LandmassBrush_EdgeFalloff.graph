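# Material function for Landmass landscape brushes that outputs an Edge Mask, fading toward the borders of the brush UV square.
# It builds a rounded-box distance field from UVs mirrored around the center, shaped by Falloff Width, Falloff Radius and Offset inputs.
# The distance is remapped over a falloff band and inverted without clamping, with an If that outputs 1 when its tested value equals a near-zero threshold.
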MaterialFunction /Landmass/Landscape/BlueprintBrushes/MF/LandmassBrush_EdgeFalloff
graph {nodes [
  n1 = Constant()
  n2 = FunctionInput(Preview=MaterialFunctionCall() {materialFunction=&/LandmassBrush_UVs}.UVs) {bUsePreviewValueAsDefault=true, inputName=UVs, inputType=FunctionInput_Vector2, previewValue=<0.1 0 0 1>}
  n3 = Abs(None=Subtract(A=NamedRerouteDeclaration(None=@n2) {name=UVs, nodeColor=<rgba 0.3542 1 0 1>}) {constB=0.5})
  n4 = Subtract(B=NamedRerouteUsage()) {constA=0.5}
  n5 = Multiply(A=NamedRerouteUsage()) {constB=2}
  n6 = Subtract(B=Add(A=Multiply(A=@n4, B=@n5), B=NamedRerouteUsage())) {constA=0.5}
  n7 = Max(A=Add(A=NamedRerouteUsage(), B=Subtract(A=@n3, B=@n6))) {constB=0}
  n8 = Max(A=Add(A=Distance(A=@n1, B=@n7), B=@n6)) {constB=0}
  n9 = NamedRerouteUsage()
  n10 = Subtract(B=@n9) {constA=0.5}
  n11 = OneMinus(None=Divide(A=Subtract(A=@n8, B=@n10), B=@n9))
  FunctionOutput(None=If(A=NamedRerouteUsage(), "A > B"=@n11, "A == B"=Constant() {r=1}, "A < B"=@n11) {constB=0.0001}) {outputName=Edge Mask}
  n12 = FunctionInput() {bUsePreviewValueAsDefault=true, inputName=Falloff Width, inputType=FunctionInput_Scalar, sortPriority=1, previewValue=<0.1 0 0 1>}
  NamedRerouteDeclaration(None=@n12) {name=Falloff Width, nodeColor=<rgba 0.3359 0 1 1>}
  n13 = FunctionInput() {bUsePreviewValueAsDefault=true, inputName=Falloff Radius, inputType=FunctionInput_Scalar, sortPriority=2, previewValue=<0.2 0 0 1>}
  NamedRerouteDeclaration(None=@n13) {name=Falloff Radius, nodeColor=<rgba 0.1016 1 0 1>}
  NamedRerouteDeclaration(None=FunctionInput() {bUsePreviewValueAsDefault=true, inputName=Offset, inputType=FunctionInput_Scalar, sortPriority=3}) {name=Offset, nodeColor=<rgba 1 0.2344 0 1>}
 ]}
