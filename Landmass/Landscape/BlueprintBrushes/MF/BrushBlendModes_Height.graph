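# Material function for Landmass Blueprint brushes that blends a brush's height into the existing landscape height.
# It point-samples and unpacks the previous height from HeightRT, taking current height, edge and displacement masks as inputs.
# Alpha, smooth min, smooth max and additive candidates feed BrushBlendModes_Manual, with displacement depths scaled by brush size over landscape Z scale.
# An edge falloff mask, overridable by DrawToEntireLandscape, limits where the blended result applies.
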MaterialFunction /Landmass/Landscape/BlueprintBrushes/MF/BrushBlendModes_Height
graph {nodes [
  n1 = NamedRerouteUsage()
  n2 = LinearInterpolate(A=NamedRerouteUsage(), Alpha=NamedRerouteUsage()) {constB=0}
  n3 = NamedRerouteUsage()
  n4 = LinearInterpolate(B=NamedRerouteUsage(), Alpha=ScalarParameter() {defaultValue=1, parameterName=RelativeDepth}) {constA=1}
  n5 = LinearInterpolate(
   A=LinearInterpolate(A=NamedRerouteUsage(), B=Subtract(A=@n1, B=@n2), Alpha=NamedRerouteUsage())
   B=LinearInterpolate(A=NamedRerouteUsage(), B=NamedRerouteUsage(), Alpha=Multiply(A=@n3, B=@n4))
   Alpha=ScalarParameter() {parameterName=DisplacementAsAlpha}
  )
  n6 = NamedRerouteUsage()
  n7 = NamedRerouteUsage()
  n8 = MaterialFunctionCall!(
   A=NamedRerouteUsage()
   B=Add(A=NamedRerouteUsage(), B=LinearInterpolate(A=@n6, Alpha=@n7) {constB=0})
   K=Multiply(A=ScalarParameter() {parameterName=Smoothness}, B=NamedRerouteUsage())
  ) {materialFunction=/Landmass/Landscape/OldPrototype_BP/Mats/MF/SmoothIntersections}
  n9 = LinearInterpolate(A=NamedRerouteUsage(), Alpha=NamedRerouteUsage()) {constB=0}
  n10 = LinearInterpolate(A=NamedRerouteUsage(), Alpha=NamedRerouteUsage()) {constB=0}
  n11 = MaterialFunctionCall(
   A=NamedRerouteUsage()
   B=Subtract(A=NamedRerouteUsage(), B=Add(A=@n9, B=@n10))
   K=Multiply(A=ScalarParameter() {parameterName=Smoothness}, B=NamedRerouteUsage())
  ) {materialFunction=&SmoothIntersections}
  n12 = NamedRerouteUsage()
  n13 = NamedRerouteUsage()
  n14 = Add(A=NamedRerouteUsage(), B=Multiply(A=NamedRerouteUsage(), B=Multiply(A=@n12, B=@n13)))
  n15 = Saturate(None=MaterialFunctionCall!(UVs, Falloff Width=Constant() {r=0.001}, Falloff Radius=Constant(), Offset) {materialFunction=&/LandmassBrush_EdgeFalloff}.Edge Mask)
  n16 = ScalarParameter() {parameterName=DrawToEntireLandscape}
  FunctionOutput(None=LinearInterpolate(
    A=NamedRerouteUsage()
    B=MaterialFunctionCall!(AlphaBlend=@n5, Min=@n8.Smooth Min, Max=@n11.Smooth Max, Additive=@n14) {materialFunction=/Landmass/Landscape/OldPrototype_BP/Mats/MF/BrushBlendModes_Manual}.Result
    Alpha=LinearInterpolate(A=@n15, Alpha=@n16)
   ))
  n17 = TextureSampleParameter2D(UVs=MaterialFunctionCall() {materialFunction=&/PointSampleLandscapeUVs}.TexelCenterUVs) {parameterName=HeightRT, texture=/Engine/EngineResources/Black_Low.Black}
  NamedRerouteDeclaration(None=MaterialFunctionCall!(RG=ComponentMask(None=@n17.RGB) {g=true, r=true}) {materialFunction=/Landmass/Landscape/OldPrototype_BP/Mats/MF/UnpackRG8Height}.Result) {name=Previous Height, nodeColor=<rgba 0 0.4609 1 1>}
  NamedRerouteDeclaration(None=FunctionInput() {bUsePreviewValueAsDefault=true, inputName=Current Height, inputType=FunctionInput_Scalar}) {name=Current Height, nodeColor=<rgba 1 0.7969 0 1>}
  NamedRerouteDeclaration(None=FunctionInput() {bUsePreviewValueAsDefault=true, inputName=Edge Mask, inputType=FunctionInput_Scalar, sortPriority=1}) {name=Edge Alpha, nodeColor=<rgba 0 0.9531 1 1>}
  n18 = FunctionInput() {bUsePreviewValueAsDefault=true, inputName=Displacement Mask, inputType=FunctionInput_Scalar, sortPriority=2, previewValue=<1 0 0 1>}
  NamedRerouteDeclaration(None=@n18) {name=Displacement Mask, nodeColor=<rgba 0.1953 0 1 1>}
  n19 = ScalarParameter() {defaultValue=1, parameterName=RelativeDepth}
  n20 = ScalarParameter() {defaultValue=4096, parameterName=BrushSize}
  NamedRerouteDeclaration(None=Divide(
    A=Multiply(A=@n19, B=@n20)
    B=CollectionParameter() {collection=/Landmass/Landscape/BlueprintBrushes/MPC/MPC_Landscape, parameterName=LandscapeZScale}
   )) {name=World Depth Displacement, nodeColor=<rgba 0 0.7187 1 1>}
  n21 = ScalarParameter() {defaultValue=1, parameterName=FalloffRelativeHeight}
  n22 = ScalarParameter() {defaultValue=4096, parameterName=BrushSize}
  NamedRerouteDeclaration(None=Divide(
    A=Multiply(A=@n21, B=@n22)
    B=CollectionParameter() {collection=&MPC_Landscape, parameterName=LandscapeZScale}
   )) {name=World Depth for Falloff, nodeColor=<rgba 1 0.4922 0 1>}
  n23 = NamedRerouteUsage()
  NamedRerouteDeclaration(None=Saturate(None=Ceil(None=@n23))) {name=Mask For MinMax, nodeColor=<rgba 1 0.7031 0 1>}
 ]}
