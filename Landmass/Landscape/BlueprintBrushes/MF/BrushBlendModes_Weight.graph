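# Material function for Landmass Blueprint brushes that blends a brush's masked weight into the existing landscape weightmap.
# It reads the previous weight from the point-sampled WeightmapRT red channel and builds lerp and additive candidates from reroutes.
# BrushBlendModes_Manual combines these, with AlphaBlend, Min and Max all receiving the same lerp.
# A saturated edge falloff mask, bypassed by DrawToEntireLandscape, drives the final lerp to that result.
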MaterialFunction /Landmass/Landscape/BlueprintBrushes/MF/BrushBlendModes_Weight
graph {nodes [
  n1 = LinearInterpolate(A=NamedRerouteUsage(), B=NamedRerouteUsage(), Alpha=NamedRerouteUsage())
  n2 = Add(A=NamedRerouteUsage(), B=Multiply(A=NamedRerouteUsage(), B=NamedRerouteUsage()))
  n3 = Saturate(None=MaterialFunctionCall!(UVs, Falloff Width=Constant() {r=0.001}, Falloff Radius=Constant(), Offset) {materialFunction=&/LandmassBrush_EdgeFalloff}.Edge Mask)
  n4 = ScalarParameter() {parameterName=DrawToEntireLandscape}
  FunctionOutput(None=LinearInterpolate(
    A=NamedRerouteUsage()
    B=MaterialFunctionCall!(AlphaBlend=@n1, Min=@n1, Max=@n1, Additive=@n2) {materialFunction=/Landmass/Landscape/OldPrototype_BP/Mats/MF/BrushBlendModes_Manual}.Result
    Alpha=LinearInterpolate(A=@n3, Alpha=@n4)
   ))
  NamedRerouteDeclaration(None=TextureSampleParameter2D(UVs=MaterialFunctionCall() {materialFunction=&/PointSampleLandscapeUVs}.TexelCenterUVs) {parameterName=WeightmapRT, texture=/Engine/EngineResources/Black_Low.Black}.R) {name=Previous Weightmap, nodeColor=<rgba 0 0.4609 1 1>}
  NamedRerouteDeclaration(None=FunctionInput() {bUsePreviewValueAsDefault=true, inputName=Weightmap Mask, inputType=FunctionInput_Scalar}) {name=Weightmap Texture, nodeColor=<rgba 1 0.7969 0 1>}
  NamedRerouteDeclaration(None=FunctionInput() {bUsePreviewValueAsDefault=true, inputName=Mask, inputType=FunctionInput_Scalar, sortPriority=1}) {name=Mask, nodeColor=<rgba 0 0.9531 1 1>}
 ]}
