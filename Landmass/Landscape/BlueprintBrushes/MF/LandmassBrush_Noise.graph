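# Landmass material function that outputs layered 3D noise for landscape blueprint brushes, with Position, Octaves, Persistence and Turbulence inputs.
# A Custom HLSL loop sums Noise3D_Multiplexer octaves, doubling frequency and scaling amplitude by Persistence, with Octaves capped at 12.
# Turbulence sums absolute values; otherwise the signed sum is remapped toward 0..1 without clamping.
# The noise type is hardwired to gradient, and the edge-seed blending mentioned in its description is absent from this graph.
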MaterialFunction /Landmass/Landscape/BlueprintBrushes/MF/LandmassBrush_Noise
graph {
 properties {
  description="This function allows more advanced sampling of Noise including control of Persistence and Octaves as well as blending to a common Edge Seed if desired. "
  libraryCategoriesText [Landmass]
 }
 nodes [
  MaterialExpressionReroute_0 = Reroute(None=NamedRerouteDeclaration(None=FunctionInput() {bUsePreviewValueAsDefault=true, inputName=Position}) {name=Position, nodeColor=<rgba 0 0 0 1>})
  n1 = FunctionInput() {bUsePreviewValueAsDefault=true, inputName=Octaves, inputType=FunctionInput_Scalar, sortPriority=1, previewValue=<10 0 0 1>}
  MaterialExpressionMin_0 = Min(A=NamedRerouteDeclaration(None=@n1) {name=Octaves, nodeColor=<rgba 0 0 0 1>}) {constB=12}
  n2 = FunctionInput() {bUsePreviewValueAsDefault=true, inputName=Persistence, inputType=FunctionInput_Scalar, sortPriority=3, previewValue=<0.4 0 0 1>}
  MaterialExpressionNamedRerouteDeclaration_3 = NamedRerouteDeclaration(None=@n2) {name=Persistence, nodeColor=<rgba 0 0 0 1>}
  n3 = Constant()
  MaterialExpressionReroute_1 = Reroute(None=NamedRerouteDeclaration(None=FunctionInput(Preview=@n3) {bUsePreviewValueAsDefault=true, inputName=Turbulence, inputType=FunctionInput_Scalar, sortPriority=4}) {name=Turbulence, nodeColor=<rgba 0 0 0 1>})
  MaterialExpressionReroute_2 = Reroute(None=Constant() {desc="3 = Gradient, 5 = Voronoi", r=3})
  n4 = Custom!(
   Position=@MaterialExpressionReroute_0
   Octaves=@MaterialExpressionMin_0
   Persistence=@MaterialExpressionNamedRerouteDeclaration_3
   bTurbulence=@MaterialExpressionReroute_1
   Function=@MaterialExpressionReroute_2
  ) {
   code="\r\nfloat Out = 0;\r\nfloat OutScale = 1;\r\n\r\n[loop]\r\nfor(int i = 0; i < Octaves; ++i)\r\n{\r\n\tif(bTurbulence)\r\n\t{\r\n\t\tOut += abs(Noise3D_Multiplexer(Function, Position, 4, 0, 512)) * OutScale;\r\n\t}\r\n\telse\r\n\t{\r\n\t\tOut += Noise3D_Multiplexer(Function, Position, 4, 0, 512) * OutScale;\r\n\t}\r\n\r\n\tPosition *= 2;\r\n\tOutScale *= Persistence;\r\n\r\n}\r\n\r\nif(!bTurbulence)\r\n{\r\n\t// bring -1..1 to 0..1 range\r\n\tOut = Out * 0.5f + 0.5f;\r\n}\r\n\r\n// Out is in 0..1 range\r\nreturn lerp(0, 1, Out);"
   description=Custom Noise
   outputType=CMOT_Float1
   inputs [
    {inputName=Position, input {expression=&self:MaterialExpressionReroute_0}}
    {inputName=Octaves, input {expression=&self:MaterialExpressionMin_0}}
    {
     inputName=Persistence
     input {expression=&self:MaterialExpressionNamedRerouteDeclaration_3}
    }
    {inputName=bTurbulence, input {expression=&self:MaterialExpressionReroute_1}}
    {inputName=Function, input {expression=&self:MaterialExpressionReroute_2}}
   ]
  }
  FunctionOutput(None=@n4)
 ]
}
