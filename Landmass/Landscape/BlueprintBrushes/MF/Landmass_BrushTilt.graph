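# Landmass brush material function that outputs a Tilt value, intended to rotate a landscape brush along the X and Y axes.
# It appends the red channel of BrushTiltTangents to itself, divides by the landscape Z scale from MPC_Landscape and multiplies by BrushSize.
# The result is dotted with centered brush UVs (0.5 minus UV), giving a linear ramp.
# A static Enable switch, on by default, outputs a constant instead when off.
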MaterialFunction /Landmass/Landscape/BlueprintBrushes/MF/Landmass_BrushTilt
graph {nodes [
  n1 = VectorParameter() {parameterName=BrushTiltTangents, defaultValue=<rgba 0 0 1 0>}
  n2 = Multiply(
   A=Divide(
    A=AppendVector(A=@n1.R, B=@n1.R)
    B=CollectionParameter() {collection=/Landmass/Landscape/BlueprintBrushes/MPC/MPC_Landscape, parameterName=LandscapeZScale}
   )
   B=ScalarParameter() {defaultValue=4096, parameterName=BrushSize}
  )
  n3 = Subtract(B=MaterialFunctionCall() {materialFunction=&/LandmassBrush_UVs}.UVs) {constA=0.5, constB=0.5}
  n4 = StaticBool() {value=true}
  FunctionOutput(None=StaticSwitch(
    True=DotProduct(A=@n2, B=@n3)
    False=Constant()
    Value=FunctionInput(Preview=@n4) {bUsePreviewValueAsDefault=true, description=Defaults to True, inputName=Enable, inputType=FunctionInput_StaticBool}
   )) {description=Allows rotating Brush along the X and Y axes as well., outputName=Tilt}
 ]}
