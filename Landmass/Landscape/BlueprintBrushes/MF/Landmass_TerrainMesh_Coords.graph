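# Landmass material function that gives terrain mesh materials landscape-wide UVs and world XY position, using MPC_Landscape parameters.
# It finds each component's grid index from the mesh origin relative to the landscape, passed through a vertex interpolator for pixel shaders.
# That index offsets local UVs in texels, with half-texel correction, before division by render target resolution.
# World XY rescales those UVs by landscape quads and scale, adds location and subtracts a scale-based offset.
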MaterialFunction /Landmass/Landscape/BlueprintBrushes/MF/Landmass_TerrainMesh_Coords
graph {nodes [
  n1 = CollectionParameter() {collection=/Landmass/Landscape/BlueprintBrushes/MPC/MPC_Landscape, parameterName=LSQuadsX}
  n2 = CollectionParameter() {collection=&MPC_Landscape, parameterName=LSQuadsY}
  n3 = CollectionParameter() {collection=&MPC_Landscape, parameterName=LandscapeScale}
  n4 = Multiply(A=NamedRerouteUsage(), B=Multiply(A=AppendVector(A=@n1, B=@n2), B=ComponentMask(None=@n3) {r=true}))
  n5 = ComponentMask(None=CollectionParameter() {collection=&MPC_Landscape, parameterName=LandscapeLocation}) {g=true, r=true}
  n6 = ComponentMask(None=CollectionParameter() {collection=&MPC_Landscape, parameterName=LandscapeScale}) {r=true}
  FunctionOutput(None=Subtract(A=Add(A=@n4, B=@n5), B=Divide(A=@n6))) {outputName=World Position XY, sortPriority=1}
  n7 = CollectionParameter() {collection=&MPC_Landscape, parameterName=ComponentRes}
  n8 = Multiply(A=TextureCoordinate(), B=Divide(A=Subtract(A=@n7), B=@n7))
  n9 = CollectionParameter() {collection=&MPC_Landscape, parameterName=ComponentRes}
  n10 = TransformPosition(None=Constant3Vector()) {transformType=TRANSFORMPOSSOURCE_World}
  n11 = ShaderStageSwitch(PixelShader=VertexInterpolator(VS=@n10).PS, VertexShader=@n10)
  n12 = ComponentMask(None=CollectionParameter() {collection=&MPC_Landscape, parameterName=LandscapeLocation}) {b=true, g=true, r=true}
  n13 = CollectionParameter() {collection=&MPC_Landscape, parameterName=LandscapeScale}
  n14 = CollectionParameter() {collection=&MPC_Landscape, parameterName=ComponentRes}
  n15 = Divide(
   A=ComponentMask(None=Subtract(A=@n11, B=@n12)) {g=true, r=true}
   B=Multiply(A=ComponentMask(None=@n13) {r=true}, B=Subtract(A=@n14) {desc=Component Quads}) {desc=Size Per Component}
  )
  n16 = CollectionParameter() {collection=&MPC_Landscape, parameterName=ComponentRes}
  n17 = Add(A=Add(A=Multiply(A=@n8, B=@n9)) {constB=0.5}, B=Multiply(A=Floor(None=@n15), B=Subtract(A=@n16) {desc=Component Quads}))
  n18 = AppendVector(
   A=CollectionParameter() {collection=&MPC_Landscape, parameterName=RTResX}
   B=CollectionParameter() {collection=&MPC_Landscape, parameterName=RTResY}
  )
  FunctionOutput(None=NamedRerouteDeclaration(None=Divide(A=@n17, B=@n18)) {name=UVs, nodeColor=<rgba 1 0.2109 0 1>}) {outputName=Landscape UVs, sortPriority=1}
 ]}
